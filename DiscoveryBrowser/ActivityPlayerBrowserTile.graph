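# ActivityPlayerBrowserTile is a Widget Blueprint for a tile in the Discovery browser, derived from the native FortActivityPlayerBrowserTile class.
# Its root is an ActivityBrowserTileDetailsDisplay widget, bound as Display_TileDetails, with the details button shown.
# The inherited last-played date text and input action widget slots have no widget assigned here.
# The event graph and variable list are empty, so the Blueprint adds no logic of its own.
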WidgetBlueprint /DiscoveryBrowser/ActivityPlayerBrowserTile
graphs [EventGraph {links [], nodes []}]
variables []
widgets {
 info {
  inheritedWidgetCount=3
  parentClass=/Script/DiscoveryBrowserUI.FortActivityPlayerBrowserTile
  rootWidgetClass=&/ActivityBrowserTileDetailsDisplay.ActivityBrowserTileDetailsDisplay_C
  widgetCount=3
 }
 widgets [
  Display_TileDetails {
   widgetClassPath=&/ActivityBrowserTileDetailsDisplay
   bInherited=true
   bIsVariable=true
   widget=Display_TileDetails
   properties {bShowDetailsButton=true}
  }
  Text_LastPlayedDate {widgetClassPath=&CommonTextBlock, bInherited=true, widget=null}
  InputActionWidget {widgetClassPath=&CommonActionWidget, bInherited=true, widget=null}
 ]
}
