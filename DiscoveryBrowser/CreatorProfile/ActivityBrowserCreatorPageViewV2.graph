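# Discovery browser creator profile page, a widget on FortActivityCreatorPageView listing a creator's islands in a tile view under the creator name, with input-dependent back/close buttons.
# A switcher shows the grid, a "Currently Unavailable" message when the list is empty, or "No Additional Islands" when no creator content is found.
# A throbber animation plays on construct and reverses when the query finishes.
# Each initialized tile resets grid entry height from its tile size plus vertical padding.
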WidgetBlueprint /DiscoveryBrowser/CreatorProfile/ActivityBrowserCreatorPageViewV2
animations [ShowAndHideRowLoadingSpinner {
  binding_count=1
  length_seconds=5
  object=&self:ShowAndHideRowLoadingSpinner
  sequence {bindings [RowLoadingSpinner {
     id=#1
     possessed_class=Image
     tracks [
      {
       type=MovieSceneWidgetMaterialTrack
       sections [{
         type=MovieSceneParameterSection
         end_frame=100
         start_frame=0
         channels [
          Solid_Animate {type=MovieSceneScriptingFloatChannel, keys [[6.4,1],[12.8,0]]}
          Fill_Animate {type=MovieSceneScriptingFloatChannel, keys [[6.4,1],[12.8,0]]}
         ]
        }]
      }
      {
       type=MovieSceneFloatTrack
       display_name=Render Opacity
       sections [{type=MovieSceneFloatSection, end_frame=6, start_frame=1, channels [None {type=MovieSceneScriptingFloatChannel, keys [[1.6,0],[6.4,1]]}]}]
      }
      {
       type=MovieScene2DTransformTrack
       sections [{
         type=MovieScene2DTransformSection
         end_frame=6
         start_frame=1
         channels [
          Translation.X {type=MovieSceneScriptingFloatChannel}
          Translation.Y {type=MovieSceneScriptingFloatChannel}
          Angle {type=MovieSceneScriptingFloatChannel}
          Scale.X {type=MovieSceneScriptingFloatChannel, keys [[1.6,0.875],[6.4,1]]}
          Scale.Y {type=MovieSceneScriptingFloatChannel, keys [[1.6,0.875],[6.4,1]]}
          Shear.X {type=MovieSceneScriptingFloatChannel}
          Shear.Y {type=MovieSceneScriptingFloatChannel}
         ]
        }]
      }
      {
       type=MovieSceneByteTrack
       display_name=Visibility
       sections [{type=MovieSceneByteSection, end_frame=6, start_frame=0, channels [None {type=MovieSceneScriptingByteChannel, keys [[0,1],[1.6,3]]}]}]
      }
     ]
    }]}
 }]
graphs [EventGraph {nodes [
   n1 = "On Entry Initialized (TileView_PlayerActivities)"!() -> (OutputDelegate, then, Item:Object Reference, Widget:User Widget Object Reference) {is=title K2Node_ComponentBoundEvent}
   > n2 = Cast To ActivityPlayerBrowserTile!(execute, Object=@n1.Widget) -> (then, CastFailed, AsActivity Player Browser Tile:Activity Player Browser Tile Object Reference) {is=title K2Node_DynamicCast}
   > n3 = Cast To ActivityBrowserTileDetailsDisplay!(
    execute
    Object=Get Display_TileDetails!(self:Fort Activity Player Browser Tile Object Reference=@n2) -> (Display_TileDetails:Fort Activity Tile Details Display Object Reference)
   ) -> (then, CastFailed, AsActivity Browser Tile Details Display:Activity Browser Tile Details Display Object Reference) {is=title K2Node_DynamicCast}
   n4 = Get TileSize!(self:Activity Browser Tile Details Display Object Reference=@n3) -> (TileSize:Vector 2D Structure)
   > SetEntryHeight!(
    execute
    self:Tile View Object Reference=Get TileView_PlayerActivities() -> (TileView_PlayerActivities:Fort Activity Tile View Object Reference)
    NewHeight:float=float + float(A=Get VerticalTilePadding() -> (VerticalTilePadding:double), B=BreakVector2D(InVec=@n4).Y)
   ) -> (then)
   > SetExtraVerticalPaddingOffset!(execute, self:Activity Browser Tile Details Display Object Reference=@n3, InVerticalTilePadding:double=Get VerticalTilePadding()) -> (then)
   Event OnCreatorActivitiesQueryFinished() {is=title K2Node_Event}
   > PlayAnimationReverse(InAnimation=Get ShowAndHideRowLoadingSpinner() -> (ShowAndHideRowLoadingSpinner:Widget Animation Object Reference))
   Event OnNoContentFoundForCreator() {is=title K2Node_Event}
   > SetActiveWidgetIndex(self=Get WidgetSwitcher_Results() -> (WidgetSwitcher_Results:Common Activatable Widget Switcher Object Reference), Index=2)
   Event On Tile View Updated() {is=title K2Node_Event}
   n5 = Get TileView_PlayerActivities()
   > n6 = Branch(Condition="integer > integer"(A=GetNumItems!(self:List View Object Reference=@n5) -> (ReturnValue:int=0)))
   n7 = Get WidgetSwitcher_Results()
   > SetActiveWidgetIndex(self=@n7, Index=1)
   SetActiveWidgetIndex(execute=@n6.else, self=@n7, Index=0)
   Event PreConstruct()
   > SetScrollbarVisibility!(execute, self:List View Base Object Reference=Get TileView_PlayerActivities(), InVisibility:ESlateVisibility Enum=Collapsed) -> (then)
   Event Construct()
   > PlayAnimationForward(InAnimation=Get ShowAndHideRowLoadingSpinner())
   > SetActiveWidgetIndex(self=Get WidgetSwitcher_Results(), Index=1)
  ]}]
variables [VerticalTilePadding {type=number, default=126}]
widgets {
 info {
  inheritedWidgetCount=7
  parentClass=/Script/DiscoveryBrowserUI.FortActivityCreatorPageView
  rootWidgetClass=&Overlay
  widgetCount=26
 }
 widgets [
  Overlay_539 {widgetClassPath=&Overlay, widget=Overlay_539}
  Widget_BlurBackground {
   widgetClassPath=&BackgroundBlur
   parent=Overlay_539
   widget=Widget_BlurBackground
   properties {
    blurStrength=8
    lowQualityFallbackBrush {
     bIsDynamicallyLoaded=false
     drawAs=NoDrawType
     imageType=NoImage
     mirroring=NoMirror
     resourceName=None
     resourceObject=null
     tiling=NoTile
     imageSize=<0 0>
     margin=<ltrb 0 0 0 0>
     outlineSettings {
      bUseBrushTransparency=false
      roundingType=HalfHeightRadius
      width=0
      color {colorUseRule=UseColor_Specified, specifiedColor=<rgba 0 0 0 0>}
      cornerRadii=<0 0 0 0>
     }
     tintColor {colorUseRule=UseColor_Specified, specifiedColor=<rgba 0 0 0 0.3>}
     uVRegion {bIsValid=false, max=<0 0>, min=<0 0>}
    }
   }
   slot {class=OverlaySlot, properties {horizontalAlignment=HAlign_Fill, verticalAlignment=VAlign_Fill}}
  }
  Lightbox {
   widgetClassPath=&Image
   parent=Overlay_539
   widget=Lightbox
   properties {
    visibility=SelfHitTestInvisible
    brush {
     bIsDynamicallyLoaded=false
     drawAs=Image
     imageType=NoImage
     mirroring=NoMirror
     resourceName=None
     resourceObject=null
     tiling=NoTile
     imageSize=<32 32>
     margin=<ltrb 0 0 0 0>
     outlineSettings {
      bUseBrushTransparency=false
      roundingType=HalfHeightRadius
      width=0
      color {colorUseRule=UseColor_Specified, specifiedColor=<rgba 0 0 0 0>}
      cornerRadii=<0 0 0 0>
     }
     tintColor {colorUseRule=UseColor_Specified, specifiedColor=<rgba 0 0 0 1>}
     uVRegion {bIsValid=false, max=<0 0>, min=<0 0>}
    }
    colorAndOpacity=<rgba 1 1 1 0.75>
   }
   slot {class=OverlaySlot, properties {horizontalAlignment=HAlign_Fill, verticalAlignment=VAlign_Fill}}
  }
  Image_Background {
   widgetClassPath=&Image
   parent=Overlay_539
   widget=Image_Background
   properties {brush {
     bIsDynamicallyLoaded=false
     drawAs=Image
     imageType=NoImage
     mirroring=NoMirror
     resourceName=None
     resourceObject=/FortniteCoreUI/UI/DiscoveryShared/Art/MaterialInstances/MI_UI_Discovery_Background
     tiling=NoTile
     imageSize=<32 32>
     margin=<ltrb 0 0 0 0>
     outlineSettings {
      bUseBrushTransparency=false
      roundingType=HalfHeightRadius
      width=0
      color {colorUseRule=UseColor_Specified, specifiedColor=<rgba 0 0 0 0>}
      cornerRadii=<0 0 0 0>
     }
     tintColor {colorUseRule=UseColor_Specified, specifiedColor=<rgba 1 1 1 1>}
     uVRegion {bIsValid=false, max=<0 0>, min=<0 0>}
    }}
   slot {class=OverlaySlot, properties {horizontalAlignment=HAlign_Fill, verticalAlignment=VAlign_Fill}}
  }
  RowLoadingSpinner {
   widgetClassPath=&Image
   bIsVariable=true
   parent=Overlay_539
   widget=RowLoadingSpinner
   properties {
    visibility=SelfHitTestInvisible
    brush {
     bIsDynamicallyLoaded=false
     drawAs=Image
     imageType=NoImage
     mirroring=NoMirror
     resourceName=None
     resourceObject=/Game/UI/DiscoveryShared/Art/MI_UI_Discovery_Throbber
     tiling=NoTile
     imageSize=<192 192>
     margin=<ltrb 0 0 0 0>
     outlineSettings {
      bUseBrushTransparency=false
      roundingType=HalfHeightRadius
      width=0
      color {colorUseRule=UseColor_Specified, specifiedColor=<rgba 0 0 0 0>}
      cornerRadii=<0 0 0 0>
     }
     tintColor {colorUseRule=UseColor_Specified, specifiedColor=<rgba 1 1 1 1>}
     uVRegion {bIsValid=false, max=<0 0>, min=<0 0>}
    }
   }
   slot {class=OverlaySlot, properties {horizontalAlignment=HAlign_Center, verticalAlignment=VAlign_Center}}
  }
  SafeZone_PageContent {
   widgetClassPath=&SafeZone
   bIsVariable=true
   parent=Overlay_539
   widget=SafeZone_PageContent
   properties {padLeft=false, padRight=false}
   slot {class=OverlaySlot, properties {horizontalAlignment=HAlign_Fill, verticalAlignment=VAlign_Fill}}
  }
  VerticalBox_1 {widgetClassPath=&VerticalBox, parent=SafeZone_PageContent, widget=VerticalBox_1, slot {class=SafeZoneSlot}}
  SizeBox_53 {
   widgetClassPath=&SizeBox
   parent=VerticalBox_1
   widget=SizeBox_53
   properties {bOverride_MinDesiredWidth=true, minDesiredWidth=1728}
   slot {class=VerticalBoxSlot, properties {horizontalAlignment=HAlign_Center, padding=<ltrb 0 32 0 0>}}
  }
  Text_CategoryTitle {
   widgetClassPath=&CommonRichTextBlock
   bInherited=true
   bIsVariable=true
   parent=SizeBox_53
   widget=Text_CategoryTitle
   properties {
    bOverrideDefaultStyle=true
    text="{Creator Name}"
    textTransformPolicy=ToUpper
    defaultTextStyleOverride {
     overflowPolicy=Clip
     transformPolicy=None
     colorAndOpacity {colorUseRule=UseColor_Specified, specifiedColor=<rgba 1 1 1 1>}
     font {
      bForceMonospaced=false
      bMaterialIsStencil=false
      fontMaterial=null
      fontObject=&Burbank
      letterSpacing=0
      monospacedWidth=1
      size=42
      skewAmount=0
      typefaceFontName=BurbankRegular-Black
      outlineSettings {
       bApplyOutlineToDropShadows=false
       bMiteredCorners=false
       bSeparateFillAlpha=false
       outlineMaterial=null
       outlineSize=0
       outlineColor=<rgba 0 0 0 1>
      }
     }
     highlightColor {colorUseRule=UseColor_Specified, specifiedColor=<rgba 0 0 0 1>}
     highlightShape {
      bIsDynamicallyLoaded=false
      drawAs=Image
      imageType=NoImage
      mirroring=NoMirror
      resourceName=None
      resourceObject=null
      tiling=NoTile
      imageSize=<32 32>
      margin=<ltrb 0 0 0 0>
      outlineSettings {
       bUseBrushTransparency=false
       roundingType=HalfHeightRadius
       width=0
       color {colorUseRule=UseColor_Specified, specifiedColor=<rgba 0 0 0 0>}
       cornerRadii=<0 0 0 0>
      }
      tintColor {colorUseRule=UseColor_Specified, specifiedColor=<rgba 1 1 1 1>}
      uVRegion {bIsValid=false, max=<0 0>, min=<0 0>}
     }
     selectedBackgroundColor {colorUseRule=UseColor_Foreground, specifiedColor=<rgba 1 0 1 1>}
     shadowColorAndOpacity=<rgba 0 0 0 1>
     shadowOffset=<0 0>
     strikeBrush {
      bIsDynamicallyLoaded=false
      drawAs=Image
      imageType=NoImage
      mirroring=NoMirror
      resourceName=None
      resourceObject=null
      tiling=NoTile
      imageSize=<32 32>
      margin=<ltrb 0 0 0 0>
      outlineSettings {
       bUseBrushTransparency=false
       roundingType=HalfHeightRadius
       width=0
       color {colorUseRule=UseColor_Specified, specifiedColor=<rgba 0 0 0 0>}
       cornerRadii=<0 0 0 0>
      }
      tintColor {colorUseRule=UseColor_Specified, specifiedColor=<rgba 1 1 1 1>}
      uVRegion {bIsValid=false, max=<0 0>, min=<0 0>}
     }
     underlineBrush {
      bIsDynamicallyLoaded=false
      drawAs=Image
      imageType=NoImage
      mirroring=NoMirror
      resourceName=None
      resourceObject=null
      tiling=NoTile
      imageSize=<32 32>
      margin=<ltrb 0 0 0 0>
      outlineSettings {
       bUseBrushTransparency=false
       roundingType=HalfHeightRadius
       width=0
       color {colorUseRule=UseColor_Specified, specifiedColor=<rgba 0 0 0 0>}
       cornerRadii=<0 0 0 0>
      }
      tintColor {colorUseRule=UseColor_Specified, specifiedColor=<rgba 1 1 1 1>}
      uVRegion {bIsValid=false, max=<0 0>, min=<0 0>}
     }
    }
   }
   slot {class=SizeBoxSlot}
  }
  WidgetSwitcher_Results {
   widgetClassPath=&CommonActivatableWidgetSwitcher
   bIsVariable=true
   parent=VerticalBox_1
   widget=WidgetSwitcher_Results
   properties {activeWidgetIndex=1, clipping=ClipToBounds}
   slot {class=VerticalBoxSlot, properties {size {sizeRule=Fill, value=1}}}
  }
  WBP_Discovery_NoResultsMessage {
   widgetClassPath=/DiscoveryBrowser/WBP_Discovery_MessageNoResults.WBP_Discovery_NoResultsMessage
   bIsVariable=true
   parent=WidgetSwitcher_Results
   widget=WBP_Discovery_NoResultsMessage
   properties {flavorMessageBase=Check back later for new content., flavorTitleBase=Currently Unavailable}
   slot {class=WidgetSwitcherSlot, properties {verticalAlignment=VAlign_Top, padding=<ltrb 0 128 0 0>}}
  }
  SizeBox_78 {
   widgetClassPath=&SizeBox
   parent=WidgetSwitcher_Results
   widget=SizeBox_78
   properties {bOverride_MinDesiredWidth=true, minDesiredWidth=1728}
   slot {class=WidgetSwitcherSlot, properties {horizontalAlignment=HAlign_Center, padding=<ltrb 0 24 0 0>}}
  }
  TileView_PlayerActivities {
   widgetClassPath=/Script/DiscoveryBrowserUI.FortActivityTileView
   bInherited=true
   bIsVariable=true
   parent=SizeBox_78
   widget=TileView_PlayerActivities
   properties {
    bEnableFixedLineOffset=true
    clipping=Inherit
    entryHeight=288
    entryWidgetClass=/DiscoveryBrowser/ActivityPlayerBrowserTile.ActivityPlayerBrowserTile_C
    entryWidth=432
    horizontalEntrySpacing=24
    numDesignerPreviewEntries=13
    tileAlignment=LeftAligned
    verticalEntrySpacing=24
   }
   slot {class=SizeBoxSlot}
  }
  WBP_Discovery_NoResultsMessage_67 {
   widgetClassPath=&WBP_Discovery_NoResultsMessage
   bIsVariable=true
   parent=WidgetSwitcher_Results
   widget=WBP_Discovery_NoResultsMessage_67
   properties {flavorMessageBase=There are no additional Islands available from this creator., flavorTitleBase=No Additional Islands}
   slot {class=WidgetSwitcherSlot, properties {verticalAlignment=VAlign_Top, padding=<ltrb 0 400 0 0>}}
  }
  SZBottomButtons {
   widgetClassPath=&SafeZone
   parent=Overlay_539
   widget=SZBottomButtons
   properties {padTop=false}
   slot {class=OverlaySlot, properties {horizontalAlignment=HAlign_Right, verticalAlignment=VAlign_Bottom}}
  }
  Overlay_ButtonsAndBackerGradient {widgetClassPath=&Overlay, parent=SZBottomButtons, widget=Overlay_ButtonsAndBackerGradient, slot {class=SafeZoneSlot}}
  BgFooter {
   widgetClassPath=&Image
   parent=Overlay_ButtonsAndBackerGradient
   widget=BgFooter
   properties {
    visibility=SelfHitTestInvisible
    brush {
     bIsDynamicallyLoaded=false
     drawAs=Image
     imageType=NoImage
     mirroring=NoMirror
     resourceName=None
     resourceObject=/FortniteCoreUI/UI/DiscoveryShared/Art/MaterialInstances/MI_UI_Discovery_Gradient
     tiling=NoTile
     imageSize=<32 32>
     margin=<ltrb 0 0 0 0>
     outlineSettings {
      bUseBrushTransparency=false
      roundingType=HalfHeightRadius
      width=0
      color {colorUseRule=UseColor_Specified, specifiedColor=<rgba 0 0 0 0>}
      cornerRadii=<0 0 0 0>
     }
     tintColor {colorUseRule=UseColor_Specified, specifiedColor=<rgba 1 1 1 1>}
     uVRegion {bIsValid=false, max=<0 0>, min=<0 0>}
    }
   }
   slot {class=OverlaySlot, properties {horizontalAlignment=HAlign_Fill, verticalAlignment=VAlign_Fill}}
  }
  HBBottomButtons {
   widgetClassPath=&HorizontalBox
   parent=Overlay_ButtonsAndBackerGradient
   widget=HBBottomButtons
   slot {class=OverlaySlot, properties {horizontalAlignment=HAlign_Right, verticalAlignment=VAlign_Bottom, padding=<ltrb 32 32 32 32>}}
  }
  Button_BackToTop {
   widgetClassPath=/FortUILibrary/UIKit/Buttons/WBP_UIKit_Button_Quiet
   bInherited=true
   bIsVariable=true
   parent=HBBottomButtons
   widget=Button_BackToTop
   properties {
    minWidth=120
    text=Back to Top
    visibility=Collapsed
    triggeringInputAction {dataTable=/FortUILibrary/Foundation/ActivityBrowserInputActionTable, rowName=BackToTop}
   }
   slot {class=HorizontalBoxSlot, properties {verticalAlignment=VAlign_Bottom, padding=<ltrb 32 0 0 0>}}
  }
  Button_MobileShowGameDetails {
   widgetClassPath=&WBP_UIKit_Button_Quiet
   bInherited=true
   bIsVariable=true
   parent=HBBottomButtons
   widget=Button_MobileShowGameDetails
   properties {
    minWidth=120
    text=Game Details
    triggeringInputAction {dataTable=&ActivityBrowserInputActionTable, rowName=ShowGameDetails}
   }
   slot {class=HorizontalBoxSlot, properties {verticalAlignment=VAlign_Bottom, padding=<ltrb 32 0 0 0>}}
  }
  VisibilityBorder_NonTouch {
   widgetClassPath=&CommonHardwareVisibilityBorder
   parent=HBBottomButtons
   widget=VisibilityBorder_NonTouch
   properties {visibilityQuery {
     autoDescription=" NONE( Input.Touch )"
     tokenStreamVersion=0
     userDescription=""
     queryTokenStream [0,1,3,1,0]
     tagDictionary [{tagName=Input.Touch}]
    }}
   slot {class=HorizontalBoxSlot, properties {verticalAlignment=VAlign_Bottom, padding=<ltrb 32 0 0 0>}}
  }
  Button_Back {
   widgetClassPath=&WBP_UIKit_Button_Quiet
   bInherited=true
   bIsVariable=true
   parent=VisibilityBorder_NonTouch
   widget=Button_Back
   properties {minWidth=120, mobileMinWidth=200, text=Back, triggeringInputAction {dataTable=&InputActionDataTable, rowName=Back}}
   slot {class=BorderSlot, properties {padding=<ltrb 0 0 0 0>}}
  }
  Button_MobileAccept {
   widgetClassPath=/FortUILibrary/UIKit/Buttons/WBP_UIKit_Button_Loud
   bInherited=true
   bIsVariable=true
   parent=HBBottomButtons
   widget=Button_MobileAccept
   properties {
    buttonSize=Small
    minHeight=80
    minWidth=120
    mobileMinWidth=200
    text=Accept
    triggeringInputAction {dataTable=&ActivityBrowserInputActionTable, rowName=SelectTile}
   }
   slot {class=HorizontalBoxSlot, properties {verticalAlignment=VAlign_Bottom, padding=<ltrb 32 0 0 0>}}
  }
  VisibilityBorder_Touch {
   widgetClassPath=&CommonHardwareVisibilityBorder
   parent=Overlay_539
   widget=VisibilityBorder_Touch
   properties {visibilityQuery {
     autoDescription=" ANY( Input.Touch )"
     tokenStreamVersion=0
     userDescription=""
     queryTokenStream [0,1,1,1,0]
     tagDictionary [{tagName=Input.Touch}]
    }}
   slot {class=OverlaySlot, properties {horizontalAlignment=HAlign_Right, padding=<ltrb 0 32 32 0>}}
  }
  SafeZone_ButtonCloseTouch {
   widgetClassPath=&SafeZone
   parent=VisibilityBorder_Touch
   widget=SafeZone_ButtonCloseTouch
   properties {padBottom=false, padLeft=false}
   slot {class=BorderSlot, properties {padding=<ltrb 0 0 0 0>}}
  }
  Button_CloseTouch {
   widgetClassPath=/FortUILibrary/Foundation/Buttons/CloseButton
   bInherited=true
   parent=SafeZone_ButtonCloseTouch
   widget=Button_CloseTouch
   properties {
    onButtonBaseClicked=(&self:WidgetTree.Button_CloseTouch.HandleOnButtonClicked)
    onButtonBaseDoubleClicked=(&self:WidgetTree.Button_CloseTouch.HandleOnButtonDoubleClicked)
    onButtonBaseHovered=(&self:WidgetTree.Button_CloseTouch.HandleOnButtonHovered)
    onButtonBaseUnhovered=(&self:WidgetTree.Button_CloseTouch.HandleOnButtonUnhovered)
    onSelectedChangedBase=(&self:WidgetTree.Button_CloseTouch.HandleOnSelectedChanged)
    paddingOveride=false
   }
   slot {class=SafeZoneSlot, properties {padding=<ltrb 14 0 0 0>}}
  }
 ]
}
