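# Widget Blueprint for an island tile in the Discovery browser's player/creator page, derived from the native FortActivityPlayerBrowserTile.
# Its root is a WBP_IslandTile_Full bound as Display_TileDetails, flagged as running old Discover code and sitting in a creator page.
# The inherited last-played date text and input action widget slots are left unfilled.
# The event graph is empty, so behavior comes from the parent class and the nested tile widget.
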WidgetBlueprint /DiscoveryBrowser/ActivityPlayerBrowserTile_New_VM
graphs [EventGraph {links [], nodes []}]
variables []
widgets {
 info {
  inheritedWidgetCount=3
  parentClass=/Script/DiscoveryBrowserUI.FortActivityPlayerBrowserTile
  rootWidgetClass=/DiscoveryBrowser/Tiles/Islands/WBP_IslandTile_Full.WBP_IslandTile_Full_C
  widgetCount=3
 }
 widgets [
  Display_TileDetails {
   widgetClassPath=/DiscoveryBrowser/Tiles/Islands/WBP_IslandTile_Full
   bInherited=true
   bIsVariable=true
   widget=Display_TileDetails
   properties {isOldDiscoverCode=true, isTileInCreatorPage=true}
  }
  Text_LastPlayedDate {widgetClassPath=&CommonTextBlock, bInherited=true, widget=null}
  InputActionWidget {widgetClassPath=&CommonActionWidget, bInherited=true, widget=null}
 ]
}
