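# Discovery browser page widget on FortActivityCreatorPageView that shows a creator's islands in a tile view under the creator name.
# A widget switcher flips between tiles and two no-content messages, with a throbber until the activities query finishes.
# Generated entries size their inner tile to the list's entry size and report context menu open and close.
# Scroll offset, input method and context menu state drive Back and Back to Top visibility.
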WidgetBlueprint /DiscoveryBrowser/CreatorProfile/ActivityBrowserCreatorPageView_NEW_VM
graphs [
 EventGraph {nodes [
   n1 = "On List View Scrolled (TileView_PlayerActivities)"!() -> (OutputDelegate, then, ItemOffset:float, DistanceRemaining:float) {is=title K2Node_ComponentBoundEvent}
   > Set ListViewItem Offset(ListViewItem Offset:float=@n1.ItemOffset) -> (Output_Get:float)
   > Set Visibility Action Bar Buttons() {is=title K2Node_CallFunction}
   n2 = "On Entry Generated (TileView_PlayerActivities)"!() -> (OutputDelegate, then, Widget:User Widget Object Reference) {is=title K2Node_ComponentBoundEvent}
   > n3 = Cast To ActivityPlayerBrowserTile_New_VM!(execute, Object=@n2.Widget) -> (then, CastFailed, AsActivity Player Browser Tile New VM:Activity Player Browser Tile New VM Object Reference) {is=title K2Node_DynamicCast}
   > n4 = Cast To WBP_IslandTile_Minimal!(
    execute
    Object=Get Display_TileDetails!(self:Fort Activity Player Browser Tile Object Reference=@n3) -> (Display_TileDetails:Fort Activity Tile Details Display Object Reference)
   ) -> (then, CastFailed, AsWBP Island Tile Minimal:WBP Island Tile Minimal Object Reference) {is=title K2Node_DynamicCast}
   n5 = Get TileView_PlayerActivities() -> (TileView_PlayerActivities:Fort Activity Tile View Object Reference)
   n6 = Get WBP_Discover_Tile_Base!(self:WBP Island Tile Minimal Object Reference=@n4) -> (WBP_Discover_Tile_Base:WBP Discover Tile Base Object Reference)
   > Set Tile Size!(
    execute
    self:WBP Discover Tile Base Object Reference=@n6
    InSize:Vector 2D Structure=MakeVector2D(
     X=GetEntryWidth!(self:Tile View Object Reference=@n5) -> (ReturnValue:float)
     Y=GetEntryHeight!(self:Tile View Object Reference=@n5) -> (ReturnValue:float)
    )
   ) -> (then) {is=title K2Node_CallFunction}
   n7 = Get WBP_UIKit_ContextMenuAnchorButton!(self:WBP Discover Tile Base Object Reference=@n6) -> (WBP_UIKit_ContextMenuAnchorButton:WBP UIKit Context Menu Anchor Button Object Reference)
   > Bind Event to On Context Menu Opened(self:WBP UIKit Context Menu Anchor Button Base Object Reference=@n7, Delegate=Create Event()) {is=title K2Node_AddDelegate}
   > Bind Event to On Context Menu Closed(self:WBP UIKit Context Menu Anchor Button Base Object Reference=@n7, Delegate=Create Event()) {is=title K2Node_AddDelegate}
   n8 = HandleInputMethodChanged!() -> (OutputDelegate, then, bNewInputType:ECommonInputType Enum) {is=title K2Node_CustomEvent}
   > Handle Input Changed!(execute, self, CurrentInputType:ECommonInputType Enum=@n8.bNewInputType) -> (then)
   ContextMenuOpened() {is=title K2Node_CustomEvent}
   > SetVisibility(
    self=[
     Get Button_Back() -> (Button_Back:Common Button Base Object Reference)
     Get Button_BackToTop() -> (Button_BackToTop:Common Button Base Object Reference)
    ]
    InVisibility=Collapsed
   )
   > Set IsContextMenuOpen(IsContextMenuOpen:bool=true) -> (Output_Get:bool)
   ContextMenuClosed() {is=title K2Node_CustomEvent}
   > Set Visibility Action Bar Buttons()
   > Set IsContextMenuOpen(IsContextMenuOpen=false)
   Event Destruct()
   n9 = Create Event()
   n10 = Get CommonInputSubsystem()
   > Unbind Event from On Input Method Changed(self=@n10, Delegate=@n9)
   Event OnCreatorActivitiesQueryFinished() {is=title K2Node_Event}
   > SetVisibility(self=Get WBP_UIKit_Throbber() -> (WBP_UIKit_Throbber:WBP UIKit Throbber Object Reference), InVisibility=Hidden)
   Event OnNoContentFoundForCreator() {is=title K2Node_Event}
   > SetActiveWidgetIndex(self=Get WidgetSwitcher_Results() -> (WidgetSwitcher_Results:Common Activatable Widget Switcher Object Reference), Index=2)
   Event On Tile View Updated() {is=title K2Node_Event}
   n11 = Get TileView_PlayerActivities()
   > n12 = Branch(Condition="integer > integer"(A=GetNumItems!(self:List View Object Reference=@n11) -> (ReturnValue:int=0)))
   n13 = Get WidgetSwitcher_Results()
   > SetActiveWidgetIndex(self=@n13, Index=1)
   SetActiveWidgetIndex(execute=@n12.else, self=@n13, Index=0)
   Event PreConstruct()
   > SetScrollbarVisibility!(execute, self:List View Base Object Reference=Get TileView_PlayerActivities(), InVisibility:ESlateVisibility Enum=Collapsed) -> (then)
   Event Construct()
   > Bind Event to On Input Method Changed(self=@n10, Delegate=@n9)
   > HandleInputMethodChanged!(execute, self, bNewInputType:ECommonInputType Enum=GetCurrentInputType(self=@n10)) -> (then) {is=title K2Node_CallFunction}
   > SetVisibility(self=Get WBP_UIKit_Throbber(), InVisibility=HitTestInvisible)
   > SetActiveWidgetIndex(self=Get WidgetSwitcher_Results(), Index=1)
   > SetVisibility(self=Get Button_BackToTop(), InVisibility=Collapsed)
  ]}
 Handle Input Changed {nodes [
   n1 = Handle Input Changed!() -> (then, CurrentInputType:ECommonInputType Enum) {is=title K2Node_FunctionEntry}
   > n2 = Branch(Condition=AND Boolean(A=Get IsContextMenuOpen() -> (IsContextMenuOpen:bool), B="Equal (Enum)"(A=@n1, B=MouseAndKeyboard)))
   > SetVisibility(self=[Get Button_Back(), Get Button_BackToTop()], InVisibility=Collapsed)
   n3 = Get IsContextMenuOpen()
   n4 = "Equal (Enum)"(A=@n1, B=Touch)
   n5 = Branch(execute=@n2.else, Condition=AND Boolean(A=NOT Boolean(A=@n4), B=NOT Boolean(A=@n3)))
   > Set Visibility Action Bar Buttons()
   SetVisibility(execute=@n5.else, self=Get Button_Back(), InVisibility=Collapsed)
  ]}
 Set Visibility Action Bar Buttons {nodes [
   Set Visibility Action Bar Buttons() {is=title K2Node_FunctionEntry}
   n1 = "float > float"(A=Get ListViewItem Offset() -> (ListViewItem Offset:float))
   > SetVisibility(self=Get Button_BackToTop(), InVisibility=Select(Option 0=Collapsed, Option 1=SelfHitTestInvisible, Index=@n1))
   n2 = Get CommonInputSubsystem()
   > n3 = Branch(Condition="Equal (Enum)"(A=GetCurrentInputType(self=@n2), B=Touch))
   SetVisibility(execute=@n3.else, self=Get Button_Back(), InVisibility=Select(Option 0=SelfHitTestInvisible, Option 1=Collapsed, Index=@n1))
  ]}
]
variables [
 IsContextMenuOpen {type=boolean}
 ListViewItem Offset {type=number}
 VerticalTilePadding {type=number, default=126}
]
widgets {
 info {
  inheritedWidgetCount=7
  namedSlotCount=1
  parentClass=/Script/DiscoveryBrowserUI.FortActivityCreatorPageView
  rootWidgetClass=&Overlay
  widgetCount=27
 }
 widgets [
  Overlay_539 {widgetClassPath=&Overlay, widget=Overlay_539}
  Background {
   widgetClassPath=&Image
   parent=Overlay_539
   widget=Background
   properties {
    visibility=SelfHitTestInvisible
    brush {
     bIsDynamicallyLoaded=false
     drawAs=Image
     imageType=NoImage
     mirroring=NoMirror
     resourceName=None
     resourceObject=/FortUILibrary/MaterialInstances/Background/Nebulous/MI_UI_Background_Nebulous
     tiling=NoTile
     imageSize=<32 32>
     margin=<ltrb 0 0 0 0>
     outlineSettings {
      bUseBrushTransparency=false
      roundingType=HalfHeightRadius
      width=0
      color {colorUseRule=UseColor_Specified, specifiedColor=<rgba 0 0 0 0>}
      cornerRadii=<0 0 0 0>
     }
     tintColor {colorUseRule=UseColor_Specified, specifiedColor=<rgba 1 1 1 1>}
     uVRegion {bIsValid=false, max=<0 0>, min=<0 0>}
    }
   }
   slot {class=OverlaySlot, properties {horizontalAlignment=HAlign_Fill, verticalAlignment=VAlign_Fill}}
  }
  SafeZone_PageContent {
   widgetClassPath=&SafeZone
   bIsVariable=true
   parent=Overlay_539
   widget=SafeZone_PageContent
   properties {padBottom=false}
   slot {class=OverlaySlot, properties {horizontalAlignment=HAlign_Fill, verticalAlignment=VAlign_Fill}}
  }
  Overlay_1 {widgetClassPath=&Overlay, parent=SafeZone_PageContent, widget=Overlay_1, slot {class=SafeZoneSlot}}
  VerticalBox_1 {
   widgetClassPath=&VerticalBox
   parent=Overlay_1
   widget=VerticalBox_1
   slot {class=OverlaySlot, properties {horizontalAlignment=HAlign_Fill, verticalAlignment=VAlign_Fill}}
  }
  SizeBox_53 {
   widgetClassPath=&SizeBox
   parent=VerticalBox_1
   widget=SizeBox_53
   properties {bOverride_MinDesiredWidth=true, minDesiredWidth=1728}
   slot {class=VerticalBoxSlot, properties {horizontalAlignment=HAlign_Center, padding=<ltrb 0 128 0 0>}}
  }
  Text_CategoryTitle {
   widgetClassPath=&CommonRichTextBlock
   bInherited=true
   bIsVariable=true
   parent=SizeBox_53
   widget=Text_CategoryTitle
   properties {
    bOverrideDefaultStyle=true
    text="{Creator Name}"
    textTransformPolicy=ToUpper
    defaultTextStyleOverride {
     overflowPolicy=Clip
     transformPolicy=None
     colorAndOpacity {colorUseRule=UseColor_Specified, specifiedColor=<rgba 1 1 1 1>}
     font {
      bForceMonospaced=false
      bMaterialIsStencil=false
      fontMaterial=null
      fontObject=&HeadingNow
      letterSpacing=-40
      monospacedWidth=1
      size=32.25
      skewAmount=0
      typefaceFontName=86 Bold Italic
      outlineSettings {
       bApplyOutlineToDropShadows=false
       bMiteredCorners=false
       bSeparateFillAlpha=false
       outlineMaterial=null
       outlineSize=0
       outlineColor=<rgba 0 0 0 1>
      }
     }
     highlightColor {colorUseRule=UseColor_Specified, specifiedColor=<rgba 0 0 0 1>}
     highlightShape {
      bIsDynamicallyLoaded=false
      drawAs=Image
      imageType=NoImage
      mirroring=NoMirror
      resourceName=None
      resourceObject=null
      tiling=NoTile
      imageSize=<32 32>
      margin=<ltrb 0 0 0 0>
      outlineSettings {
       bUseBrushTransparency=false
       roundingType=HalfHeightRadius
       width=0
       color {colorUseRule=UseColor_Specified, specifiedColor=<rgba 0 0 0 0>}
       cornerRadii=<0 0 0 0>
      }
      tintColor {colorUseRule=UseColor_Specified, specifiedColor=<rgba 1 1 1 1>}
      uVRegion {bIsValid=false, max=<0 0>, min=<0 0>}
     }
     selectedBackgroundColor {colorUseRule=UseColor_Foreground, specifiedColor=<rgba 1 0 1 1>}
     shadowColorAndOpacity=<rgba 0 0 0 1>
     shadowOffset=<0 0>
     strikeBrush {
      bIsDynamicallyLoaded=false
      drawAs=Image
      imageType=NoImage
      mirroring=NoMirror
      resourceName=None
      resourceObject=null
      tiling=NoTile
      imageSize=<32 32>
      margin=<ltrb 0 0 0 0>
      outlineSettings {
       bUseBrushTransparency=false
       roundingType=HalfHeightRadius
       width=0
       color {colorUseRule=UseColor_Specified, specifiedColor=<rgba 0 0 0 0>}
       cornerRadii=<0 0 0 0>
      }
      tintColor {colorUseRule=UseColor_Specified, specifiedColor=<rgba 1 1 1 1>}
      uVRegion {bIsValid=false, max=<0 0>, min=<0 0>}
     }
     underlineBrush {
      bIsDynamicallyLoaded=false
      drawAs=Image
      imageType=NoImage
      mirroring=NoMirror
      resourceName=None
      resourceObject=null
      tiling=NoTile
      imageSize=<32 32>
      margin=<ltrb 0 0 0 0>
      outlineSettings {
       bUseBrushTransparency=false
       roundingType=HalfHeightRadius
       width=0
       color {colorUseRule=UseColor_Specified, specifiedColor=<rgba 0 0 0 0>}
       cornerRadii=<0 0 0 0>
      }
      tintColor {colorUseRule=UseColor_Specified, specifiedColor=<rgba 1 1 1 1>}
      uVRegion {bIsValid=false, max=<0 0>, min=<0 0>}
     }
    }
   }
   slot {class=SizeBoxSlot}
  }
  WidgetSwitcher_Results {
   widgetClassPath=&CommonActivatableWidgetSwitcher
   bIsVariable=true
   parent=VerticalBox_1
   widget=WidgetSwitcher_Results
   properties {activeWidgetIndex=1, clipping=ClipToBounds}
   slot {class=VerticalBoxSlot, properties {size {sizeRule=Fill, value=1}}}
  }
  WBP_Discovery_NoResultsMessage {
   widgetClassPath=/DiscoveryBrowser/WBP_Discovery_MessageNoResults.WBP_Discovery_NoResultsMessage
   bIsVariable=true
   parent=WidgetSwitcher_Results
   widget=WBP_Discovery_NoResultsMessage
   properties {flavorMessageBase=Check back later for new content., flavorTitleBase=Currently Unavailable}
   slot {class=WidgetSwitcherSlot, properties {verticalAlignment=VAlign_Center, padding=<ltrb 0 0 0 80>}}
  }
  Overlay_0 {
   widgetClassPath=&Overlay
   parent=WidgetSwitcher_Results
   widget=Overlay_0
   slot {class=WidgetSwitcherSlot, properties {horizontalAlignment=HAlign_Center, padding=<ltrb 0 0 0 100>}}
  }
  SizeBox_78 {
   widgetClassPath=&SizeBox
   parent=Overlay_0
   widget=SizeBox_78
   properties {bOverride_MinDesiredWidth=true, minDesiredWidth=1760}
   slot {class=OverlaySlot}
  }
  TileView_PlayerActivities {
   widgetClassPath=/Script/DiscoveryBrowserUI.FortActivityTileView
   bInherited=true
   bIsVariable=true
   parent=SizeBox_78
   widget=TileView_PlayerActivities
   properties {
    bEnableFixedLineOffset=true
    bEntrySizeIncludesEntrySpacing=false
    clipping=Inherit
    entryHeight=258
    entryWidgetClass=/DiscoveryBrowser/ActivityPlayerBrowserTile_New_VM.ActivityPlayerBrowserTile_New_VM_C
    entryWidth=408
    horizontalEntrySpacing=24
    numDesignerPreviewEntries=13
    tileAlignment=LeftAligned
    verticalEntrySpacing=24
   }
   slot {class=SizeBoxSlot}
  }
  Image_HeaderSoftEdge {
   widgetClassPath=&Image
   bIsVariable=true
   parent=Overlay_0
   widget=Image_HeaderSoftEdge
   properties {brush {
     bIsDynamicallyLoaded=false
     drawAs=Image
     imageType=NoImage
     mirroring=NoMirror
     resourceName=None
     resourceObject=/DiscoveryBrowser/TabScreens/DiscoverHome/Materials/MI_UI_Discovery_NebulousBackgroundWithTopMask
     tiling=NoTile
     imageSize=<32 50>
     margin=<ltrb 0 0 0 0>
     outlineSettings {
      bUseBrushTransparency=false
      roundingType=HalfHeightRadius
      width=0
      color {colorUseRule=UseColor_Specified, specifiedColor=<rgba 0 0 0 0>}
      cornerRadii=<0 0 0 0>
     }
     tintColor {colorUseRule=UseColor_Specified, specifiedColor=<rgba 1 1 1 1>}
     uVRegion {bIsValid=false, max=<0 0>, min=<0 0>}
    }}
   slot {class=OverlaySlot, properties {horizontalAlignment=HAlign_Fill}}
  }
  WBP_Discovery_NoResultsMessage_67 {
   widgetClassPath=&WBP_Discovery_NoResultsMessage
   bIsVariable=true
   parent=WidgetSwitcher_Results
   widget=WBP_Discovery_NoResultsMessage_67
   properties {flavorMessageBase=There are no additional Islands available from this creator., flavorTitleBase=No Additional Islands}
   slot {class=WidgetSwitcherSlot, properties {verticalAlignment=VAlign_Center, padding=<ltrb 0 0 0 80>}}
  }
  WBP_UIKit_Throbber {
   widgetClassPath=/FortUILibrary/UIKit/Throbber/WBP_UIKit_Throbber
   bIsVariable=true
   parent=Overlay_1
   widget=WBP_UIKit_Throbber
   properties {displayBackground=false, throbberSize=Large}
   slot {class=OverlaySlot, properties {horizontalAlignment=HAlign_Center, verticalAlignment=VAlign_Center}}
  }
  VisibilityBorder_Touch {
   widgetClassPath=&CommonHardwareVisibilityBorder
   parent=Overlay_1
   widget=VisibilityBorder_Touch
   properties {visibilityQuery {
     autoDescription=" ANY( Input.Touch )"
     tokenStreamVersion=0
     userDescription=""
     queryTokenStream [0,1,1,1,0]
     tagDictionary [{tagName=Input.Touch}]
    }}
   slot {class=OverlaySlot, properties {horizontalAlignment=HAlign_Right, padding=<ltrb 0 32 32 0>}}
  }
  Button_CloseTouch {
   widgetClassPath=/FortUILibrary/Foundation/Buttons/CloseButton
   bInherited=true
   parent=VisibilityBorder_Touch
   widget=Button_CloseTouch
   properties {
    onButtonBaseClicked=(&self:WidgetTree.Button_CloseTouch.HandleOnButtonClicked)
    onButtonBaseDoubleClicked=(&self:WidgetTree.Button_CloseTouch.HandleOnButtonDoubleClicked)
    onButtonBaseHovered=(&self:WidgetTree.Button_CloseTouch.HandleOnButtonHovered)
    onButtonBaseUnhovered=(&self:WidgetTree.Button_CloseTouch.HandleOnButtonUnhovered)
    onSelectedChangedBase=(&self:WidgetTree.Button_CloseTouch.HandleOnSelectedChanged)
    paddingOveride=false
   }
   slot {class=BorderSlot, properties {padding=<ltrb 0 0 0 0>}}
  }
  WBP_CaptureForPostBufferUpdate {
   widgetClassPath=&WBP_CaptureForPostBufferUpdate
   bIsVariable=true
   parent=Overlay_539
   slot {class=OverlaySlot, properties {horizontalAlignment=HAlign_Fill, verticalAlignment=VAlign_Fill}}
  }
  Overlay_ButtonsAndBackerGradient {
   widgetClassPath=&Overlay
   parent=Overlay_539
   widget=Overlay_ButtonsAndBackerGradient
   slot {class=OverlaySlot, properties {horizontalAlignment=HAlign_Fill, verticalAlignment=VAlign_Bottom}}
  }
  BgFooter {
   widgetClassPath=&Image
   parent=Overlay_ButtonsAndBackerGradient
   widget=BgFooter
   properties {
    visibility=SelfHitTestInvisible
    brush {
     bIsDynamicallyLoaded=false
     drawAs=Image
     imageType=NoImage
     mirroring=NoMirror
     resourceName=None
     resourceObject=/DiscoveryBrowser/TabScreens/DiscoverHome/Materials/MI_UI_Discovery_NebulousBackgroundWithBottomMask
     tiling=NoTile
     imageSize=<32 32>
     margin=<ltrb 0 0 0 0>
     outlineSettings {
      bUseBrushTransparency=false
      roundingType=HalfHeightRadius
      width=0
      color {colorUseRule=UseColor_Specified, specifiedColor=<rgba 0 0 0 0>}
      cornerRadii=<0 0 0 0>
     }
     tintColor {colorUseRule=UseColor_Specified, specifiedColor=<rgba 1 1 1 1>}
     uVRegion {bIsValid=false, max=<0 0>, min=<0 0>}
    }
   }
   slot {class=OverlaySlot, properties {horizontalAlignment=HAlign_Fill, verticalAlignment=VAlign_Fill}}
  }
  WBP_AthenaBottomBarWidget {
   widgetClassPath=/Game/UI/Frontend/BottomBar/WBP_AthenaBottomBarWidget
   bIsVariable=true
   parent=Overlay_ButtonsAndBackerGradient
   widget=WBP_AthenaBottomBarWidget
   slot {class=OverlaySlot, properties {horizontalAlignment=HAlign_Right, verticalAlignment=VAlign_Bottom}}
  }
  WBBottomButtons {
   widgetClassPath=&WrapBox
   namedSlotHost=WBP_AthenaBottomBarWidget
   widget=WBBottomButtons
   properties {bExplicitWrapSize=true, horizontalAlignment=HAlign_Right, wrapSize=1500}
  }
  AthenaBoundActionBar_BottomBar {
   widgetClassPath=/Game/UI/Frontend/BottomBar/AthenaBoundActionBar_BottomBar
   parent=WBBottomButtons
   widget=AthenaBoundActionBar_BottomBar
   properties {actionButtonClass=&/WBP_Button_Quiet_NotFocusable.WBP_Button_Quiet_NotFocusable_C, entrySpacing=<0 0>}
   slot {class=WrapBoxSlot}
  }
  Button_BackToTop {
   widgetClassPath=/FortUILibrary/UIKit/Buttons/WBP_UIKit_Button_Quiet
   bInherited=true
   bIsVariable=true
   parent=WBBottomButtons
   widget=Button_BackToTop
   properties {
    bIsFocusable=false
    minWidth=120
    text=Back to Top
    visibility=Collapsed
    triggeringInputAction {dataTable=/FortUILibrary/Foundation/ActivityBrowserInputActionTable, rowName=BackToTop}
   }
   slot {class=WrapBoxSlot, properties {horizontalAlignment=HAlign_Right}}
  }
  Button_Back {
   widgetClassPath=&WBP_UIKit_Button_Quiet
   bInherited=true
   bIsVariable=true
   parent=WBBottomButtons
   widget=Button_Back
   properties {
    bIsFocusable=false
    minWidth=120
    mobileMinWidth=200
    text=Back
    triggeringInputAction {dataTable=&InputActionDataTable, rowName=Back}
   }
   slot {class=WrapBoxSlot, properties {horizontalAlignment=HAlign_Right}}
  }
  Button_MobileAccept {widgetClassPath=&CommonButtonBase, bInherited=true, widget=null}
  Button_MobileShowGameDetails {widgetClassPath=&CommonButtonBase, bInherited=true, widget=null}
 ]
}
