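# UI material that draws an animated, glowing wind streak for the Play section of the Discover creator profile.
# Twice-rotated UI UVs feed a wave SDF, producing wind and glow masks with a directional falloff.
# Time-scrolled electric noise, shaped by power and easing curves, breaks up the streak.
# A static switch picks oscillating Rift or two-parameter colors, output premultiplied for AlphaComposite with edge fade, opacity and vertex alpha.
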Material /DiscoveryBrowser/CreatorProfile/Materials/M_UI_Discover_CreatorProfile_PlaySectionWind
graph {
 properties {
  bAutomaticallySetUsageInEditor=false
  bUsedWithStaticMesh=false
  blendMode=BLEND_AlphaComposite
  materialDomain=MD_UI
  usedShadingModels=Default Lit
 }
 nodes [
  n1 = VertexColor()
  n2 = Multiply(A=NamedRerouteUsage()) {constB=0.05}
  n3 = NamedRerouteUsage()
  n4 = ComponentMask(None=@n3) {r=true}
  n5 = Multiply(A=Time()) {constB=-1}
  n6 = MaterialFunctionCall(UVs=AppendVector(A=Add(A=@n4, B=@n5), B=ComponentMask(None=@n3) {g=true}), Texture Scale=Constant2Vector() {g=1, r=16}) {materialFunction=&ScaleUVsByCenter}
  n7 = TextureObject() {samplerType=SAMPLERTYPE_Alpha, texture=/DiscoveryBrowser/CreatorProfile/Textures/T_UI_Discover_ElectricNoise}
  n8 = NamedRerouteUsage()
  n9 = Power(
   Base=NamedRerouteDeclaration(None=TextureSample(UVs=@n6.UVs, Tex=@n7).R) {name=Wave Pattern Grayscale, nodeColor=<rgba 0 0.1328 1 1>}
   Exp=LinearInterpolate(Alpha=ComponentMask(None=@n8) {r=true}) {constA=0.8}
  ) {constExponent=1}
  n10 = MaterialFunctionCall("Value (0-1)"=@n9) {materialFunction=&MF_UI_EaseCurves}
  n11 = ComponentMask(None=NamedRerouteUsage()) {r=true}
  n12 = Saturate(None=LinearInterpolate(A=LinearInterpolate(A=@n10.EaseOutQuint, Alpha=@n9), B=@n9, Alpha=Saturate(None=@n11)) {constAlpha=0})
  n13 = NamedRerouteUsage()
  n14 = LinearInterpolate(A=Multiply(A=NamedRerouteUsage(), B=SmoothStep(Min=@n12, Value=@n13) {constMax=-1}), B=@n12, Alpha=NamedRerouteUsage())
  n15 = NamedRerouteUsage()
  n16 = Multiply(A=Min(A=@n15, B=OneMinus(None=@n15))) {constB=2}
  n17 = Constant()
  n18 = Constant() {r=0.125}
  n19 = Multiply(
   A=Multiply(A=LinearInterpolate(A=@n2, Alpha=@n14), B=ScalarParameter() {defaultValue=1, parameterName=Opacity, sliderMax=1})
   B=Saturate(None=MaterialFunctionCall(Input=@n16, Input Low=@n17, Input High=@n18) {materialFunction=&RemapValue_0-1}.Result) {desc=Safeguard}
  )
  Multiply(A=@n1.A, B=@n19)
  n20 = NamedRerouteUsage()
  n21 = MaterialFunctionCall!(
   PrimaryUV
   SecondaryUV=LinearInterpolate(A=Constant2Vector(), B=Constant2Vector() {g=1, r=1}, Alpha=NamedRerouteUsage())
   UVMix=AppendVector(A=ComponentMask(None=@n20) {g=true}, B=ComponentMask(None=@n20) {r=true})
   PixelSize
  ) {materialFunction=/FortUILibrary/MaterialFunctions/CoreUI/MF_UI_CoreUI_OscillatingColor}
  n22 = Constant3Vector() {constant=<rgba 1 1 1 0>}
  n23 = VectorParameter() {parameterName=Color 2, defaultValue=<rgba 0 0.283 0.3385 1>}
  n24 = VectorParameter() {parameterName=Color 1, defaultValue=<rgba 0 1 0.5146 1>}
  n25 = NamedRerouteUsage()
  n26 = NamedRerouteDeclaration(None=StaticSwitchParameter(True=LinearInterpolate(A=@n21.Result, B=@n22), False=LinearInterpolate(A=@n23.RGB, B=@n24.RGB, Alpha=@n25)) {defaultValue=true, parameterName=Use Rift Colors}) {name=Color Morphing, nodeColor=<rgba 0 0 0 1>}
  n27 = NamedRerouteUsage()
  n28 = LinearInterpolate(A=Multiply(A=@n27) {constB=0.5}, Alpha=Multiply(A=@n27, B=@n27))
  n29 = LinearInterpolate(A=@n26, B=Multiply(A=Multiply(A=@n26, B=@n28)) {constB=8}, Alpha=@n28)
  Multiply(A=Multiply(A=Saturate(None=@n29), B=@n19), B=@n1.A)
  n30 = MaterialFunctionCall() {materialFunction=&GetUserInterfaceUV}
  n31 = ComponentMask(None=@n30.Normalized UV) {r=true}
  n32 = Rotator(
   Coordinate=Rotator(Coordinate=@n30.Normalized UV, Time=Constant() {r=-0.09}) {speed=6.2832}
   Time=LinearInterpolate(Alpha=MaterialFunctionCall("Value (0-1)"=@n31) {materialFunction=&MF_UI_EaseCurves}.EaseInOutCubic) {constA=-0.0175, constB=0.125}
  ) {speed=6.2832}
  n33 = Constant() {r=0.5}
  n34 = ComponentMask(None=@n32) {r=true}
  n35 = MaterialFunctionCall!(
   UV=NamedRerouteDeclaration(None=MaterialFunctionCall(UVs=@n32, Texture Scale=@n33) {materialFunction=&ScaleUVsByCenter}.UVs) {name=Wave UV, nodeColor=<rgba 0 0 0 1>}
   Wave Height=Constant() {r=0.25}
   Wave Elevation
   Wave Length=Constant() {r=0.5}
   Wave Speed=Constant() {r=-0.25}
   GlowMin
   GlowMax
   Outline Thickness=Multiply(A=LinearInterpolate(Alpha=@n34) {constA=0.25, constB=0.1}, B=ScalarParameter() {defaultValue=1, parameterName=Curve Thickness})
   Outline Alignment
   Custom Curve
  ) {materialFunction=/FortUILibrary/MaterialFunctions/Shapes/MF_UI_Wave}
  n36 = OneMinus(None=SmoothStep(Value=@n35.SDF Outline) {constMax=0.5})
  n37 = ComponentMask(None=NamedRerouteUsage()) {r=true}
  n38 = Constant()
  n39 = Constant() {r=1.375}
  n40 = OneMinus(None=Saturate(None=MaterialFunctionCall(Input=@n37, Input Low=@n38, Input High=@n39) {materialFunction=&RemapValue_0-1}.Result))
  NamedRerouteDeclaration(None=Multiply(A=@n36, B=@n40)) {name=Wind Mask, nodeColor=<rgba 0 0 0 1>}
  n41 = SmoothStep(Value=@n35.SDF Outline) {constMax=0.75}
  NamedRerouteDeclaration(None=Multiply(A=OneMinus(None=@n41), B=@n40)) {name=Wind Glow Mask, nodeColor=<rgba 0 0 0 1>}
  NamedRerouteDeclaration(None=@n36) {name=Wave Mask, nodeColor=<rgba 0 0.5078 1 1>}
  NamedRerouteDeclaration(None=ComponentMask(None=@n30.Normalized UV) {g=true}) {name=Vertical Gradient, nodeColor=<rgba 0 0 0 1>}
 ]
}
