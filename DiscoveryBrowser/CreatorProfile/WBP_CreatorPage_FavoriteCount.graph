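# Creator page widget that shows a creator's favorite count as an icon, number and label in a horizontal row.
# It starts collapsed on PreConstruct; Set Favorite Count updates it, showing it only for a count above zero.
# The creator-follow rebranding runtime option switches the icon and singular/plural label between Favorite and Follower styles.
# Counts over 999 display in rounded-down shorthand, otherwise as plain integers.
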WidgetBlueprint /DiscoveryBrowser/CreatorProfile/WBP_CreatorPage_FavoriteCount
graphs [
 EventGraph {nodes [
   Event PreConstruct()
   > SetVisibility(InVisibility=Collapsed)
  ]}
 Set Favorite Count {nodes [
   n1 = Set Favorite Count!() -> (then, Favorite Count:int) {is=title K2Node_FunctionEntry}
   n2 = "Equal (Integer)"(A=@n1, B=1)
   n3 = GetRuntimeOptions()
   > Set Text(
    Text=SelectText(
     A=SelectText(A="LOCTEXT(\"Follower\")", B="LOCTEXT(\"Followers\")", bPickA=@n2)
     B=SelectText(A="LOCTEXT(\"Favorite\")", B="LOCTEXT(\"Favorites\")", bPickA=@n2)
     bPickA=GetIsCreatorFollowRebrandingEnabled!(self:Fort Runtime Options Object Reference=@n3) -> (ReturnValue:bool)
    )
    self=Get Txt_Favorite() -> (Txt_Favorite:Common Text Object Reference)
   )
   n4 = GetRuntimeOptions()
   > SetBrushFromMaterial(
    self=Get Img_Favorite() -> (Img_Favorite:Image Object Reference)
    Material=SelectMaterialInstance!(
     A:Material Instance Object Reference=/Game/UI/DiscoveryShared/Art/MI_UI_Icon_CCU
     B:Material Instance Object Reference=/DiscoveryBrowser/Tiles/Islands/Materials/MI_UIKit_Icon_Favorite_Discover
     bSelectA:bool=GetIsCreatorFollowRebrandingEnabled(self=@n4)
    ) -> (ReturnValue:Material Instance Object Reference)
   )
   n5 = Get Favorite Count() -> (Favorite Count:int=0)
   > SetVisibility(InVisibility=Select(Option 0=HitTestInvisible, Option 1=Collapsed, Index="integer <= integer"(A=@n5)))
   > n6 = Branch(Condition="integer > integer"(A=@n1, B=999))
   n7 = Get Txt_Count() -> (Txt_Count:Common Text Object Reference)
   > "SetText (Text)"(
    self=@n7
    InText=GetShorthandQuantity!(Quantity:int=@n1, MinimumFractionalDigits:int=0, MaximumFractionalDigits:int=1, bShouldRoundDown:bool=true) -> (ReturnValue:Text)
   )
   "SetText (Text)"(execute=@n6.else, self=@n7, InText="To Text (Integer)"(Value=@n1))
  ]}
]
variables []
widgets {
 info {inheritedWidgetCount=0, parentClass=&UserWidget, rootWidgetClass=&HorizontalBox, widgetCount=4}
 widgets [
  HorizontalBox_21 {widgetClassPath=&HorizontalBox, widget=HorizontalBox_21}
  Img_Favorite {
   widgetClassPath=&Image
   bIsVariable=true
   parent=HorizontalBox_21
   widget=Img_Favorite
   properties {brush {
     bIsDynamicallyLoaded=false
     drawAs=Image
     imageType=FullColor
     mirroring=NoMirror
     resourceName=None
     resourceObject=&MI_UIKit_Icon_Favorite_Discover
     tiling=NoTile
     imageSize=<32 32>
     margin=<ltrb 0 0 0 0>
     outlineSettings {
      bUseBrushTransparency=false
      roundingType=HalfHeightRadius
      width=0
      color {colorUseRule=UseColor_Specified, specifiedColor=<rgba 0 0 0 0>}
      cornerRadii=<0 0 0 0>
     }
     tintColor {colorUseRule=UseColor_Specified, specifiedColor=<rgba 1 1 1 1>}
     uVRegion {bIsValid=false, max=<0 0>, min=<0 0>}
    }}
   slot {class=HorizontalBoxSlot, properties {padding=<ltrb 0 0 8 0>}}
  }
  Txt_Count {
   widgetClassPath=&CommonTextBlock
   bIsVariable=true
   parent=HorizontalBox_21
   widget=Txt_Count
   properties {
    text=$FavoriteCount
    font {
     bForceMonospaced=false
     bMaterialIsStencil=false
     fontMaterial=null
     fontObject=&NotoSans
     letterSpacing=0
     monospacedWidth=1
     size=24
     skewAmount=0
     typefaceFontName=NotoSans-Regular
     outlineSettings {
      bApplyOutlineToDropShadows=false
      bMiteredCorners=false
      bSeparateFillAlpha=false
      outlineMaterial=null
      outlineSize=0
      outlineColor=<rgba 0 0 0 1>
     }
    }
    shadowOffset=<0 0>
   }
   slot {class=HorizontalBoxSlot, properties {padding=<ltrb 0 0 8 0>}}
  }
  Txt_Favorite {
   widgetClassPath=&CommonTextBlock
   bIsVariable=true
   parent=HorizontalBox_21
   widget=Txt_Favorite
   properties {
    renderOpacity=0.7
    text=Favorite
    font {
     bForceMonospaced=false
     bMaterialIsStencil=false
     fontMaterial=null
     fontObject=&NotoSans
     letterSpacing=0
     monospacedWidth=1
     size=24
     skewAmount=0
     typefaceFontName=NotoSans-Regular
     outlineSettings {
      bApplyOutlineToDropShadows=false
      bMiteredCorners=false
      bSeparateFillAlpha=false
      outlineMaterial=null
      outlineSize=0
      outlineColor=<rgba 0 0 0 1>
     }
    }
    shadowOffset=<0 0>
   }
   slot {class=HorizontalBoxSlot}
  }
 ]
}
