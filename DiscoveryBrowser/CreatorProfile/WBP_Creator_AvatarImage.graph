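# CommonUserWidget for a creator avatar in the Discovery Browser, with a live outline behind the thumbnail that shows when playing now.
# PreConstruct applies the avatar size by rebuilding the thumbnail brush with a new image size.
# Setting a texture feeds the thumbnail's dynamic material, sets "Is Empty" from validity and plays a material alpha and zoom animation.
# The loading state zeroes thumbnail alpha and plays that animation in reverse.
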WidgetBlueprint /DiscoveryBrowser/CreatorProfile/WBP_Creator_AvatarImage
animations [Anim_OnLoadedImage {
  binding_count=1
  length_seconds=0.24
  object=&self:Anim_OnLoadedImage
  sequence {bindings [CreatorThumbnail {
     id=#1
     possessed_class=Image
     tracks [{
       type=MovieSceneWidgetMaterialTrack
       sections [{
         type=MovieSceneParameterSection
         end_frame=4
         start_frame=0
         channels [
          Thumbnail Alpha {type=MovieSceneScriptingFloatChannel, keys [[0,0],[4.8,1]]}
          Thumbnail Zoom {type=MovieSceneScriptingFloatChannel, keys [[0,1.05],[4.8,1]]}
         ]
        }]
      }]
    }]}
 }]
graphs [
 EventGraph {nodes [
   Event PreConstruct()
   > SetCreatorAvatarImageSize!(execute, self, CreatorAvatarSize:Vector 2D Structure=Get CreatorAvatarSize() -> (CreatorAvatarSize:Vector 2D Structure)) -> (then)
  ]}
 SetCreatorAvatarImageSize {nodes [
   n1 = SetCreatorAvatarImageSize!() -> (then, CreatorAvatarSize:Vector 2D Structure) {is=title K2Node_FunctionEntry}
   > n2 = Set CreatorAvatarSize(CreatorAvatarSize:Vector 2D Structure=@n1) -> (Output_Get:Vector 2D Structure)
   n3 = Get CreatorThumbnail() -> (CreatorThumbnail:Image Object Reference)
   n4 = Break Slate Brush!(SlateBrush:"Slate Brush Structure (by ref)"=Get Brush!(self:Image Object Reference=@n3) -> (Brush:Slate Brush Structure)) -> (
    TintColor:Slate Color Structure
    DrawAs:ESlateBrushDrawType Enum
    Tiling:ESlateBrushTileType Enum
    Mirroring:ESlateBrushMirrorType Enum
    ImageSize:Vector 2D Structure
    Margin:Margin Structure
    ResourceObject:Object Reference
    OutlineSettings:Slate Brush Outline Settings Structure
   ) {is=title K2Node_BreakStruct}
   > SetBrush(
    self=@n3
    InBrush=Make Slate Brush!(
     TintColor:Slate Color Structure=@n4
     DrawAs:ESlateBrushDrawType Enum=@n4.DrawAs
     Tiling:ESlateBrushTileType Enum=@n4.Tiling
     Mirroring:ESlateBrushMirrorType Enum=@n4.Mirroring
     ImageSize:Vector 2D Structure=@n2
     Margin:Margin Structure=@n4.Margin
     ResourceObject:Object Reference=@n4.ResourceObject
     OutlineSettings:Slate Brush Outline Settings Structure=@n4.OutlineSettings
    ) -> (SlateBrush:Slate Brush Structure) {is=title K2Node_MakeStruct}
   )
  ]}
 SetCreatorAvatarImageTexture {nodes [
   n1 = SetCreatorAvatarImageTexture!() -> (then, CreatorAvatarTexture:Texture Object Reference) {is=title K2Node_FunctionEntry}
   > n2 = GetDynamicMaterial(self=Get CreatorThumbnail())
   > SetTextureParameterValue(
    self=@n2
    ParameterName=Get TextureParam() -> (TextureParam:Name=None)
    Value=Get CreatorAvatarTexture() -> (CreatorAvatarTexture:Texture Object Reference)
   )
   n3 = IsValid(Object=@n1)
   > SetScalarParameterValue(self=@n2, ParameterName=Is Empty, Value="To Float (Boolean)"(InBool=NOT Boolean(A=@n3)))
   > PlayAnimationForward(InAnimation=Get Anim_OnLoadedImage() -> (Anim_OnLoadedImage:Widget Animation Object Reference))
  ]}
 SetIsPlayingNow {nodes [
   n1 = SetIsPlayingNow!() -> (then, IsPlayingNow:bool=~) {is=title K2Node_FunctionEntry}
   > SetVisibility(
    self=Get LiveFeedback() -> (LiveFeedback:Image Object Reference)
    InVisibility=Select(Option 0=Collapsed, Option 1=SelfHitTestInvisible, Index=@n1)
   )
  ]}
 SetLoadingState {nodes [
   SetLoadingState() {is=title K2Node_FunctionEntry}
   > n1 = GetDynamicMaterial(self=Get CreatorThumbnail())
   > SetScalarParameterValue(self=@n1, ParameterName=Thumbnail Alpha, Value=0)
   > PlayAnimationReverse(InAnimation=Get Anim_OnLoadedImage())
  ]}
]
variables [
 CreatorAvatarSize {type=Vector2D, default=<128 128>}
 TextureParam {type=string, default=Thumbnail}
]
widgets {
 info {inheritedWidgetCount=0, parentClass=&CommonUserWidget, rootWidgetClass=&GridPanel, widgetCount=3}
 widgets [
  Grid_Avatar {widgetClassPath=&GridPanel, widget=Grid_Avatar, properties {columnFill [1], rowFill [1]}}
  LiveFeedback {
   widgetClassPath=&Image
   bIsVariable=true
   parent=Grid_Avatar
   widget=LiveFeedback
   properties {
    visibility=SelfHitTestInvisible
    brush {
     bIsDynamicallyLoaded=false
     drawAs=Image
     imageType=FullColor
     mirroring=NoMirror
     resourceName=None
     resourceObject=/DiscoveryBrowser/Tiles/Creator/Materials/MI_UI_Creator_Tile_OutlineState
     tiling=NoTile
     imageSize=<32 32>
     margin=<ltrb 0 0 0 0>
     outlineSettings {
      bUseBrushTransparency=false
      roundingType=HalfHeightRadius
      width=0
      color {colorUseRule=UseColor_Specified, specifiedColor=<rgba 0 0 0 0>}
      cornerRadii=<0 0 0 0>
     }
     tintColor {colorUseRule=UseColor_Specified, specifiedColor=<rgba 1 1 1 1>}
     uVRegion {bIsValid=false, max=<0 0>, min=<0 0>}
    }
   }
   slot {class=GridSlot, properties {padding=<ltrb -12 -12 -12 -12>}}
  }
  CreatorThumbnail {
   widgetClassPath=&Image
   bIsVariable=true
   parent=Grid_Avatar
   widget=CreatorThumbnail
   properties {
    visibility=SelfHitTestInvisible
    brush {
     bIsDynamicallyLoaded=false
     drawAs=Image
     imageType=FullColor
     mirroring=NoMirror
     resourceName=None
     resourceObject=/DiscoveryBrowser/Tiles/Creator/Materials/MI_UI_CreatorTile
     tiling=NoTile
     imageSize=<32 32>
     margin=<ltrb 0 0 0 0>
     outlineSettings {
      bUseBrushTransparency=false
      roundingType=HalfHeightRadius
      width=0
      color {colorUseRule=UseColor_Specified, specifiedColor=<rgba 0 0 0 0>}
      cornerRadii=<0 0 0 0>
     }
     tintColor {colorUseRule=UseColor_Specified, specifiedColor=<rgba 1 1 1 1>}
     uVRegion {bIsValid=false, max=<0 0>, min=<0 0>}
    }
   }
   slot {class=GridSlot, properties {layer=1}}
  }
 ]
}
