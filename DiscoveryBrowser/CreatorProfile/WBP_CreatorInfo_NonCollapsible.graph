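# Non-collapsible creator info header widget for creator profiles in the Discovery browser.
# A horizontal row pairs the avatar image with a column holding the wrapping, outlined creator name and a favorite count widget.
# It declares event dispatchers for creator reported, follow clicked and last-focused widget updates, though it contains no report or follow button.
# The event graph is empty, so the widget neither populates its content nor fires those dispatchers itself.
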WidgetBlueprint /DiscoveryBrowser/CreatorProfile/WBP_CreatorInfo_NonCollapsible
graphs [
 EventGraph {links [], nodes []}
 HandleCreatorReported {nodes [HandleCreatorReported() {is=title K2Node_FunctionEntry}]}
 HandleFollowClicked {nodes [HandleFollowClicked!() -> (then, Button:Common Button Base Object Reference) {is=title K2Node_FunctionEntry}]}
 LastFocusedUpdatedFromCreatorInfos {nodes [LastFocusedUpdatedFromCreatorInfos!() -> (then, LastFocusedWidget:User Widget Object Reference) {is=title K2Node_FunctionEntry}]}
]
variables [
 HandleCreatorReported {}
 HandleFollowClicked {}
 LastFocusedUpdatedFromCreatorInfos {}
]
widgets {
 info {inheritedWidgetCount=0, parentClass=&UserWidget, rootWidgetClass=&Overlay, widgetCount=6}
 widgets [
  O_Root {widgetClassPath=&Overlay, widget=O_Root}
  CreatorImageHeader {
   widgetClassPath=&HorizontalBox
   parent=O_Root
   widget=CreatorImageHeader
   slot {class=OverlaySlot, properties {horizontalAlignment=HAlign_Fill}}
  }
  WBP_Creator_AvatarImage {
   widgetClassPath=&/WBP_Creator_AvatarImage
   bIsVariable=true
   parent=CreatorImageHeader
   widget=WBP_Creator_AvatarImage
   properties {creatorAvatarSize=<140 140>}
   slot {class=HorizontalBoxSlot, properties {horizontalAlignment=HAlign_Left, verticalAlignment=VAlign_Center}}
  }
  VB_UserInfos {
   widgetClassPath=&VerticalBox
   parent=CreatorImageHeader
   widget=VB_UserInfos
   slot {class=HorizontalBoxSlot, properties {verticalAlignment=VAlign_Center, padding=<ltrb 24 0 0 0>, size {sizeRule=Fill, value=1}}}
  }
  Txt_CreatorName {
   widgetClassPath=&CommonTextBlock
   bIsVariable=true
   parent=VB_UserInfos
   widget=Txt_CreatorName
   properties {
    autoWrapText=true
    text="{Creator Name}"
    wrapTextAt=720
    font {
     bForceMonospaced=false
     bMaterialIsStencil=false
     fontMaterial=null
     fontObject=&HeadingNow
     letterSpacing=-40
     monospacedWidth=1
     size=41.25
     skewAmount=0
     typefaceFontName=86 Bold Italic
     outlineSettings {
      bApplyOutlineToDropShadows=false
      bMiteredCorners=false
      bSeparateFillAlpha=false
      outlineMaterial=/FortUILibrary/MaterialInstances/Text/MI_UI_Block_Text_Outline
      outlineSize=2
      outlineColor=<rgba 0 0 0 1>
     }
    }
    shadowOffset=<0 0>
   }
   slot {class=VerticalBoxSlot}
  }
  WBP_CreatorPage_FavoriteCount {
   widgetClassPath=&/WBP_CreatorPage_FavoriteCount
   bIsVariable=true
   parent=VB_UserInfos
   widget=WBP_CreatorPage_FavoriteCount
   slot {class=VerticalBoxSlot, properties {padding=<ltrb 0 8 0 0>}}
  }
 ]
}
