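# WBP_CreatorPage_Button_Link is a social link button for the Discovery browser's creator profile page, derived from the generic UIKit button with no widget tree of its own.
# It stores a platform, URL and entry box index.
# On focus, it broadcasts OnSocialButtonFocused with that index so listeners know which button is focused.
# On construct, it enlarges its icon for mobile games through SetCompactUIStyle.
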WidgetBlueprint /DiscoveryBrowser/CreatorProfile/WBP_CreatorPage_Button_Link
graphs [
 EventGraph {nodes [
   Event On Focused() {is=title K2Node_Event}
   > Call On Social Button Focused!(execute:Exec, self:WBP Creator Page Button Link Object Reference, EntryBoxIndex:int=Get EntryBoxIndex() -> (EntryBoxIndex:int=0)) -> (then:Exec) {is=title K2Node_CallDelegate}
   Event Construct()
   > SetCompactUIStyle() {is=title K2Node_CallFunction}
  ]}
 OnSocialButtonFocused {nodes [OnSocialButtonFocused!() -> (then, EntryBoxIndex:int) {is=title K2Node_FunctionEntry}]}
 SetCompactUIStyle {nodes [
   SetCompactUIStyle() {is=title K2Node_FunctionEntry}
   > Branch(Condition=IsMobileGame())
   > SetIconSize!(execute, self:WBP UIKit Button Generic Base Object Reference, Icon:Vector 2D Structure=(X=48,Y=48)) -> (then)
  ]}
]
variables [
 EntryBoxIndex {type=integer}
 OnSocialButtonFocused {}
 Platform {type=string, default=""}
 URL {type=string, default=""}
]
widgets {
 info {
  inheritedWidgetCount=0
  parentClass=/FortUILibrary/UIKit/Buttons/WBP_UIKit_Button_Generic.WBP_UIKit_Button_Generic_C
  rootWidgetClass=null
  widgetCount=0
 }
 widgets []
}
