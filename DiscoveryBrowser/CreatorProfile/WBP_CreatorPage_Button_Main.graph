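# Creator profile page main button, derived from the generic UIKit button, that shows follow or favorite state.
# SetFollowState picks Follow/Following or Favorite/Favorited text and icons from the follow-rebranding runtime option, hiding the following-state icon under rebranding.
# Anim_Following feeds a setter that fades the background material's Display Loud parameter out.
# IsInitialized, cleared on Construct, makes the first state update snap instead of animate.
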WidgetBlueprint /DiscoveryBrowser/CreatorProfile/WBP_CreatorPage_Button_Main
animations [Anim_Following {
  binding_count=1
  length_seconds=0.3334
  object=&self:Anim_Following
  sequence {
   bindings [WBP_CreatorPage_Button_Follow_C_30 {
     id=#1
     possessed_class=WBP_CreatorPage_Button_Main_C
     tracks [
      {
       type=MovieSceneDoubleTrack
       display_name=Anim Value Following Icon
       sections [{
         type=MovieSceneDoubleSection
         end_frame=20
         start_frame=0
         channels [None {type=MovieSceneScriptingDoubleChannel, keys [[0,0],[3,0],[20,1]]}]
        }]
      }
      {
       type=MovieSceneDoubleTrack
       display_name=Anim Value Following Background
       sections [{type=MovieSceneDoubleSection, end_frame=20, start_frame=0, channels [None {type=MovieSceneScriptingDoubleChannel, keys [[0,0],[20,1]]}]}]
      }
     ]
    }]
   display_rate {numerator=60}
  }
 }]
graphs [
 EventGraph {nodes [
   n1 = Event PreConstruct()
   > Parent: PreConstruct(IsDesignTime=@n1.IsDesignTime)
   > SetBackgroundMaterials!(
    execute
    self:WBP UIKit Button Generic Base Object Reference
    Background MI:Material Instance Object Reference=/DiscoveryBrowser/CreatorProfile/MaterialInstances/MI_CreatorPage_Button_Main_Background
    Outline MI:Material Instance Object Reference=Get OutlineDynamicMaterial!(self:WBP UIKit Block Background Base Object Reference=Get Block_Background() -> (Block_Background:WBP UIKit Block Background Base Object Reference)) -> (OutlineDynamicMaterial:Material Instance Dynamic Object Reference)
   ) -> (then)
   > Set IconRightPadding(IconRightPadding:double=Get!() -> (IconPadding_Left:float=~, IconPadding_Top:float=~, IconPadding_Right:float=~, IconPadding_Bottom:float=~).IconPadding_Right) -> (Output_Get:double)
   Event Construct()
   > Parent: Construct()
   > Set IsInitialized(IsInitialized:bool=false) -> (Output_Get:bool)
  ]}
 SetAnimValue_Following_Background {nodes [
   n1 = SetAnimValue_Following_Background!() -> (then, NewParam:double=~) {is=title K2Node_FunctionEntry}
   > SetScalarParameterValue(
    self=Get BackgroundDynamicMaterial!(self:WBP UIKit Block Background Base Object Reference=Get Block_Background()) -> (BackgroundDynamicMaterial:Material Instance Dynamic Object Reference)
    ParameterName=Display Loud
    Value=Lerp(A=1, B=0, Alpha=@n1)
   )
  ]}
 SetFollowState {nodes [
   n1 = SetFollowState!() -> (then, IsFollowing:bool=~) {is=title K2Node_FunctionEntry}
   > Set Is Following(Is Following:bool=@n1) -> (Output_Get:bool)
   n2 = GetRuntimeOptions()
   n3 = Get Is Following() -> (Is Following:bool)
   > SetText!(
    execute
    self:WBP UIKit Button Generic Base Object Reference
    Text:Text=Select(
     Option 0:Text=Select(Option 0="LOCTEXT(\"Favorite\")", Option 1="LOCTEXT(\"Favorited\")", Index=@n3)
     Option 1:Text=Select(Option 0="LOCTEXT(\"Follow\")", Option 1="LOCTEXT(\"Following\")", Index=@n3)
     Index=GetIsCreatorFollowRebrandingEnabled!(self:Fort Runtime Options Object Reference=@n2) -> (ReturnValue:bool)
    ) -> (ReturnValue:Text)
   ) -> (then)
   > n4 = Branch(Condition=Get IsInitialized() -> (IsInitialized:bool))
   > n5 = Branch(Condition=Get Is Following())
   > PlayAnimationForward(InAnimation=Get Anim_Following() -> (Anim_Following:Widget Animation Object Reference))
   n6 = Get Block_Icon() -> (Block_Icon:WBP UIKit Block Image Object Reference)
   > SetBrushFromMaterial(
    self=Get Image!(self:WBP UIKit Block Image Object Reference=@n6) -> (Image:Fort Mobile Image Object Reference)
    Material=/DiscoveryBrowser/CreatorProfile/MaterialInstances/MI_CreatorPage_Button_Main_Icon_Favorited
   )
   n7 = GetIsCreatorFollowRebrandingEnabled(self=GetRuntimeOptions())
   > SetBlockVisibility!(execute, self:WBP UIKit Block Base Object Reference=@n6, Visible:bool=NOT Boolean(A=@n7)) -> (then)
   StopAnimation(execute=@n5.else, InAnimation=Get Anim_Following())
   > SetAnimValue_Following_Background!(execute, self, NewParam:double=0) -> (then) {is=title K2Node_CallFunction}
   n8 = Get Block_Icon()
   > SetBrushFromMaterial(
    self=Get Image(self=@n8)
    Material=Select(
     Option 0:Material Interface Object Reference=/DiscoveryBrowser/CreatorProfile/MaterialInstances/MI_CreatorPage_Button_Main_Icon_Favorite
     Option 1:Material Interface Object Reference=/DiscoveryBrowser/CreatorProfile/MaterialInstances/MI_CreatorPage_Button_Icon_Follow
     Index=@n7
    ) -> (ReturnValue:Material Interface Object Reference)
   )
   > SetBlockVisibility(self=@n8, Visible=true)
   StopAnimation(execute=@n4.else, InAnimation=Get Anim_Following())
   > SetAnimValue_Following_Background(NewParam="To Float (Boolean)"(InBool=Get Is Following()))
   n9 = GetRuntimeOptions()
   n10 = Get Block_Icon()
   > SetBrushFromMaterial(
    self=Get Image(self=@n10)
    Material=Select(
     Option 0=&MI_CreatorPage_Button_Main_Icon_Favorite
     Option 1=&MI_CreatorPage_Button_Icon_Follow
     Index=GetIsCreatorFollowRebrandingEnabled(self=@n9)
    )
   )
   > SetBlockVisibility(self=@n10, Visible=true)
   > Set IsInitialized(IsInitialized=true)
  ]}
]
variables [
 AnimValue_Following_Background {type=number}
 AnimValue_Following_Icon {type=number}
 IconRightPadding {type=number}
 Is Following {type=boolean}
 IsInitialized {type=boolean}
]
widgets {
 info {
  inheritedWidgetCount=0
  parentClass=/FortUILibrary/UIKit/Buttons/WBP_UIKit_Button_Generic.WBP_UIKit_Button_Generic_C
  rootWidgetClass=null
  widgetCount=0
 }
 widgets []
}
