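# Sidebar onboarding tooltip that shows a coach mark after the player follows or favorites their first creator.
# On Construct, the creator follow rebranding runtime option selects the header, description and button wording.
# OnShowTooltip activates the wrapped UIKit coach mark, whose single action button takes focus and is bound to Escape.
# Clicking it runs the parent's action handler, reports close animation finished immediately, then deactivates both widgets.
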WidgetBlueprint /DiscoveryBrowser/CreatorProfile/WBP_CreatorPage_Followed_CoachMark
graphs [
 BP_GetDesiredFocusTarget {nodes [
   Get Desired Focus Target()
   > Return Node!(
    execute
    ReturnValue:Widget Object Reference=Get OtherActionButton!(self:WBP UIKit Coach Mark Base Object Reference=Get WBP_UIKit_CoachMark() -> (WBP_UIKit_CoachMark:WBP UIKit Coach Mark Object Reference)) -> (OtherActionButton:WBP UIKit Button Regular Object Reference)
   )
  ]}
 EventGraph {nodes [
   "Button Other Action Clicked (WBP_UIKit_CoachMark)"() {is=title K2Node_ComponentBoundEvent}
   > HandleOtherActionClicked(self:Fort Sidebar Onboard Tooltip Widget Object Reference)
   > NotifyCloseAnimationFinished(self:Fort Sidebar Onboard Tooltip Widget Object Reference)
   > DeactivateWidget(self=Get WBP_UIKit_CoachMark())
   > DeactivateWidget()
   Event Construct()
   > Parent: Construct()
   n1 = GetIsCreatorFollowRebrandingEnabled!(self:Fort Runtime Options Object Reference=GetRuntimeOptions()) -> (ReturnValue:bool)
   > SetText!(
    execute
    self:Fort CTAButton Object Reference=Get OtherActionButton(self=Get WBP_UIKit_CoachMark())
    ButtonText:"Text (by ref)"=SelectText(A="LOCTEXT(\"Next\")", B="LOCTEXT(\"Got it\")", bPickA=@n1)
   ) -> (then)
   > Set TextDescription!(
    execute
    TextDescription:Text=SelectText(
     A="LOCTEXT(\"Content from this developer will now be shown in the Following tab.\")"
     B="LOCTEXT(\"Content from this creator is now available in your Library and Discover.\")"
     bPickA=@n1
    )
    self:WBP UIKit Coach Mark Base Object Reference=Get WBP_UIKit_CoachMark()
   ) -> (then, Output_Get:Text)
   > Set TextHeader!(
    execute
    TextHeader:Text=SelectText(A="LOCTEXT(\"You followed your first developer!\")", B="LOCTEXT(\"You favorited your first creator!\")", bPickA=@n1)
    self:WBP UIKit Coach Mark Base Object Reference=Get WBP_UIKit_CoachMark()
   ) -> (then, Output_Get:Text)
   Event OnShowTooltip() {is=title K2Node_Event}
   > ActivateWidget(self=Get WBP_UIKit_CoachMark())
  ]}
]
variables []
widgets {
 info {
  inheritedWidgetCount=2
  parentClass=/Script/FortniteUI.FortSidebarOnboardTooltipWidget
  rootWidgetClass=/FortUILibrary/UIKit/CoachMark/WBP_UIKit_CoachMark.WBP_UIKit_CoachMark_C
  widgetCount=3
 }
 widgets [
  WBP_UIKit_CoachMark {
   widgetClassPath=/FortUILibrary/UIKit/CoachMark/WBP_UIKit_CoachMark
   bIsVariable=true
   widget=WBP_UIKit_CoachMark
   properties {
    arrowPosition=Left_Up
    bIsFocusable=true
    navigation=&self:WidgetTree.WBP_UIKit_CoachMark.WidgetNavigation_0
    shouldShowDismissButton=false
    textDescription=Content from this creator is now available in your Library and Discover.
    textDismissButton=""
    textHeader="You favorited your first creator!"
    textOtherActionButton=GOT IT
    visibility=SelfHitTestInvisible
    widget Width=650
    input Action Other {dataTable=&InputActionDataTable, rowName=GenericEscape}
   }
  }
  Dismiss_Button {widgetClassPath=&CommonButtonLegacy, bInherited=true, widget=null}
  OtherAction_Button {widgetClassPath=&CommonButtonLegacy, bInherited=true, widget=null}
 ]
}
