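# Creator-page modal dialog (CommonActivatableWidget) that shows a creator's social link as a scannable QR code and a clickable URL.
# SetCreatorData stores the platform and URL, formats the title, and generates the QR texture from the URL with nearest filtering for crisp pixels.
# The URL button opens the link via FortPlayerPagesSubSystem, logs an analytics interaction tagged with the platform, and toggles an outline on hover.
# The Close button deactivates the widget.
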WidgetBlueprint /DiscoveryBrowser/CreatorProfile/WBP_CreatorPage_QRCode_Modal
graphs [
 EventGraph {nodes [
   "On Clicked (Button_Close)"!() -> (OutputDelegate, then, Button:Common Button Base Object Reference) {is=title K2Node_ComponentBoundEvent}
   > DeactivateWidget()
   "On Clicked (Button_URL)"() {is=title K2Node_ComponentBoundEvent}
   > OpenSocialUrl!(
    execute
    self:Fort Player Pages Sub System Object Reference=Get FortPlayerPagesSubSystem() -> (ReturnValue:Fort Player Pages Sub System Object Reference) {is=title K2Node_GetSubsystem}
    InUrl:String=Get URL() -> (URL:String)
   ) -> (then)
   > Fire Event Element Interacted(
    Element=Get Button_URL() -> (Button_URL:Button Object Reference)
    AdditionalInfo="Add String (Non-Localized)"!(
     Info:"Flow Analytics Additional Info Structure (by ref)"=Make Flow Analytics Additional Info!() -> (FlowAnalyticsAdditionalInfo:Flow Analytics Additional Info Structure) {is=title K2Node_MakeStruct}
     Key:String=SocialPlatform
     Value:String=Get SocialPlatform() -> (SocialPlatform:String)
    ) -> (ReturnValue:Flow Analytics Additional Info Structure)
   )
   "On Hovered (Button_URL)"() {is=title K2Node_ComponentBoundEvent}
   n1 = Get WBP_Outline() -> (WBP_Outline:WBP UIKit Block Outline Object Reference)
   > RequestTransition!(execute, self:UIKit Block Object Reference=@n1, State:EUIKitBlockVisualState Enum=Focused) -> (then)
   "On Unhovered (Button_URL)"() {is=title K2Node_ComponentBoundEvent}
   > RequestTransition(self=@n1, State=Unfocused)
  ]}
 SetCreatorData {nodes [
   n1 = SetCreatorData!() -> (then, Creator Name:Text, Social Platform:String, URL:String) {is=title K2Node_FunctionEntry}
   > Set SocialPlatform(SocialPlatform:String=@n1.Social Platform) -> (Output_Get:String)
   > Set URL(URL:String=@n1.URL) -> (Output_Get:String)
   > Set Text(
    Text=Format Text!(Format="LOCTEXT(\"{Name} on {Platform}\")", Name:Text=@n1, Platform:String=@n1.Social Platform) -> (Result)
    self=Get Text_ModalTitle() -> (Text_ModalTitle:Common Text Object Reference)
   )
   > Set Text(Text="To Text (String)"(InString=Get URL()), self=Get Text_URL() -> (Text_URL:Common Text Object Reference))
   > n2 = GenerateQRCodeTextureFromString!(execute, StringToEncode:String=Get URL()) -> (then, ReturnValue:Texture 2D Object Reference)
   > Set Filter!(execute, Filter:TextureFilter Enum=TF_Nearest, self:Texture Object Reference=@n2) -> (then, Output_Get:TextureFilter Enum=TF_Nearest)
   > SetBrushFromTexture!(
    execute
    self:Image Object Reference=Get Image_QRCode() -> (Image_QRCode:Image Object Reference)
    Texture:Texture 2D Object Reference=@n2
    bMatchSize:bool=false
   ) -> (then)
  ]}
]
variables [
 Mobile Scale {type=number, default=1.4}
 SocialPlatform {type=string, default=""}
 URL {type=string, default=""}
]
widgets {
 info {
  inheritedWidgetCount=0
  namedSlotCount=2
  parentClass=&CommonActivatableWidget
  rootWidgetClass=/FortUILibrary/Systems/Dialog/WBP_UIKit_Dialog_Internal.WBP_UIKit_Dialog_Internal_C
  widgetCount=14
 }
 widgets [
  Dialog_Internal {
   widgetClassPath=/FortUILibrary/Systems/Dialog/WBP_UIKit_Dialog_Internal
   bIsVariable=true
   widget=Dialog_Internal
   properties {
    bIsFocusable=true
    useClickCatch=true
    dialogSIze {
     height Override {enable=false, value=684}
     max Aspect Ratio {enable=false, value=0}
     max Desired Height {enable=true, value=800}
     max Desired Width {enable=false, value=0}
     min Aspect Ratio {enable=false, value=0}
     min Desired Height {enable=false, value=0}
     min Desired Width {enable=false, value=0}
     width Override {enable=true, value=864}
    }
   }
  }
  Overlay_0 {widgetClassPath=&Overlay, namedSlotHost=Dialog_Internal, widget=Overlay_0}
  VerticalBox_4 {
   widgetClassPath=&VerticalBox
   parent=Overlay_0
   widget=VerticalBox_4
   slot {class=OverlaySlot, properties {horizontalAlignment=HAlign_Fill, verticalAlignment=VAlign_Fill, padding=<ltrb 80 80 80 48>}}
  }
  Text_ModalTitle {
   widgetClassPath=&CommonTextBlock
   bIsVariable=true
   parent=VerticalBox_4
   widget=Text_ModalTitle
   properties {
    autoWrapText=true
    justification=Center
    text="{CREATOR} ON {PLATFORM}"
    textTransformPolicy=ToUpper
    font {
     bForceMonospaced=false
     bMaterialIsStencil=false
     fontMaterial=null
     fontObject=&HeadingNow
     letterSpacing=0
     monospacedWidth=1
     size=25.5
     skewAmount=0
     typefaceFontName=86 Bold Italic
     outlineSettings {
      bApplyOutlineToDropShadows=false
      bMiteredCorners=false
      bSeparateFillAlpha=false
      outlineMaterial=null
      outlineSize=0
      outlineColor=<rgba 0 0 0 1>
     }
    }
    shadowOffset=<0 0>
   }
   slot {class=VerticalBoxSlot, properties {verticalAlignment=VAlign_Top}}
  }
  Image_QRCode {
   widgetClassPath=&Image
   bIsVariable=true
   parent=VerticalBox_4
   widget=Image_QRCode
   properties {brush {
     bIsDynamicallyLoaded=false
     drawAs=Image
     imageType=NoImage
     mirroring=NoMirror
     resourceName=None
     resourceObject=null
     tiling=NoTile
     imageSize=<200 200>
     margin=<ltrb 0 0 0 0>
     outlineSettings {
      bUseBrushTransparency=false
      roundingType=HalfHeightRadius
      width=0
      color {colorUseRule=UseColor_Specified, specifiedColor=<rgba 0 0 0 0>}
      cornerRadii=<0 0 0 0>
     }
     tintColor {colorUseRule=UseColor_Specified, specifiedColor=<rgba 1 1 1 1>}
     uVRegion {bIsValid=false, max=<0 0>, min=<0 0>}
    }}
   slot {class=VerticalBoxSlot, properties {horizontalAlignment=HAlign_Center, verticalAlignment=VAlign_Top, padding=<ltrb 0 50 0 50>}}
  }
  VerticalBox_154 {
   widgetClassPath=&VerticalBox
   parent=VerticalBox_4
   widget=VerticalBox_154
   slot {class=VerticalBoxSlot, properties {padding=<ltrb 0 0 0 64>}}
  }
  Text_Desc {
   widgetClassPath=&CommonTextBlock
   bIsVariable=true
   parent=VerticalBox_154
   widget=Text_Desc
   properties {
    autoWrapText=true
    justification=Center
    text=Scan the QR Code with your mobile device
    colorAndOpacity {colorUseRule=UseColor_Specified, specifiedColor=<rgba 0.5776 0.5776 0.5776 1>}
    font {
     bForceMonospaced=false
     bMaterialIsStencil=false
     fontMaterial=null
     fontObject=&HeadingNow
     letterSpacing=0
     monospacedWidth=1
     size=15.75
     skewAmount=0
     typefaceFontName=74 Regular
     outlineSettings {
      bApplyOutlineToDropShadows=false
      bMiteredCorners=false
      bSeparateFillAlpha=false
      outlineMaterial=null
      outlineSize=0
      outlineColor=<rgba 0 0 0 1>
     }
    }
    shadowOffset=<0 0>
   }
   slot {class=VerticalBoxSlot, properties {verticalAlignment=VAlign_Top, padding=<ltrb 0 0 0 36>}}
  }
  Button_URL {
   widgetClassPath=&Button
   parent=VerticalBox_154
   widget=Button_URL
   properties {
    isFocusable=false
    widgetStyle {
     clickedSlateSound {resourceObject=null}
     disabled {
      bIsDynamicallyLoaded=false
      drawAs=NoDrawType
      imageType=NoImage
      mirroring=NoMirror
      resourceName=None
      resourceObject=null
      tiling=NoTile
      imageSize=<0 0>
      margin=<ltrb 0 0 0 0>
      outlineSettings {
       bUseBrushTransparency=false
       roundingType=HalfHeightRadius
       width=0
       color {colorUseRule=UseColor_Specified, specifiedColor=<rgba 0 0 0 0>}
       cornerRadii=<0 0 0 0>
      }
      tintColor {colorUseRule=UseColor_Specified, specifiedColor=<rgba 1 1 1 1>}
      uVRegion {bIsValid=false, max=<0 0>, min=<0 0>}
     }
     disabledForeground {colorUseRule=UseColor_Specified, specifiedColor=<rgba 0.5271 0.5271 0.5271 1>}
     hovered {
      bIsDynamicallyLoaded=false
      drawAs=NoDrawType
      imageType=NoImage
      mirroring=NoMirror
      resourceName=None
      resourceObject=null
      tiling=NoTile
      imageSize=<32 32>
      margin=<ltrb 0 0 0 0>
      outlineSettings {
       bUseBrushTransparency=true
       roundingType=FixedRadius
       width=1
       color {colorUseRule=UseColor_Specified, specifiedColor=<rgba 0.7243 0.7243 0.7243 1>}
       cornerRadii=<4 4 4 4>
      }
      tintColor {colorUseRule=UseColor_Specified, specifiedColor=<rgba 0.7243 0.7243 0.7243 1>}
      uVRegion {bIsValid=false, max=<0 0>, min=<0 0>}
     }
     hoveredForeground {colorUseRule=UseColor_Specified, specifiedColor=<rgba 0 1 1 1>}
     hoveredSlateSound {resourceObject=null}
     normal {
      bIsDynamicallyLoaded=false
      drawAs=NoDrawType
      imageType=NoImage
      mirroring=NoMirror
      resourceName=None
      resourceObject=null
      tiling=NoTile
      imageSize=<32 32>
      margin=<ltrb 0 0 0 0>
      outlineSettings {
       bUseBrushTransparency=true
       roundingType=FixedRadius
       width=1
       color {colorUseRule=UseColor_Specified, specifiedColor=<rgba 0.6951 0.6951 0.6951 1>}
       cornerRadii=<4 4 4 4>
      }
      tintColor {colorUseRule=UseColor_Specified, specifiedColor=<rgba 0.4955 0.4955 0.4955 1>}
      uVRegion {bIsValid=false, max=<0 0>, min=<0 0>}
     }
     normalForeground {colorUseRule=UseColor_Specified, specifiedColor=<rgba 0 1 1 1>}
     normalPadding=<ltrb 0 0 0 0>
     pressed {
      bIsDynamicallyLoaded=false
      drawAs=NoDrawType
      imageType=NoImage
      mirroring=NoMirror
      resourceName=None
      resourceObject=null
      tiling=NoTile
      imageSize=<32 32>
      margin=<ltrb 0 0 0 0>
      outlineSettings {
       bUseBrushTransparency=true
       roundingType=FixedRadius
       width=1
       color {colorUseRule=UseColor_Specified, specifiedColor=<rgba 0.7243 0.7243 0.7243 1>}
       cornerRadii=<4 4 4 4>
      }
      tintColor {colorUseRule=UseColor_Specified, specifiedColor=<rgba 0.3843 0.3843 0.3843 1>}
      uVRegion {bIsValid=false, max=<0 0>, min=<0 0>}
     }
     pressedForeground {colorUseRule=UseColor_Specified, specifiedColor=<rgba 0 1 1 1>}
     pressedPadding=<ltrb 0 0 0 0>
     pressedSlateSound {resourceObject=null}
    }
   }
   slot {class=VerticalBoxSlot, properties {horizontalAlignment=HAlign_Center, verticalAlignment=VAlign_Top}}
  }
  Overlay_2 {
   widgetClassPath=&Overlay
   parent=Button_URL
   widget=Overlay_2
   slot {class=ButtonSlot, properties {horizontalAlignment=HAlign_Fill, verticalAlignment=VAlign_Fill, padding=<ltrb 0 0 0 0>}}
  }
  Text_URL {
   widgetClassPath=&CommonTextBlock
   bIsVariable=true
   parent=Overlay_2
   widget=Text_URL
   properties {
    applyLineHeightToBottomLine=false
    autoWrapText=true
    bIsScrollingEnabled=false
    justification=Center
    lineHeightPercentage=1.28
    mobileFontSizeMultiplier=1.2
    text=www.epicgames.com
    wrappingPolicy=AllowPerCharacterWrapping
    font {
     bForceMonospaced=false
     bMaterialIsStencil=false
     fontMaterial=null
     fontObject=&HeadingNow
     letterSpacing=0
     monospacedWidth=1
     size=18
     skewAmount=0
     typefaceFontName=75 Medium
     outlineSettings {
      bApplyOutlineToDropShadows=false
      bMiteredCorners=false
      bSeparateFillAlpha=false
      outlineMaterial=null
      outlineSize=0
      outlineColor=<rgba 0 0 0 1>
     }
    }
    shadowOffset=<0 0>
    shapedTextOptions {bOverride_TextFlowDirection=true, bOverride_TextShapingMethod=false, textFlowDirection=Culture, textShapingMethod=Auto}
    strikeBrush {
     bIsDynamicallyLoaded=false
     drawAs=NoDrawType
     imageType=NoImage
     mirroring=NoMirror
     resourceName=None
     resourceObject=null
     tiling=NoTile
     imageSize=<32 32>
     margin=<ltrb 0 0 0 0>
     outlineSettings {
      bUseBrushTransparency=false
      roundingType=HalfHeightRadius
      width=0
      color {colorUseRule=UseColor_Specified, specifiedColor=<rgba 0 0 0 0>}
      cornerRadii=<0 0 0 0>
     }
     tintColor {colorUseRule=UseColor_Specified, specifiedColor=<rgba 1 1 1 1>}
     uVRegion {bIsValid=false, max=<0 0>, min=<0 0>}
    }
   }
   slot {class=OverlaySlot, properties {horizontalAlignment=HAlign_Center, verticalAlignment=VAlign_Center}}
  }
  WBP_Outline {
   widgetClassPath=/FortUILibrary/UIKit/Buttons/Blocks/WBP_UIKit_Block_Outline
   bIsVariable=true
   parent=Overlay_2
   widget=WBP_Outline
   properties {
    bottomLeftCornerRadius=20
    bottomRightCornerRadius=20
    navigation=&self:WidgetTree.WBP_Outline.WidgetNavigation_0
    priority=100
    radiusInPixels=true
    topLeftCornerRadius=20
    topRightCornerRadius=20
    transitionTiming {
     deselectingFocused=0
     deselectingIdle=0
     disabling=0
     enabling=0
     focusing=1
     hovering=0
     pressing=0
     releasing=0
     selecting=0
     unfocusing=0
     unhovering=0
    }
   }
   slot {class=OverlaySlot, properties {horizontalAlignment=HAlign_Fill, verticalAlignment=VAlign_Fill, padding=<ltrb -16 -16 -16 -16>}}
  }
  Button_Close {
   widgetClassPath=&WBP_UIKit_Button_Regular
   bIsVariable=true
   parent=VerticalBox_4
   widget=Button_Close
   properties {
    bAutoUpdateFromInputAction=true
    bHideInputActionWithKeyboard=true
    text=Close
    triggeringInputAction {dataTable=&InputActionDataTable, rowName=Close}
   }
   slot {class=VerticalBoxSlot, properties {verticalAlignment=VAlign_Top}}
  }
  VisiblityBorder_Touch {
   widgetClassPath=&CommonHardwareVisibilityBorder
   namedSlotHost=Dialog_Internal
   widget=VisiblityBorder_Touch
   properties {
    horizontalAlignment=HAlign_Right
    verticalAlignment=VAlign_Top
    padding=<ltrb 0 32 32 0>
    visibilityQuery {
     autoDescription=" ANY( Input.Touch )"
     tokenStreamVersion=0
     userDescription=""
     queryTokenStream [0,1,1,1,0]
     tagDictionary [{tagName=Input.Touch}]
    }
   }
  }
  Button_BackTouch {
   widgetClassPath=/FortUILibrary/Foundation/Buttons/CloseButton
   bIsVariable=true
   parent=VisiblityBorder_Touch
   widget=Button_BackTouch
   properties {
    onButtonBaseClicked=(&self:WidgetTree.Button_BackTouch.HandleOnButtonClicked)
    onButtonBaseDoubleClicked=(&self:WidgetTree.Button_BackTouch.HandleOnButtonDoubleClicked)
    onButtonBaseHovered=(&self:WidgetTree.Button_BackTouch.HandleOnButtonHovered)
    onButtonBaseUnhovered=(&self:WidgetTree.Button_BackTouch.HandleOnButtonUnhovered)
    onSelectedChangedBase=(&self:WidgetTree.Button_BackTouch.HandleOnSelectedChanged)
    renderTransform {angle=0, scale=<1 1>, shear=<0 0>, translation=<0 10>}
   }
   slot {class=BorderSlot, properties {horizontalAlignment=HAlign_Right, verticalAlignment=VAlign_Top, padding=<ltrb 0 32 32 0>}}
  }
 ]
}
