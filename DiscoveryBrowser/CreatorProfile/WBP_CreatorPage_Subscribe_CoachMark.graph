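# Onboarding coach mark on the creator profile page that tells players they can follow developers to get notifications about new islands.
# A sidebar onboard tooltip widget wrapping the shared UIKit coach mark as a modal, with the dismiss button hidden and a single "GOT IT" action bound to escape.
# Showing the tooltip activates the coach mark, which takes focus on that button.
# Clicking it reports the action and close-animation completion to the parent, then deactivates both widgets.
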WidgetBlueprint /DiscoveryBrowser/CreatorProfile/WBP_CreatorPage_Subscribe_CoachMark
graphs [
 BP_GetDesiredFocusTarget {nodes [
   Get Desired Focus Target()
   > Return Node!(
    execute
    ReturnValue:Widget Object Reference=Get OtherActionButton!(self:WBP UIKit Coach Mark Base Object Reference=Get WBP_UIKit_CoachMark() -> (WBP_UIKit_CoachMark:WBP UIKit Coach Mark Object Reference)) -> (OtherActionButton:WBP UIKit Button Regular Object Reference)
   )
  ]}
 EventGraph {nodes [
   "Button Other Action Clicked (WBP_UIKit_CoachMark)"() {is=title K2Node_ComponentBoundEvent}
   > HandleOtherActionClicked(self:Fort Sidebar Onboard Tooltip Widget Object Reference)
   > NotifyCloseAnimationFinished(self:Fort Sidebar Onboard Tooltip Widget Object Reference)
   > DeactivateWidget(self=Get WBP_UIKit_CoachMark())
   > DeactivateWidget()
   Event OnShowTooltip() {is=title K2Node_Event}
   > ActivateWidget(self=Get WBP_UIKit_CoachMark())
  ]}
]
variables []
widgets {
 info {
  inheritedWidgetCount=2
  parentClass=/Script/FortniteUI.FortSidebarOnboardTooltipWidget
  rootWidgetClass=/FortUILibrary/UIKit/CoachMark/WBP_UIKit_CoachMark.WBP_UIKit_CoachMark_C
  widgetCount=3
 }
 widgets [
  WBP_UIKit_CoachMark {
   widgetClassPath=/FortUILibrary/UIKit/CoachMark/WBP_UIKit_CoachMark
   bIsVariable=true
   widget=WBP_UIKit_CoachMark
   properties {
    actionDomainOverride=/Game/UI/ActionDomain/ActionDomain_Modal
    arrowPosition=Left_Up
    bIsFocusable=true
    navigation=&self:WidgetTree.WBP_UIKit_CoachMark.WidgetNavigation_0
    shouldShowDismissButton=false
    textDescription=Get in-game notifications when developers you follow release new islands.
    textDismissButton=""
    textHeader="Stay updated!"
    textOtherActionButton=GOT IT
    visibility=SelfHitTestInvisible
    input Action Other {dataTable=&InputActionDataTable, rowName=GenericEscape}
   }
  }
  Dismiss_Button {widgetClassPath=&CommonButtonLegacy, bInherited=true, widget=null}
  OtherAction_Button {widgetClassPath=&CommonButtonLegacy, bInherited=true, widget=null}
 ]
}
