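# Creator page Support-A-Creator modal, a CommonActivatableWidget dialog for applying, overwriting or removing a creator code.
# Setup Creator Support Modal initializes FortSupportCreatorVM with the creator's name, then compares it to the supported creator to pick the Support, Overwrite or Applied panel.
# Support and Confirm apply the code and show Applied; Clear applies an empty creator and shows Removed.
# Panel changes refresh focus next tick to the button stored in LastFocusedCreatorWidget.
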WidgetBlueprint /DiscoveryBrowser/CreatorProfile/WBP_CreatorPage_SupportACreator_Modal
graphs [
 BP_GetDesiredFocusTarget {nodes [
   Get Desired Focus Target()
   > Return Node!(execute, ReturnValue:Widget Object Reference=Get LastFocusedCreatorWidget() -> (LastFocusedCreatorWidget:Widget Object Reference))
  ]}
 EventGraph {nodes [
   n1 = "On Clicked (WBP_UIKit_Button_Support)"!() -> (OutputDelegate, then, Button:Common Button Base Object Reference) {is=title K2Node_ComponentBoundEvent}
   n2 = Get Current Creator Name() -> (Current Creator Name:Text)
   ApplyProspectSupportedCreator!(
    execute=[@n1.then, @n4.then]
    self:Support A Creator View Model Object Reference=Get FortSupportCreatorVM() -> (FortSupportCreatorVM:Support A Creator View Model Object Reference)
    InSourceName:String="To String (Text)"(InText=@n2)
   ) -> (then)
   > Set Current Creator Name!(execute, self, Name:String="To String (Text)"(InText=@n2)) -> (then) {is=title K2Node_CallFunction}
   > Set LastFocusedCreatorWidget(LastFocusedCreatorWidget:Widget Object Reference=Get WBP_UIKit_Button_GotIt() -> (WBP_UIKit_Button_GotIt:WBP UIKit Button Regular Object Reference)) -> (Output_Get:Widget Object Reference)
   > SetActiveWidget(
    self:Widget Switcher Object Reference=Get VisibilitySwitcher() -> (VisibilitySwitcher:Common Animated Switcher Object Reference)
    Widget=Get SB_AppliedCreatorCode() -> (SB_AppliedCreatorCode:Size Box Object Reference)
   )
   n3 = "On Button Clicked (Button_BackTouch)"!() -> (OutputDelegate, then, Button:Common Button Legacy Object Reference) {is=title K2Node_ComponentBoundEvent}
   DeactivateWidget(execute=[@n3.then, @n5.then, @n6.then, @n7.then, @n8.then])
   n4 = "On Clicked (WBP_UIKit_Button_Confirm)"!() -> (OutputDelegate, then, Button:Common Button Base Object Reference) {is=title K2Node_ComponentBoundEvent}
   n5 = "On Clicked (WBP_UIKit_Button_Cancel)"!() -> (OutputDelegate, then, Button:Common Button Base Object Reference) {is=title K2Node_ComponentBoundEvent}
   n6 = "On Clicked (WBP_UIKit_Button_Close)"!() -> (OutputDelegate, then, Button:Common Button Base Object Reference) {is=title K2Node_ComponentBoundEvent}
   n7 = "On Clicked (WBP_UIKit_Button_Back)"!() -> (OutputDelegate, then, Button:Common Button Base Object Reference) {is=title K2Node_ComponentBoundEvent}
   "On Clicked (WBP_UIKit_Button_Clear)"!() -> (OutputDelegate, then, Button:Common Button Base Object Reference) {is=title K2Node_ComponentBoundEvent}
   > Init!(execute, self:Support A Creator View Model Object Reference=Get FortSupportCreatorVM(), InProspectSupportedCreator:String) -> (then)
   > ApplyProspectSupportedCreator(self=Get FortSupportCreatorVM(), InSourceName=Creator Page)
   > Set LastFocusedCreatorWidget(LastFocusedCreatorWidget=Get WBP_UIKit_Button_Close() -> (WBP_UIKit_Button_Close:WBP UIKit Button Regular Object Reference))
   > SetActiveWidget(self=Get VisibilitySwitcher(), Widget=Get SB_RemovedCode() -> (SB_RemovedCode:Size Box Object Reference))
   "On Active Widget Index Changed BP (VisibilitySwitcher)"!() -> (OutputDelegate, then, ActiveWidget:Widget Object Reference, ActiveIndex:int=0) {is=title K2Node_ComponentBoundEvent}
   > DelayUntilNextTick!(execute) -> (then)
   > RequestRefreshFocus(self:Common Activatable Widget Object Reference)
   n8 = "On Clicked (WBP_UIKit_Button_GotIt)"!() -> (OutputDelegate, then, Button:Common Button Base Object Reference) {is=title K2Node_ComponentBoundEvent}
   n9 = Setup Creator Support Modal!() -> (OutputDelegate, then, Current Creator Name:Text) {is=title K2Node_CustomEvent}
   > n10 = Set Current Creator Name(Current Creator Name:Text=@n9.Current Creator Name) -> (Output_Get:Text) {is=title K2Node_VariableSet}
   n11 = "To String (Text)"(InText=@n10)
   > Init(self=Get FortSupportCreatorVM(), InProspectSupportedCreator=@n11)
   n12 = Get Current Supported Creator() -> (Current Supported Creator:String)
   n13 = "Equal Exactly (String)"(A=@n12, B=@n11)
   n14 = IsEmpty(InString=@n12)
   n15 = Get WBP_UIKit_Button_Support() -> (WBP_UIKit_Button_Support:WBP UIKit Button Regular Object Reference)
   n16 = Get WBP_UIKit_Button_Confirm() -> (WBP_UIKit_Button_Confirm:WBP UIKit Button Regular Object Reference)
   > Set LastFocusedCreatorWidget(LastFocusedCreatorWidget=Select(
     Option 0:WBP UIKit Button Regular Object Reference=Select(Option 0=@n16, Option 1=@n15, Index=@n14)
     Option 1:WBP UIKit Button Regular Object Reference=Get WBP_UIKit_Button_GotIt()
     Index=@n13
    ) -> (ReturnValue:WBP UIKit Button Regular Object Reference))
   n17 = Get SB_SupportACreator() -> (SB_SupportACreator:Size Box Object Reference)
   n18 = Get SB_OverwriteCreatorCode() -> (SB_OverwriteCreatorCode:Size Box Object Reference)
   > SetActiveWidget(
    self=Get VisibilitySwitcher()
    Widget=Select(
     Option 0:Size Box Object Reference=Select(Option 0=@n18, Option 1=@n17, Index=@n14)
     Option 1:Size Box Object Reference=Get SB_AppliedCreatorCode()
     Index=@n13
    ) -> (ReturnValue:Size Box Object Reference)
   )
  ]}
 Set Current Creator Name {nodes [
   n1 = Set Current Creator Name!() -> (then, Name:String) {is=title K2Node_FunctionEntry}
   > Set Current Supported Creator(Current Supported Creator:String=@n1) -> (Output_Get:String)
   n2 = "To Text (String)"(InString=@n1)
   > "SetText (Text)"(
    self=[
     Get Text_Username_1() -> (Text_Username_1:Common Text Object Reference)
     Get Text_CurrentCreator() -> (Text_CurrentCreator:Common Text Object Reference)
    ]
    InText=@n2
   )
   > "SetText (Text)"(self=Get Text_Username_2() -> (Text_Username_2:Common Text Object Reference), InText=Get Current Creator Name())
   > Set Text(
    Text=Format Text!(Format="LOCTEXT(\"For the next 14 days, your in-game purchases will now support {CREATOR NAME}.\")", CREATOR NAME:Text=@n2) -> (Result)
    self=Get Text_Desc_With_Name() -> (Text_Desc_With_Name:Common Text Object Reference)
   )
  ]}
 Set Prospect Creator Name {nodes [
   n1 = Set Prospect Creator Name!() -> (then, Name:String) {is=title K2Node_FunctionEntry}
   > Set Prospect Supporting Name(Prospect Supporting Name:String=@n1) -> (Output_Get:String)
   > "SetText (Text)"(
    self=[Get Text_Username() -> (Text_Username:Common Text Object Reference), Get Text_NewCreator() -> (Text_NewCreator:Common Text Object Reference)]
    InText="To Text (String)"(InString=@n1)
   )
  ]}
]
variables [
 Current Creator Name {type=string, default=""}
 Current Supported Creator {type=string, default=""}
 LastFocusedCreatorWidget {type=Widget, default=null}
 Prospect Supporting Name {type=string, default=""}
]
widgets {
 info {
  inheritedWidgetCount=0
  namedSlotCount=2
  parentClass=&CommonActivatableWidget
  rootWidgetClass=/FortUILibrary/Systems/Dialog/WBP_UIKit_Dialog_Internal.WBP_UIKit_Dialog_Internal_C
  widgetCount=54
 }
 widgets [
  Dialog_Internal {
   widgetClassPath=/FortUILibrary/Systems/Dialog/WBP_UIKit_Dialog_Internal
   bIsVariable=true
   widget=Dialog_Internal
   properties {
    bIsFocusable=true
    designTimePreview_DialogVM=null
    useClickCatch=true
    dialogSIze {
     height Override {enable=false, value=684}
     max Aspect Ratio {enable=false, value=0}
     max Desired Height {enable=false, value=800}
     max Desired Width {enable=false, value=0}
     min Aspect Ratio {enable=false, value=0}
     min Desired Height {enable=false, value=0}
     min Desired Width {enable=false, value=0}
     width Override {enable=false, value=864}
    }
   }
  }
  Content {widgetClassPath=&Overlay, namedSlotHost=Dialog_Internal, widget=Content}
  VisibilitySwitcher {
   widgetClassPath=/Script/CommonUI.CommonAnimatedSwitcher
   bIsVariable=true
   parent=Content
   widget=VisibilitySwitcher
   properties {
    bOverride_Cursor=true
    navigation=&self:WidgetTree.VisibilitySwitcher.WidgetNavigation_0
    transitionDuration=0.2
   }
   slot {class=OverlaySlot, properties {horizontalAlignment=HAlign_Fill, verticalAlignment=VAlign_Fill}}
  }
  SB_SupportACreator {
   widgetClassPath=&SizeBox
   bIsVariable=true
   parent=VisibilitySwitcher
   widget=SB_SupportACreator
   properties {bOverride_HeightOverride=true, bOverride_WidthOverride=true, heightOverride=460, maxDesiredWidth=820, widthOverride=820}
   slot {class=WidgetSwitcherSlot, properties {padding=<ltrb 80 80 80 48>}}
  }
  VB_SupportACreator {widgetClassPath=&VerticalBox, parent=SB_SupportACreator, widget=VB_SupportACreator, slot {class=SizeBoxSlot}}
  Text_ModalTitle {
   widgetClassPath=&CommonTextBlock
   parent=VB_SupportACreator
   widget=Text_ModalTitle
   properties {
    autoWrapText=true
    text=Support a creator
    textTransformPolicy=ToUpper
    font {
     bForceMonospaced=false
     bMaterialIsStencil=false
     fontMaterial=null
     fontObject=&HeadingNow
     letterSpacing=0
     monospacedWidth=1
     size=25.5
     skewAmount=0
     typefaceFontName=86 Bold Italic
     outlineSettings {
      bApplyOutlineToDropShadows=false
      bMiteredCorners=false
      bSeparateFillAlpha=false
      outlineMaterial=null
      outlineSize=0
      outlineColor=<rgba 0 0 0 1>
     }
    }
    shadowOffset=<0 0>
   }
   slot {class=VerticalBoxSlot, properties {verticalAlignment=VAlign_Top, padding=<ltrb 0 0 0 48>}}
  }
  Text_Desc {
   widgetClassPath=&CommonTextBlock
   parent=VB_SupportACreator
   widget=Text_Desc
   properties {
    autoWrapText=true
    text="Declare your support! Your in-game purchases will help support this creator."
    colorAndOpacity {colorUseRule=UseColor_Specified, specifiedColor=<rgba 0.5776 0.5776 0.5776 1>}
    font {
     bForceMonospaced=false
     bMaterialIsStencil=false
     fontMaterial=null
     fontObject=&HeadingNow
     letterSpacing=0
     monospacedWidth=1
     size=18
     skewAmount=0
     typefaceFontName=74 Regular
     outlineSettings {
      bApplyOutlineToDropShadows=false
      bMiteredCorners=false
      bSeparateFillAlpha=false
      outlineMaterial=null
      outlineSize=0
      outlineColor=<rgba 0 0 0 1>
     }
    }
    shadowOffset=<0 0>
   }
   slot {class=VerticalBoxSlot, properties {verticalAlignment=VAlign_Top}}
  }
  Username {
   widgetClassPath=&Overlay
   parent=VB_SupportACreator
   widget=Username
   slot {class=VerticalBoxSlot, properties {padding=<ltrb 0 32 0 12>}}
  }
  UIKit_Backplate {
   widgetClassPath=&WBP_UIKit_Backplate
   parent=Username
   widget=UIKit_Backplate
   properties {brightness=Medium, cornerRadius=Interactive Component, colorAndOpacity=<rgba 1 1 1 0.5>}
   slot {class=OverlaySlot, properties {horizontalAlignment=HAlign_Fill, verticalAlignment=VAlign_Fill}}
  }
  HorizontalBox_0 {
   widgetClassPath=&HorizontalBox
   parent=Username
   widget=HorizontalBox_0
   slot {class=OverlaySlot, properties {horizontalAlignment=HAlign_Center, verticalAlignment=VAlign_Fill, padding=<ltrb 24 24 24 24>}}
  }
  Text_Username {
   widgetClassPath=&CommonTextBlock
   bIsVariable=true
   parent=HorizontalBox_0
   widget=Text_Username
   properties {
    justification=Center
    scrollStyle=&TextScrollStyle-Base_C
    text=Creator Name
    textTransformPolicy=ToUpper
    font {
     bForceMonospaced=false
     bMaterialIsStencil=false
     fontMaterial=null
     fontObject=&HeadingNow
     letterSpacing=0
     monospacedWidth=1
     size=25.5
     skewAmount=0
     typefaceFontName=86 Bold Italic
     outlineSettings {
      bApplyOutlineToDropShadows=false
      bMiteredCorners=false
      bSeparateFillAlpha=false
      outlineMaterial=null
      outlineSize=0
      outlineColor=<rgba 0 0 0 1>
     }
    }
    shadowOffset=<0 0>
   }
   slot {class=HorizontalBoxSlot, properties {horizontalAlignment=HAlign_Left, verticalAlignment=VAlign_Top, size {sizeRule=Fill, value=1}}}
  }
  Text_Desc_Bottom {
   widgetClassPath=&CommonTextBlock
   parent=VB_SupportACreator
   widget=Text_Desc_Bottom
   properties {
    autoWrapText=true
    text="By using a Creator Code, you are supporting the creator with your purchase. You will not receive anything from using a Creator Code. If someone is claiming you will receive a benefit, it might be a scam."
    colorAndOpacity {colorUseRule=UseColor_Specified, specifiedColor=<rgba 0.5776 0.5776 0.5776 1>}
    font {
     bForceMonospaced=false
     bMaterialIsStencil=false
     fontMaterial=null
     fontObject=&HeadingNow
     letterSpacing=0
     monospacedWidth=1
     size=11.25
     skewAmount=0
     typefaceFontName=74 Regular
     outlineSettings {
      bApplyOutlineToDropShadows=false
      bMiteredCorners=false
      bSeparateFillAlpha=false
      outlineMaterial=null
      outlineSize=0
      outlineColor=<rgba 0 0 0 1>
     }
    }
    shadowOffset=<0 0>
   }
   slot {class=VerticalBoxSlot, properties {verticalAlignment=VAlign_Top, padding=<ltrb 0 0 0 24>}}
  }
  HB_SupportACreator {
   widgetClassPath=&HorizontalBox
   parent=VB_SupportACreator
   widget=HB_SupportACreator
   slot {class=VerticalBoxSlot, properties {horizontalAlignment=HAlign_Left, verticalAlignment=VAlign_Bottom, size {sizeRule=Fill, value=1}}}
  }
  WBP_UIKit_Button_Support {
   widgetClassPath=&WBP_UIKit_Button_Regular
   bIsVariable=true
   parent=HB_SupportACreator
   widget=WBP_UIKit_Button_Support
   properties {bDisplayInputActionWhenNotInteractable=false, bHideInputAction=true, bHideInputActionWithKeyboard=true, buttonType=Quiet, text=Support}
   slot {class=HorizontalBoxSlot, properties {padding=<ltrb 0 0 24 0>}}
  }
  WBP_UIKit_Button_Back {
   widgetClassPath=&WBP_UIKit_Button_Regular
   bIsVariable=true
   parent=HB_SupportACreator
   widget=WBP_UIKit_Button_Back
   properties {
    bDisplayInputActionWhenNotInteractable=false
    bHideInputAction=true
    bHideInputActionWithKeyboard=true
    buttonType=Quiet
    text=Back
    triggeringInputAction {dataTable=&InputActionDataTable, rowName=GenericBack}
   }
   slot {class=HorizontalBoxSlot, properties {padding=<ltrb 0 0 24 0>}}
  }
  SB_AppliedCreatorCode {
   widgetClassPath=&SizeBox
   bIsVariable=true
   parent=VisibilitySwitcher
   widget=SB_AppliedCreatorCode
   properties {bOverride_HeightOverride=true, bOverride_WidthOverride=true, heightOverride=460, maxDesiredWidth=820, widthOverride=820}
   slot {class=WidgetSwitcherSlot, properties {padding=<ltrb 80 80 80 48>}}
  }
  VB_AppliedCreatorCode {widgetClassPath=&VerticalBox, parent=SB_AppliedCreatorCode, widget=VB_AppliedCreatorCode, slot {class=SizeBoxSlot}}
  Text_ModalTitle_1 {
   widgetClassPath=&CommonTextBlock
   parent=VB_AppliedCreatorCode
   widget=Text_ModalTitle_1
   properties {
    autoWrapText=true
    text="Applied Creator Code!"
    textTransformPolicy=ToUpper
    font {
     bForceMonospaced=false
     bMaterialIsStencil=false
     fontMaterial=null
     fontObject=&HeadingNow
     letterSpacing=0
     monospacedWidth=1
     size=25.5
     skewAmount=0
     typefaceFontName=86 Bold Italic
     outlineSettings {
      bApplyOutlineToDropShadows=false
      bMiteredCorners=false
      bSeparateFillAlpha=false
      outlineMaterial=null
      outlineSize=0
      outlineColor=<rgba 0 0 0 1>
     }
    }
    shadowOffset=<0 0>
   }
   slot {class=VerticalBoxSlot, properties {padding=<ltrb 0 0 0 48>}}
  }
  Text_Desc_With_Name {
   widgetClassPath=&CommonTextBlock
   bIsVariable=true
   parent=VB_AppliedCreatorCode
   widget=Text_Desc_With_Name
   properties {
    autoWrapText=true
    text="For the next 14 days, your in-game purchases will now support {DONT FORGET}."
    colorAndOpacity {colorUseRule=UseColor_Specified, specifiedColor=<rgba 0.5776 0.5776 0.5776 1>}
    font {
     bForceMonospaced=false
     bMaterialIsStencil=false
     fontMaterial=null
     fontObject=&HeadingNow
     letterSpacing=0
     monospacedWidth=1
     size=18
     skewAmount=0
     typefaceFontName=74 Regular
     outlineSettings {
      bApplyOutlineToDropShadows=false
      bMiteredCorners=false
      bSeparateFillAlpha=false
      outlineMaterial=null
      outlineSize=0
      outlineColor=<rgba 0 0 0 1>
     }
    }
    shadowOffset=<0 0>
   }
   slot {class=VerticalBoxSlot}
  }
  Username_1 {
   widgetClassPath=&Overlay
   parent=VB_AppliedCreatorCode
   widget=Username_1
   slot {class=VerticalBoxSlot, properties {padding=<ltrb 0 32 0 24>}}
  }
  UIKit_Backplate_1 {
   widgetClassPath=&WBP_UIKit_Backplate
   parent=Username_1
   widget=UIKit_Backplate_1
   properties {brightness=Medium, cornerRadius=Interactive Component, colorAndOpacity=<rgba 1 1 1 0.5>}
   slot {class=OverlaySlot, properties {horizontalAlignment=HAlign_Fill, verticalAlignment=VAlign_Fill}}
  }
  HorizontalBox {
   widgetClassPath=&HorizontalBox
   parent=Username_1
   widget=HorizontalBox
   slot {class=OverlaySlot, properties {horizontalAlignment=HAlign_Center, verticalAlignment=VAlign_Fill, padding=<ltrb 24 24 24 24>}}
  }
  Icon_Check_1 {
   widgetClassPath=&Image
   parent=HorizontalBox
   widget=Icon_Check_1
   properties {
    visibility=HitTestInvisible
    brush {
     bIsDynamicallyLoaded=false
     drawAs=Image
     imageType=FullColor
     mirroring=NoMirror
     resourceName=None
     resourceObject=/Game/UI/Foundation/Textures/Icons/Checks/T_UI_Icon_Check_MP_32
     tiling=NoTile
     imageSize=<32 32>
     margin=<ltrb 0 0 0 0>
     outlineSettings {
      bUseBrushTransparency=false
      roundingType=FixedRadius
      width=2
      color {colorUseRule=UseColor_Specified, specifiedColor=<rgba 1 1 1 1>}
      cornerRadii=<0 0 0 0>
     }
     tintColor {colorUseRule=UseColor_Specified, specifiedColor=<rgba 0 1 0.1991 1>}
     uVRegion {bIsValid=false, max=<0 0>, min=<0 0>}
    }
    colorAndOpacity=<rgba 0.5776 0.5776 0.5776 1>
   }
   slot {class=HorizontalBoxSlot, properties {verticalAlignment=VAlign_Center, padding=<ltrb 0 0 8 0>}}
  }
  Text_Username_1 {
   widgetClassPath=&CommonTextBlock
   bIsVariable=true
   parent=HorizontalBox
   widget=Text_Username_1
   properties {
    justification=Center
    scrollStyle=&TextScrollStyle-Base_C
    text=Creator Name
    textTransformPolicy=ToUpper
    font {
     bForceMonospaced=false
     bMaterialIsStencil=false
     fontMaterial=null
     fontObject=&HeadingNow
     letterSpacing=0
     monospacedWidth=1
     size=25.5
     skewAmount=0
     typefaceFontName=86 Bold Italic
     outlineSettings {
      bApplyOutlineToDropShadows=false
      bMiteredCorners=false
      bSeparateFillAlpha=false
      outlineMaterial=null
      outlineSize=0
      outlineColor=<rgba 0 0 0 1>
     }
    }
    shadowOffset=<0 0>
   }
   slot {class=HorizontalBoxSlot, properties {horizontalAlignment=HAlign_Left, verticalAlignment=VAlign_Top, size {sizeRule=Fill, value=1}}}
  }
  HB_SupportACreator_1 {
   widgetClassPath=&HorizontalBox
   parent=VB_AppliedCreatorCode
   widget=HB_SupportACreator_1
   slot {class=VerticalBoxSlot, properties {horizontalAlignment=HAlign_Left, verticalAlignment=VAlign_Bottom, size {sizeRule=Fill, value=1}}}
  }
  WBP_UIKit_Button_GotIt {
   widgetClassPath=&WBP_UIKit_Button_Regular
   bIsVariable=true
   parent=HB_SupportACreator_1
   widget=WBP_UIKit_Button_GotIt
   properties {
    bDisplayInputActionWhenNotInteractable=false
    bHideInputAction=true
    bHideInputActionWithKeyboard=true
    buttonType=Quiet
    text=Got It
    triggeringInputAction {dataTable=&InputActionDataTable, rowName=GenericBack}
   }
   slot {class=HorizontalBoxSlot, properties {padding=<ltrb 0 0 24 0>}}
  }
  WBP_UIKit_Button_Clear {
   widgetClassPath=&WBP_UIKit_Button_Regular
   bIsVariable=true
   parent=HB_SupportACreator_1
   widget=WBP_UIKit_Button_Clear
   properties {
    bDisplayInputActionWhenNotInteractable=false
    bHideInputAction=true
    bHideInputActionWithKeyboard=true
    buttonType=Quiet
    navigation=&self:WidgetTree.WBP_UIKit_Button_Clear.WidgetNavigation_1
    text=Clear
   }
   slot {class=HorizontalBoxSlot, properties {padding=<ltrb 0 0 24 0>}}
  }
  SB_OverwriteCreatorCode {
   widgetClassPath=&SizeBox
   bIsVariable=true
   parent=VisibilitySwitcher
   widget=SB_OverwriteCreatorCode
   properties {bOverride_MinDesiredHeight=true, bOverride_MinDesiredWidth=true, minDesiredHeight=460, minDesiredWidth=820}
   slot {class=WidgetSwitcherSlot, properties {padding=<ltrb 80 80 80 48>}}
  }
  VB_OverwriteCreatorCode {widgetClassPath=&VerticalBox, parent=SB_OverwriteCreatorCode, widget=VB_OverwriteCreatorCode, slot {class=SizeBoxSlot}}
  Text_ModalTitle_3 {
   widgetClassPath=&CommonTextBlock
   parent=VB_OverwriteCreatorCode
   widget=Text_ModalTitle_3
   properties {
    autoWrapText=true
    text=Overwrite Creator Code?
    textTransformPolicy=ToUpper
    font {
     bForceMonospaced=false
     bMaterialIsStencil=false
     fontMaterial=null
     fontObject=&HeadingNow
     letterSpacing=0
     monospacedWidth=1
     size=25.5
     skewAmount=0
     typefaceFontName=86 Bold Italic
     outlineSettings {
      bApplyOutlineToDropShadows=false
      bMiteredCorners=false
      bSeparateFillAlpha=false
      outlineMaterial=null
      outlineSize=0
      outlineColor=<rgba 0 0 0 1>
     }
    }
    shadowOffset=<0 0>
   }
   slot {class=VerticalBoxSlot, properties {verticalAlignment=VAlign_Top, padding=<ltrb 0 0 0 48>}}
  }
  Text_Desc_3 {
   widgetClassPath=&CommonTextBlock
   parent=VB_OverwriteCreatorCode
   widget=Text_Desc_3
   properties {
    autoWrapText=true
    text=Applying a new Creator Code will overwrite your current one.
    colorAndOpacity {colorUseRule=UseColor_Specified, specifiedColor=<rgba 0.5776 0.5776 0.5776 1>}
    font {
     bForceMonospaced=false
     bMaterialIsStencil=false
     fontMaterial=null
     fontObject=&HeadingNow
     letterSpacing=0
     monospacedWidth=1
     size=18
     skewAmount=0
     typefaceFontName=74 Regular
     outlineSettings {
      bApplyOutlineToDropShadows=false
      bMiteredCorners=false
      bSeparateFillAlpha=false
      outlineMaterial=null
      outlineSize=0
      outlineColor=<rgba 0 0 0 1>
     }
    }
    shadowOffset=<0 0>
   }
   slot {class=VerticalBoxSlot, properties {verticalAlignment=VAlign_Top}}
  }
  HorizontalBox_2 {
   widgetClassPath=&HorizontalBox
   parent=VB_OverwriteCreatorCode
   widget=HorizontalBox_2
   slot {class=VerticalBoxSlot, properties {horizontalAlignment=HAlign_Center, verticalAlignment=VAlign_Center, padding=<ltrb 0 72 0 0>}}
  }
  VerticalBox_0 {
   widgetClassPath=&VerticalBox
   parent=HorizontalBox_2
   widget=VerticalBox_0
   slot {
    class=HorizontalBoxSlot
    properties {horizontalAlignment=HAlign_Right, verticalAlignment=VAlign_Center, size {sizeRule=Automatic, value=0.5}}
   }
  }
  Text_Desc_5 {
   widgetClassPath=&CommonTextBlock
   parent=VerticalBox_0
   widget=Text_Desc_5
   properties {
    autoWrapText=true
    text=Current
    colorAndOpacity {colorUseRule=UseColor_Specified, specifiedColor=<rgba 0.5776 0.5776 0.5776 1>}
    font {
     bForceMonospaced=false
     bMaterialIsStencil=false
     fontMaterial=null
     fontObject=&HeadingNow
     letterSpacing=0
     monospacedWidth=1
     size=15.75
     skewAmount=0
     typefaceFontName=74 Regular
     outlineSettings {
      bApplyOutlineToDropShadows=false
      bMiteredCorners=false
      bSeparateFillAlpha=false
      outlineMaterial=null
      outlineSize=0
      outlineColor=<rgba 0 0 0 1>
     }
    }
    shadowOffset=<0 0>
   }
   slot {class=VerticalBoxSlot, properties {horizontalAlignment=HAlign_Center}}
  }
  Text_CurrentCreator {
   widgetClassPath=&CommonTextBlock
   bIsVariable=true
   parent=VerticalBox_0
   widget=Text_CurrentCreator
   properties {
    justification=Center
    text=Current Creator
    font {
     bForceMonospaced=false
     bMaterialIsStencil=false
     fontMaterial=null
     fontObject=&HeadingNow
     letterSpacing=0
     monospacedWidth=1
     size=21
     skewAmount=0
     typefaceFontName=75 Medium
     outlineSettings {
      bApplyOutlineToDropShadows=false
      bMiteredCorners=false
      bSeparateFillAlpha=false
      outlineMaterial=null
      outlineSize=0
      outlineColor=<rgba 0 0 0 1>
     }
    }
    shadowOffset=<0 0>
   }
   slot {class=VerticalBoxSlot, properties {horizontalAlignment=HAlign_Center}}
  }
  Image_Arrow {
   widgetClassPath=&Image
   parent=HorizontalBox_2
   widget=Image_Arrow
   properties {
    visibility=HitTestInvisible
    brush {
     bIsDynamicallyLoaded=false
     drawAs=Image
     imageType=FullColor
     mirroring=NoMirror
     resourceName=None
     resourceObject=/Game/UI/Frontend/MissionSelect/OutpostShieldPowerBar/Icon-NextArrow-L
     tiling=NoTile
     imageSize=<20 20>
     margin=<ltrb 0 0 0 0>
     outlineSettings {
      bUseBrushTransparency=false
      roundingType=HalfHeightRadius
      width=0
      color {colorUseRule=UseColor_Specified, specifiedColor=<rgba 0 0 0 0>}
      cornerRadii=<0 0 0 0>
     }
     tintColor {colorUseRule=UseColor_Specified, specifiedColor=<rgba 1 1 1 1>}
     uVRegion {bIsValid=false, max=<0 0>, min=<0 0>}
    }
    colorAndOpacity=<rgba 0.5776 0.5776 0.5776 1>
   }
   slot {class=HorizontalBoxSlot, properties {horizontalAlignment=HAlign_Center, verticalAlignment=VAlign_Center, padding=<ltrb 64 0 64 0>}}
  }
  VerticalBox {
   widgetClassPath=&VerticalBox
   parent=HorizontalBox_2
   widget=VerticalBox
   slot {class=HorizontalBoxSlot, properties {horizontalAlignment=HAlign_Left, size {sizeRule=Automatic, value=0.5}}}
  }
  Text_Desc_4 {
   widgetClassPath=&CommonTextBlock
   parent=VerticalBox
   widget=Text_Desc_4
   properties {
    autoWrapText=true
    text=New
    colorAndOpacity {colorUseRule=UseColor_Specified, specifiedColor=<rgba 0.5776 0.5776 0.5776 1>}
    font {
     bForceMonospaced=false
     bMaterialIsStencil=false
     fontMaterial=null
     fontObject=&HeadingNow
     letterSpacing=0
     monospacedWidth=1
     size=15.75
     skewAmount=0
     typefaceFontName=74 Regular
     outlineSettings {
      bApplyOutlineToDropShadows=false
      bMiteredCorners=false
      bSeparateFillAlpha=false
      outlineMaterial=null
      outlineSize=0
      outlineColor=<rgba 0 0 0 1>
     }
    }
    shadowOffset=<0 0>
   }
   slot {class=VerticalBoxSlot, properties {horizontalAlignment=HAlign_Center, verticalAlignment=VAlign_Center}}
  }
  Text_NewCreator {
   widgetClassPath=&CommonTextBlock
   bIsVariable=true
   parent=VerticalBox
   widget=Text_NewCreator
   properties {
    justification=Center
    text=New Creator
    font {
     bForceMonospaced=false
     bMaterialIsStencil=false
     fontMaterial=null
     fontObject=&HeadingNow
     letterSpacing=0
     monospacedWidth=1
     size=21
     skewAmount=0
     typefaceFontName=75 Medium
     outlineSettings {
      bApplyOutlineToDropShadows=false
      bMiteredCorners=false
      bSeparateFillAlpha=false
      outlineMaterial=null
      outlineSize=0
      outlineColor=<rgba 0 0 0 1>
     }
    }
    shadowOffset=<0 0>
   }
   slot {class=VerticalBoxSlot}
  }
  HB_SupportACreator_3 {
   widgetClassPath=&HorizontalBox
   parent=VB_OverwriteCreatorCode
   widget=HB_SupportACreator_3
   slot {class=VerticalBoxSlot, properties {horizontalAlignment=HAlign_Left, verticalAlignment=VAlign_Bottom, size {sizeRule=Fill, value=1}}}
  }
  WBP_UIKit_Button_Confirm {
   widgetClassPath=&WBP_UIKit_Button_Regular
   bIsVariable=true
   parent=HB_SupportACreator_3
   widget=WBP_UIKit_Button_Confirm
   properties {bDisplayInputActionWhenNotInteractable=false, bHideInputAction=true, bHideInputActionWithKeyboard=true, buttonType=Quiet, text=Confirm}
   slot {class=HorizontalBoxSlot, properties {padding=<ltrb 0 0 24 0>}}
  }
  WBP_UIKit_Button_Cancel {
   widgetClassPath=&WBP_UIKit_Button_Regular
   bIsVariable=true
   parent=HB_SupportACreator_3
   widget=WBP_UIKit_Button_Cancel
   properties {
    bDisplayInputActionWhenNotInteractable=false
    bHideInputAction=true
    bHideInputActionWithKeyboard=true
    buttonType=Quiet
    text=Cancel
    triggeringInputAction {dataTable=&InputActionDataTable, rowName=GenericBack}
   }
   slot {class=HorizontalBoxSlot, properties {padding=<ltrb 0 0 24 0>}}
  }
  SB_RemovedCode {
   widgetClassPath=&SizeBox
   bIsVariable=true
   parent=VisibilitySwitcher
   widget=SB_RemovedCode
   properties {bOverride_HeightOverride=true, bOverride_WidthOverride=true, heightOverride=460, maxDesiredWidth=820, widthOverride=820}
   slot {class=WidgetSwitcherSlot, properties {padding=<ltrb 80 80 80 48>}}
  }
  VB_RemovedCode {widgetClassPath=&VerticalBox, parent=SB_RemovedCode, widget=VB_RemovedCode, slot {class=SizeBoxSlot}}
  Text_ModalTitle_2 {
   widgetClassPath=&CommonTextBlock
   parent=VB_RemovedCode
   widget=Text_ModalTitle_2
   properties {
    autoWrapText=true
    text=Removed Creator Code
    textTransformPolicy=ToUpper
    font {
     bForceMonospaced=false
     bMaterialIsStencil=false
     fontMaterial=null
     fontObject=&HeadingNow
     letterSpacing=0
     monospacedWidth=1
     size=25.5
     skewAmount=0
     typefaceFontName=86 Bold Italic
     outlineSettings {
      bApplyOutlineToDropShadows=false
      bMiteredCorners=false
      bSeparateFillAlpha=false
      outlineMaterial=null
      outlineSize=0
      outlineColor=<rgba 0 0 0 1>
     }
    }
    shadowOffset=<0 0>
   }
   slot {class=VerticalBoxSlot, properties {verticalAlignment=VAlign_Top, padding=<ltrb 0 0 0 48>}}
  }
  Text_Desc_2 {
   widgetClassPath=&CommonTextBlock
   parent=VB_RemovedCode
   widget=Text_Desc_2
   properties {
    autoWrapText=true
    text="Creator Code has been removed. \r\n\r\nIf you change your mind, you can re-apply the code from this page at any time!"
    colorAndOpacity {colorUseRule=UseColor_Specified, specifiedColor=<rgba 0.5776 0.5776 0.5776 1>}
    font {
     bForceMonospaced=false
     bMaterialIsStencil=false
     fontMaterial=null
     fontObject=&HeadingNow
     letterSpacing=0
     monospacedWidth=1
     size=18
     skewAmount=0
     typefaceFontName=74 Regular
     outlineSettings {
      bApplyOutlineToDropShadows=false
      bMiteredCorners=false
      bSeparateFillAlpha=false
      outlineMaterial=null
      outlineSize=0
      outlineColor=<rgba 0 0 0 1>
     }
    }
    shadowOffset=<0 0>
   }
   slot {class=VerticalBoxSlot, properties {verticalAlignment=VAlign_Top}}
  }
  Username_2 {
   widgetClassPath=&Overlay
   parent=VB_RemovedCode
   widget=Username_2
   slot {class=VerticalBoxSlot, properties {padding=<ltrb 0 32 0 24>}}
  }
  UIKit_Backplate_2 {
   widgetClassPath=&WBP_UIKit_Backplate
   parent=Username_2
   widget=UIKit_Backplate_2
   properties {brightness=Medium, cornerRadius=Interactive Component, colorAndOpacity=<rgba 1 1 1 0.5>}
   slot {class=OverlaySlot, properties {horizontalAlignment=HAlign_Fill, verticalAlignment=VAlign_Fill}}
  }
  HorizontalBox_1 {
   widgetClassPath=&HorizontalBox
   parent=Username_2
   widget=HorizontalBox_1
   slot {class=OverlaySlot, properties {horizontalAlignment=HAlign_Center, verticalAlignment=VAlign_Fill, padding=<ltrb 24 24 24 24>}}
  }
  Text_Username_2 {
   widgetClassPath=&CommonTextBlock
   bIsVariable=true
   parent=HorizontalBox_1
   widget=Text_Username_2
   properties {
    justification=Center
    scrollStyle=&TextScrollStyle-Base_C
    text=Creator Name
    textTransformPolicy=ToUpper
    font {
     bForceMonospaced=false
     bMaterialIsStencil=false
     fontMaterial=null
     fontObject=&HeadingNow
     letterSpacing=0
     monospacedWidth=1
     size=25.5
     skewAmount=0
     typefaceFontName=86 Bold Italic
     outlineSettings {
      bApplyOutlineToDropShadows=false
      bMiteredCorners=false
      bSeparateFillAlpha=false
      outlineMaterial=null
      outlineSize=0
      outlineColor=<rgba 0 0 0 1>
     }
    }
    shadowOffset=<0 0>
   }
   slot {class=HorizontalBoxSlot, properties {horizontalAlignment=HAlign_Left, verticalAlignment=VAlign_Top, size {sizeRule=Fill, value=1}}}
  }
  HB_SupportACreator_2 {
   widgetClassPath=&HorizontalBox
   parent=VB_RemovedCode
   widget=HB_SupportACreator_2
   slot {class=VerticalBoxSlot, properties {horizontalAlignment=HAlign_Left, verticalAlignment=VAlign_Bottom, size {sizeRule=Fill, value=1}}}
  }
  WBP_UIKit_Button_Close {
   widgetClassPath=&WBP_UIKit_Button_Regular
   bIsVariable=true
   parent=HB_SupportACreator_2
   widget=WBP_UIKit_Button_Close
   properties {
    bHideInputAction=true
    bHideInputActionWithKeyboard=true
    buttonType=Quiet
    navigation=&self:WidgetTree.WBP_UIKit_Button_Close.WidgetNavigation_1
    text=Close
    triggeringInputAction {dataTable=&InputActionDataTable, rowName=GenericBack}
   }
   slot {class=HorizontalBoxSlot}
  }
  VisiblityBorder_Touch {
   widgetClassPath=&CommonHardwareVisibilityBorder
   namedSlotHost=Dialog_Internal
   widget=VisiblityBorder_Touch
   properties {
    horizontalAlignment=HAlign_Right
    verticalAlignment=VAlign_Top
    padding=<ltrb 0 32 32 0>
    visibilityQuery {
     autoDescription=" ANY( Input.Touch )"
     tokenStreamVersion=0
     userDescription=""
     queryTokenStream [0,1,1,1,0]
     tagDictionary [{tagName=Input.Touch}]
    }
   }
  }
  Button_BackTouch {
   widgetClassPath=/FortUILibrary/Foundation/Buttons/CloseButton
   bIsVariable=true
   parent=VisiblityBorder_Touch
   widget=Button_BackTouch
   properties {
    onButtonBaseClicked=(&self:WidgetTree.Button_BackTouch.HandleOnButtonClicked)
    onButtonBaseDoubleClicked=(&self:WidgetTree.Button_BackTouch.HandleOnButtonDoubleClicked)
    onButtonBaseHovered=(&self:WidgetTree.Button_BackTouch.HandleOnButtonHovered)
    onButtonBaseUnhovered=(&self:WidgetTree.Button_BackTouch.HandleOnButtonUnhovered)
    onSelectedChangedBase=(&self:WidgetTree.Button_BackTouch.HandleOnSelectedChanged)
    renderTransform {angle=0, scale=<1 1>, shear=<0 0>, translation=<0 10>}
   }
   slot {class=BorderSlot, properties {horizontalAlignment=HAlign_Right, verticalAlignment=VAlign_Top, padding=<ltrb 0 32 32 0>}}
  }
 ]
}
