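# Creator profile page in the Discovery browser, showing a creator's header, avatar, social links and content list, with follow, notification, Support-A-Creator and report actions.
# FortPlayerPageVM supplies creator state and receives actions, while modals handle notifications, support and follow-limit warnings.
# List navigation drives the collapsing mini header, Back To Top button and animated footer.
# LastFocusedCreatorWidget and custom navigation rules route focus between info header, list and footer.
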WidgetBlueprint /DiscoveryBrowser/CreatorProfile/WBP_FollowCreator_ProfilePage
animations [FooterTransition {
  binding_count=2
  length_seconds=0.15
  object=&self:FooterTransition
  sequence {
   bindings [
    Footer_Spacer {
     id=#1
     possessed_class=Spacer
     tracks [{
       type=MovieSceneDoubleVectorTrack
       display_name=Size
       sections [{
         type=MovieSceneDoubleVectorSection
         end_frame=9
         start_frame=0
         channels [
          Vector.X {type=MovieSceneScriptingDoubleChannel, const [1760,2]}
          Vector.Y {type=MovieSceneScriptingDoubleChannel, keys [[0,300],[9,0]]}
         ]
        }]
      }]
    }
    Overlay_Footer {
     id=#2
     possessed_class=Overlay
     tracks [{
       type=MovieSceneByteTrack
       display_name=Visibility
       sections [{type=MovieSceneByteSection, end_frame=9, start_frame=0, channels [None {type=MovieSceneScriptingByteChannel, keys [[0,4],[9,1]]}]}]
      }]
    }
   ]
   display_rate {numerator=60}
  }
 }]
graphs [
 BP_GetDesiredFocusTarget {nodes [
   Get Desired Focus Target()
   > Branch(Condition=Get HasListLoaded() -> (HasListLoaded:bool))
   n1 = Get Overlay_Footer() -> (Overlay_Footer:Overlay Object Reference)
   n2 = GetCurrentInputType(self=Get CommonInputSubsystem())
   > n3 = Branch(Condition=AND Boolean(A="Equal (Enum)"(A=@n2, B=MouseAndKeyboard), B=IsVisible(self=@n1)))
   n4 = Is Valid(exec=@n3.else, InputObject=Get LastFocusedCreatorWidget() -> (LastFocusedCreatorWidget:Widget Object Reference))
   n5 = Get ListView!(self:Fort Polymorphic List View Object Reference=Get WBP_Poly_MegaList() -> (WBP_Poly_MegaList:WBP Poly Mega List Object Reference)) -> (ListView:Common List View Object Reference)
   n6 = Branch(execute=@n4.Is Not Valid, Condition="integer > integer"(A=GetNumItems!(self:List View Object Reference=@n5) -> (ReturnValue:int=0)))
   > n7 = Branch(Condition=Get bAmIFollowing!(self:Creator View Model Object Reference=Get FortPlayerPageVM() -> (FortPlayerPageVM:Creator Page View Model Object Reference)) -> (bAmIFollowing:bool))
   n8 = Get ListView(self=Get WBP_Poly_MegaList())
   > SetSelectedIndex!(execute, self:List View Object Reference=@n8, Index:int=0) -> (then)
   > Navigate To Item!(
    execute
    self:List View Object Reference=@n8
    Item:Object Reference=Get Selected Item!(self:List View Object Reference=@n8) -> (ReturnValue:Object Reference) {is=title K2Node_CallFunction}
   ) -> (then)
   > n9 = Set with Broadcast LastFocusedCreatorWidget!(execute, LastFocusedCreatorWidget:Widget Object Reference=Get ListView(self=Get WBP_Poly_MegaList())) -> (then, Output_Get:Widget Object Reference)
   Return Node!(execute=[@n12.then, @n9.then, @n16.then, @n14.else], ReturnValue:Widget Object Reference=Get LastFocusedCreatorWidget())
   n10 = Get WBP_CreatorPage_Infos() -> (WBP_CreatorPage_Infos:WBP Creator Page Infos Object Reference)
   n11 = Set LastFocusedWidget!(
    execute=@n6.else
    LastFocusedWidget:User Widget Object Reference=Get Button_Follow!(self:WBP Creator Page Infos Object Reference=@n10) -> (Button_Follow:WBP Creator Page Button Main Object Reference)
    self:WBP Creator Page Infos Object Reference=@n10
   ) -> (then, Output_Get:User Widget Object Reference)
   n12 = Set with Broadcast LastFocusedCreatorWidget(execute=[@n7.else, @n11.then], LastFocusedCreatorWidget=@n11)
   n13 = Get WBP_Poly_MegaList()
   n14 = Branch(execute=@n4.Is Valid, Condition="Equal (Object)"(A=Get LastFocusedCreatorWidget(), B=Get ListView(self=@n13)))
   n15 = Get ListView(self=Get WBP_Poly_MegaList())
   > n16 = Navigate To Item(self=@n15, Item=Get Selected Item(self=@n15))
  ]}
 CanBeReported {nodes [
   n1 = CanBeReported!() -> (then, CanBeReported:bool=~) {is=title K2Node_FunctionEntry}
   > SetVisibility(
    self=Get WBP_CreatorPage_Footer_Report() -> (WBP_CreatorPage_Footer_Report:WBP UIKit Button Generic Object Reference)
    InVisibility=Select(Option 0=Collapsed, Option 1=Visible, Index=@n1)
   )
  ]}
 EventGraph {nodes [
   "On Clicked (WBP_CreatorPage_Footer_Support)"!() -> (OutputDelegate, then, Button:Common Button Base Object Reference) {is=title K2Node_ComponentBoundEvent}
   > n1 = Set with Broadcast LastFocusedCreatorWidget(LastFocusedCreatorWidget=Get WBP_CreatorPage_Footer_Support() -> (WBP_CreatorPage_Footer_Support:WBP UIKit Button Generic Object Reference))
   n2 = Create Widget!(
    execute:Exec=[@n5.then, @n1.then]
    Class:User Widget Class Reference=&/WBP_CreatorPage_SupportACreator_Modal.WBP_CreatorPage_SupportACreator_Modal_C
    OwningPlayer:Player Controller Object Reference=GetOwningPlayer()
    FortSupportCreatorVM:Support A Creator View Model Object Reference=Get FortSupportCreatorVM() -> (FortSupportCreatorVM:Support A Creator View Model Object Reference)
   ) -> (then:Exec, ReturnValue:WBP Creator Page Support ACreator Modal Object Reference)
   > n3 = TryPushWidgetAsModalPanel!(execute, PushWidget:User Widget Object Reference=@n2) -> (then)
   n4 = SupportCreatorModalDeactivated() {is=title K2Node_CustomEvent}
   Bind Event to On Widget Deactivated(execute=@n3.then, self:Common Activatable Widget Object Reference=@n2, Delegate=@n4) {is=title K2Node_AddDelegate}
   > Setup Creator Support Modal!(
    execute
    self:WBP Creator Page Support ACreator Modal Object Reference=@n2
    Current Creator Name:Text=Get Name!(self:Creator View Model Object Reference=Get FortPlayerPageVM()) -> (Name:Text)
   ) -> (then)
   "On Clicked (WBP_CreatorPage_Footer_Report)"!() -> (OutputDelegate, then, Button:Common Button Base Object Reference) {is=title K2Node_ComponentBoundEvent}
   > Call Handle Creator Reported!(execute:Exec, self:WBP Creator Page Infos Object Reference=Get WBP_CreatorPage_Infos()) -> (then:Exec) {is=title K2Node_CallDelegate}
   > Set with Broadcast LastFocusedCreatorWidget(LastFocusedCreatorWidget=Get WBP_CreatorPage_Footer_Report())
   OnCreatorReported() {is=title K2Node_CustomEvent}
   > ReportCreator!(execute, self:Creator View Model Object Reference=Get FortPlayerPageVM(), ContextWidget:User Widget Object Reference=Self-Reference()) -> (then)
   n5 = SupportCreatorModal!() -> (OutputDelegate, then, Button:Common Button Base Object Reference) {is=title K2Node_CustomEvent}
   n6 = OnFooterAnimationFinished() {is=title K2Node_CustomEvent}
   > Branch(Condition=Get isFooterEnabled() -> (isFooterEnabled:bool))
   > UnbindAllFromAnimationFinished!(
    execute
    self:User Widget Object Reference
    Animation:Widget Animation Object Reference=Get FooterTransition() -> (FooterTransition:Widget Animation Object Reference)
   ) -> (then)
   > ScrollToBottom(self:List View Base Object Reference=Get ListView(self=Get WBP_Poly_MegaList()))
   OnBackToTopClicked!() -> (OutputDelegate, then, Button:Common Button Base Object Reference) {is=title K2Node_CustomEvent}
   n7 = Get ListView(self=Get WBP_Poly_MegaList())
   n8 = GetItemAt!(self:List View Object Reference=@n7, Index:int=0) -> (ReturnValue:Object Reference)
   > Set Selected Item!(execute, self:List View Object Reference=@n7, Item:Object Reference=@n8) -> (then) {is=title K2Node_CallFunction}
   > Navigate To Item(self=@n7, Item=@n8)
   > TransitionToMiniHeader!(execute, self:WBP Discover Surface Base Object Reference, IsMiniHeaderWithFullList:bool=false) -> (then)
   > Switch Back and Back To Top buttons!(execute, self:WBP Discover Surface Base Object Reference, Back To Top:bool=true) -> (then)
   > OnNavigateToNewOuterRow!(
    execute
    self:WBP Poly Mega List Object Reference=Get WBP_Poly_MegaList()
    IsDirectionUp:bool=true
    PreviousSubrowSelectedIndex:int=0
    PreviousNumItemsPerSubrow:int=0
    PreviousRowWasViewAllTileSelected:bool=false
   ) -> (then)
   n9 = UpdateInitialFocus!() -> (OutputDelegate, then, ActiveWidget:Widget Object Reference, ActiveIndex:int) {is=title K2Node_CustomEvent}
   n10 = GetCurrentInputType(self=Get CommonInputSubsystem())
   > n11 = Branch(Condition=AND Boolean(A="Equal (Enum)"(A=@n10, B=Gamepad), B="Equal (Integer)"(A=@n9.ActiveIndex)))
   n12 = Get FortPlayerPageVM()
   > n13 = Branch(Condition=OR Boolean(A=Get bIsSelf!(self:Creator View Model Object Reference=@n12) -> (bIsSelf:bool), B=Get bAmIFollowing(self=@n12)))
   > Delay(Duration=0.1)
   n14 = Get WBP_Poly_MegaList()
   > OnNavigateToNewOuterRow(self=@n14, IsDirectionUp=true, PreviousSubrowSelectedIndex=0, PreviousNumItemsPerSubrow=0, PreviousRowWasViewAllTileSelected=false)
   > Set with Broadcast LastFocusedCreatorWidget(LastFocusedCreatorWidget=Get ListView(self=@n14))
   NavigateUpFromMegalist(execute=@n13.else)
   Reroute Node(InputPin=@n11.else)
   OnSubscribeButtonClicked!() -> (OutputDelegate, then, Button:Common Button Base Object Reference) {is=title K2Node_CustomEvent}
   > n15 = Construct Object from Class(Class=/DiscoveryBrowser/CreatorProfile/Dialogs/DialogVM_CreatoPage_Subscribe.DialogVM_CreatoPage_Subscribe_C) -> (ReturnValue:Dialog VM Creato Page Subscribe Object Reference)
   > n16 = Create Widget!(
    execute
    Class=&/WBP_CreatorPage_ManageNotification_Modal.WBP_CreatorPage_ManageNotification_Modal_C
    OwningPlayer
    UIKitDialogViewModel:UI Kit Dialog Viewmodel Object Reference=@n15
   ) -> (then, ReturnValue:WBP Creator Page Manage Notification Modal Object Reference)
   > SetCreatorName!(execute, self:WBP Creator Page Manage Notification Modal Object Reference=@n16, CreatorName:Text=Get Name(self=Get FortPlayerPageVM())) -> (then)
   n17 = Get SwitcherButton_NewReleases!(self:WBP Creator Page Manage Notification Modal Object Reference=@n16) -> (SwitcherButton_NewReleases:WBP UIKit Switcher Button Object Reference)
   > Set Selected!(
    execute
    Selected:bool=Get bAmISubscribed!(self:Creator View Model Object Reference=Get FortPlayerPageVM()) -> (bAmISubscribed:bool)
    self:WBP UIKit Switcher Button Object Reference=@n17
   ) -> (then, Output_Get:bool)
   > n18 = Bind Event to On Clicked(self:Common Button Base Object Reference=@n17, Delegate=Create Event()) {is=title K2Node_AddDelegate}
   n19 = OnUpdateNotification!() -> (OutputDelegate, then, NotificationArray:Array of ECreatorNotificationTypeVM Enums=Invalid) {is=title K2Node_CustomEvent}
   n20 = Assign Update Notification!(execute:Exec=@n18.then, self:WBP Creator Page Manage Notification Modal Object Reference=@n16, Delegate:"Delegate (by ref)"=@n19) -> (then:Exec) {is=title K2Node_AssignDelegate}
   n21 = OnManageNotificationModalDeactivated() {is=title K2Node_CustomEvent}
   Assign On Widget Deactivated!(execute:Exec=@n20.then, self:Common Activatable Widget Object Reference=@n16, Delegate:"Delegate (by ref)"=@n21) -> (then:Exec) {is=title K2Node_AssignDelegate}
   > TryPushWidgetAsModalPanel(PushWidget=@n16)
   OnNavigateDownOutOfMegalist() {is=title K2Node_CustomEvent}
   > Branch(Condition=Get Has Support A Creator Enabled() -> (Has Support A Creator Enabled:bool))
   > n22 = Branch(Condition=Get isFooterEnabled())
   Set with Broadcast LastFocusedCreatorWidget(execute=[@n23.then, @n22.then], LastFocusedCreatorWidget=Get WBP_CreatorPage_Footer_Support())
   > RequestRefreshFocus(self:Common Activatable Widget Object Reference)
   n23 = EnableFooter!(execute=@n22.else, self:WBP Discover Surface Base Object Reference, Enable:bool=true) -> (then)
   n24 = OnMegalistNavigated!() -> (OutputDelegate, then, LastNavigationDirectionWasUp:bool=~, CurrentRowIndex:int, CurrentSubRowIndex:int) {is=title K2Node_CustomEvent}
   > SetHeaderStateWhenNavigating!(
    execute
    self
    LastNavigationDirectionWasUp:bool=@n24.LastNavigationDirectionWasUp
    CurrentRowIndex:int=@n24.CurrentRowIndex
    CurrentSubRowIndex:int=@n24.CurrentSubRowIndex
   ) -> (then)
   OnNewReleasesSwitcherClicked!() -> (OutputDelegate, then, Button:Common Button Base Object Reference) {is=title K2Node_CustomEvent}
   > ToggleAmISubscribed(self:Creator View Model Object Reference=Get FortPlayerPageVM())
   n25 = OnModalDismissed!() -> (OutputDelegate, then, ButtonIndex:int) {is=title K2Node_CustomEvent}
   n26 = Construct Object from Class(execute=@n34.then, Class=/DiscoveryBrowser/CreatorProfile/Dialogs/DialogVM_FollowLimitReached.DialogVM_FollowLimitReached_C) -> (ReturnValue:Dialog VM Follow Limit Reached Object Reference)
   > n27 = Create Widget(Class=/FortUILibrary/UIKit/Dialogs/WBP_UIKit_Dialog.WBP_UIKit_Dialog_C, UIKitDialogViewModel=@n26) -> (ReturnValue:WBP UIKit Dialog Object Reference)
   n28 = Get Dialog_Internal!(self:WBP UIKit Dialog Object Reference=@n27) -> (Dialog_Internal:WBP UIKit Dialog Internal Object Reference)
   Unbind all Events from Dialog Button Clicked Dispatcher!(execute:Exec=@n25.then, self:WBP UIKit Dialog Internal Object Reference=@n28) -> (then:Exec) {is=title K2Node_ClearDelegate}
   > DeactivateWidget(self=@n27)
   n29 = OnFollowButtonClicked!() -> (OutputDelegate, then, Button:Common Button Base Object Reference) {is=title K2Node_CustomEvent}
   n30 = Get bAmIFollowing(self=Get FortPlayerPageVM())
   > Fire Event Element Interacted(
    Element=@n29.Button
    AdditionalInfo="Add String (Non-Localized)"!(
     Info:"Flow Analytics Additional Info Structure (by ref)"=GetCreatorIDForTelemetry!() -> (ReturnValue:Flow Analytics Additional Info Structure)
     Key:String=Attempted Action
     Value:String=SelectString(A=Unfavorite, B=Favorite, bPickA=@n30)
    ) -> (ReturnValue:Flow Analytics Additional Info Structure)
   )
   > n31 = Sequence()
   n32 = Get FortPlayerPageVM()
   > Branch(Condition=AND Boolean(A=Get bAmIFollowing(self=@n32), B=Get bAmISubscribed(self=@n32)))
   > ToggleAmISubscribed(self=Get FortPlayerPageVM())
   > Clear(TargetArray:Array of ECreatorNotificationTypeVM Enums=Get Notification Array() -> (Notification Array:Array of ECreatorNotificationTypeVM Enums=Invalid))
   > UpdateSubscriptionType!(
    execute
    self:Creator View Model Object Reference=Get FortPlayerPageVM()
    CreatorNotificationTypes:Array of ECreatorNotificationTypeVM Enums=Get Notification Array() -> (Notification Array=Invalid)
   ) -> (then)
   ToggleAmIFollowing(execute=@n31.then_1, self:Creator View Model Object Reference=Get FortPlayerPageVM())
   > PlaySound2D(Sound=Select(
     Option 0:Sound Base Object Reference=Get Unfavorite_Sound() -> (Unfavorite_Sound:Sound Base Object Reference)
     Option 1:Sound Base Object Reference=Get Favorite_Sound() -> (Favorite_Sound:Sound Base Object Reference)
     Index=@n30
    ) -> (ReturnValue:Sound Base Object Reference))
   > TryShowFollowingCreatorTooltip()
   Handle Back Button Clicked() {is=title K2Node_CustomEvent}
   > SetPlayerPageBackButtonClicked!(
    execute
    self:Fort Activity Selector Object Reference=Get FortActivitySelector() -> (ReturnValue:Fort Activity Selector Object Reference) {is=title K2Node_GetSubsystem}
    bInIsButtonClicked:bool=true
   ) -> (then)
   OnNavigateUpToFollowButton() {is=title K2Node_CustomEvent}
   > NavigateUpFromMegalist()
   n33 = OnFocusInsideCreatorInfosChanged!() -> (OutputDelegate, then, LastFocusedWidget:User Widget Object Reference) {is=title K2Node_CustomEvent}
   > Set with Broadcast LastFocusedCreatorWidget(LastFocusedCreatorWidget=@n33.LastFocusedWidget)
   n34 = OnMaxFollowersReached() {is=title K2Node_CustomEvent}
   OnShowActivityDetailsOnCreatorPage() {is=title K2Node_CustomEvent}
   > Branch(Condition=Get bAmIFollowing(self=Get FortPlayerPageVM()))
   > UpdateInteraction(self:Creator View Model Object Reference=Get FortPlayerPageVM())
   > Set with Broadcast LastFocusedCreatorWidget(LastFocusedCreatorWidget=Get ListView(self=Get WBP_Poly_MegaList()))
   Event Construct()
   > Parent: Construct()
   n35 = Get WBP_CreatorPage_Infos()
   > Bind Event to Handle Creator Reported(self:WBP Creator Page Infos Object Reference=@n35, Delegate=Create Event()) {is=title K2Node_AddDelegate}
   > Bind Event to Handle Follow Clicked(self:WBP Creator Page Infos Object Reference=@n35, Delegate=Create Event()) {is=title K2Node_AddDelegate}
   n36 = Get FortPlayerPageVM()
   > Bind Event to On Reached Max Number Of Followed Creators(self:Creator View Model Object Reference=@n36, Delegate=Create Event()) {is=title K2Node_AddDelegate}
   n37 = Create Event()
   n38 = Get WBP_Poly_MegaList()
   > Bind Event to Navigate Up Out Of Mega List(self:WBP Poly Mega List Object Reference=@n38, Delegate=@n37) {is=title K2Node_AddDelegate}
   n39 = Get WBP_CreatorPage_Infos()
   > Bind Event to Last Focused Updated from Creator Infos(self:WBP Creator Page Infos Object Reference=@n39, Delegate=@n33) {is=title K2Node_AddDelegate}
   n40 = Get FortActivitySelector()
   > Bind Event to On Show Activity Details on Creator Page Delegate(self:Fort Activity Selector Object Reference=@n40, Delegate=Create Event()) {is=title K2Node_AddDelegate}
   > Bind Event to On Back Button Clicked(self:WBP Discover Surface Base Object Reference, Delegate=Create Event()) {is=title K2Node_AddDelegate}
   n41 = Create Event()
   n42 = Get WBP_BottomBarHack_BackToTop() -> (WBP_BottomBarHack_BackToTop:WBP Bottom Bar Decoy Button Object Reference)
   > Bind Event to On Clicked(self=@n42, Delegate=@n41)
   > Set with Broadcast LastFocusedCreatorWidget()
   n43 = Get WBP_BottomBarHack_SupportCreator() -> (WBP_BottomBarHack_SupportCreator:WBP Bottom Bar Decoy Button Object Reference)
   > Bind Event to On Clicked(self=@n43, Delegate=Create Event())
   n44 = Create Event()
   n45 = Get Switcher_List() -> (Switcher_List:Common Activatable Widget Switcher Object Reference)
   > Bind Event to On Active Widget Index Changed BP(self:Common Animated Switcher Object Reference=@n45, Delegate=@n44) {is=title K2Node_AddDelegate}
   > SetNavigationRuleCustomBoundary!(
    execute
    self:Widget Object Reference=Get HeaderOverlay() -> (HeaderOverlay:Overlay Object Reference)
    Direction:EUINavigation Enum=Down
    InCustomDelegate:Delegate=Create Event()
   ) -> (then)
   n46 = Get WBP_Poly_MegaList()
   > Bind Event to Navigate Down Out Of Mega List(self:WBP Poly Mega List Object Reference=@n46, Delegate=Create Event()) {is=title K2Node_AddDelegate}
   n47 = Create Event()
   n48 = Get WBP_Poly_MegaList()
   > Bind Event to On Navigated Mega List Row(self:WBP Poly Mega List Object Reference=@n48, Delegate=@n47) {is=title K2Node_AddDelegate}
   n49 = Create Event()
   n50 = Get Button_Subscribe!(self:WBP Creator Page Infos Object Reference=Get WBP_CreatorPage_Infos()) -> (Button_Subscribe:WBP UIKit Button Regular Object Reference)
   > Bind Event to On Clicked(self=@n50, Delegate=@n49)
   > Set with Broadcast bIsAutoSubscriptionDisabled!(
    execute
    bIsAutoSubscriptionDisabled:bool=IsAutomaticSubscriptionOnFollowingDisabled!(self:Creator View Model Object Reference=Get FortPlayerPageVM()) -> (ReturnValue:bool)
   ) -> (then, Output_Get:bool)
   n51 = Event PreConstruct()
   > Parent: PreConstruct(IsDesignTime=@n51.IsDesignTime)
   > SetCompactUIStyle(self:WBP Discover Surface Base Object Reference) {is=title K2Node_CallFunction}
   > SetupFooterButtonsVisuals()
   Event On Activated()
   > Parent: On Activated() {is=title K2Node_CallParentFunction}
   > Fire Event Main View Opened!(execute, View:Widget Object Reference=Self-Reference(), AdditionalInfo:Flow Analytics Additional Info Structure, bIsOverlay:bool=false) -> (then)
   > StartAnalyticsImpressions(self:Discover Surface View Model Object Reference=Get SurfaceVM!(self:Creator Page View Model Object Reference=Get FortPlayerPageVM()) -> (SurfaceVM:Discover Surface View Model Object Reference))
   > Set Size!(
    execute
    Size:Vector 2D Structure=(X=1,Y=904)
    self:Spacer Object Reference=Get MinHeaderSize_Spacer() -> (MinHeaderSize_Spacer:Spacer Object Reference)
   ) -> (then, Output_Get:Vector 2D Structure)
   Event Destruct()
   > Parent: Destruct()
   > Unbind all Events from Handle Creator Reported!(execute:Exec, self:WBP Creator Page Infos Object Reference=@n35) -> (then:Exec) {is=title K2Node_ClearDelegate}
   > Unbind all Events from Handle Follow Clicked!(execute:Exec, self:WBP Creator Page Infos Object Reference=@n35) -> (then:Exec) {is=title K2Node_ClearDelegate}
   > Unbind all Events from On Reached Max Number Of Followed Creators!(execute:Exec, self:Creator View Model Object Reference=@n36) -> (then:Exec) {is=title K2Node_ClearDelegate}
   > Unbind Event from Navigate Up Out Of Mega List(self:WBP Poly Mega List Object Reference=@n38, Delegate=@n37) {is=title K2Node_RemoveDelegate}
   > Unbind all Events from Last Focused Updated from Creator Infos!(execute:Exec, self:WBP Creator Page Infos Object Reference=@n39) -> (then:Exec) {is=title K2Node_ClearDelegate}
   > Unbind all Events from On Show Activity Details on Creator Page Delegate!(execute:Exec, self:Fort Activity Selector Object Reference=@n40) -> (then:Exec) {is=title K2Node_ClearDelegate}
   > Unbind all Events from On Back Button Clicked!(execute:Exec, self:WBP Discover Surface Base Object Reference) -> (then:Exec) {is=title K2Node_ClearDelegate}
   > Unbind Event from On Clicked(self:Common Button Base Object Reference=@n42, Delegate=@n41) {is=title K2Node_RemoveDelegate}
   > Unbind all Events from On Clicked!(execute:Exec, self:Common Button Base Object Reference=@n43) -> (then:Exec) {is=title K2Node_ClearDelegate}
   > Unbind Event from On Active Widget Index Changed BP(self:Common Animated Switcher Object Reference=@n45, Delegate=@n44) {is=title K2Node_RemoveDelegate}
   > Unbind all Events from Navigate Down Out Of Mega List!(execute:Exec, self:WBP Poly Mega List Object Reference=@n46) -> (then:Exec) {is=title K2Node_ClearDelegate}
   > Unbind Event from On Navigated Mega List Row(self:WBP Poly Mega List Object Reference=@n48, Delegate=@n47) {is=title K2Node_RemoveDelegate}
   > Unbind Event from On Clicked(self=@n50, Delegate=@n49)
   Event On Deactivated()
   > Parent: On Deactivated() {is=title K2Node_CallParentFunction}
   > FireExitedImpressionsFromSurface(self:Discover Surface View Model Object Reference=Get SurfaceVM(self=Get FortPlayerPageVM()))
   > SetVisibility(
    self=[
     Get Overlay_Footer()
     Get Txt_Creator_Support!(self:WBP Creator Page Infos Object Reference=Get WBP_CreatorPage_Infos()) -> (Txt_Creator_Support:Common Text Object Reference)
    ]
    InVisibility=Collapsed
   )
   > Set isFooterEnabled(isFooterEnabled:bool=false) -> (Output_Get:bool)
   > Branch(Condition=Get isSupportingCurrentCreator() -> (isSupportingCurrentCreator:bool))
   > SetTriggeringInputAction!(
    execute
    self:Common Button Base Object Reference=Get WBP_BottomBarHack_SupportCreator()
    InputActionRow_DataTable:Data Table Object Reference
    InputActionRow_RowName:Name=None
   ) -> (then)
   n52 = Event EnableFooter!() -> (OutputDelegate, then, Enable:bool) {is=title K2Node_Event}
   > Parent: EnableFooter!(execute, Enable:bool=@n52.Enable) -> (then) {is=title K2Node_CallParentFunction}
   > Branch(Condition=Get Has Support A Creator Enabled())
   > SetVisibility(
    self=Get Image_ListSoftBottomEdge() -> (Image_ListSoftBottomEdge:Image Object Reference)
    InVisibility=Select(Option 0:ESlateVisibility Enum=SelfHitTestInvisible, Option 1:ESlateVisibility Enum=Collapsed, Index=Get isFooterEnabled()) -> (ReturnValue:ESlateVisibility Enum)
   )
   > PlayAnimation(
    InAnimation=Get FooterTransition()
    PlayMode=Select(Option 0:EUMGSequencePlayMode Enum, Option 1:EUMGSequencePlayMode Enum=Reverse, Index=Get isFooterEnabled()) -> (ReturnValue:EUMGSequencePlayMode Enum)
   )
   > BindToAnimationFinished(Animation=Get FooterTransition(), Delegate=@n6)
   RequestRefreshFocus(execute=@n60.then)
   n53 = Get ListView(self=Get WBP_Poly_MegaList())
   Navigate To Item(execute=@n59.else, self=@n53, Item=Get Selected Item(self=@n53))
   TryPushWidgetAsModalPanel(execute=@n55.then, PushWidget=@n27)
   n54 = Set Notification Array(execute=@n19.then, Notification Array:Array of ECreatorNotificationTypeVM Enums=@n19.NotificationArray) -> (Output_Get:Array of ECreatorNotificationTypeVM Enums=Invalid)
   UpdateSubscriptionType(execute=@n63.then, self=Get FortPlayerPageVM(), CreatorNotificationTypes=@n54)
   SetBackgroundMaterialColor!(execute, self:WBP Discover Surface Base Object Reference, Background Color:Linear Color Structure=(R=0.0704,G=0.0152,B=0.3419,A=1)) -> (then)
   RequestRefreshFocus(execute=@n57.else)
   n55 = Bind Event to Dialog Button Clicked Dispatcher(execute=@n27.then, self:WBP UIKit Dialog Internal Object Reference=@n28, Delegate=@n25) {is=title K2Node_AddDelegate}
   UpdateSupportCreatorVisuals!(execute=@n61.then, self, Has Support A Creator Enabled:bool=Get Has Support A Creator Enabled()) -> (then)
   n56 = Get WBP_Poly_MegaList()
   > n57 = Branch(Condition="Equal (Object)"(A=Get ListView(self=@n56), B=Get LastFocusedCreatorWidget()))
   n58 = Get CommonInputSubsystem()
   > n59 = Branch(Condition="Equal (Enum)"(A=GetCurrentInputType(self=@n58), B=MouseAndKeyboard))
   Unbind all Events from Update Notification!(execute:Exec=@n64.then, self:WBP Creator Page Manage Notification Modal Object Reference=@n16) -> (then:Exec) {is=title K2Node_ClearDelegate}
   > n60 = Unbind Event from On Widget Deactivated(self:Common Activatable Widget Object Reference=@n16, Delegate=@n21) {is=title K2Node_RemoveDelegate}
   n61 = Unbind all Events from On Widget Deactivated!(execute:Exec=@n4.then, self:Common Activatable Widget Object Reference=@n2) -> (then:Exec) {is=title K2Node_ClearDelegate}
   n62 = Create Event()
   n63 = Branch(execute=@n54.then, Condition=Get bAmISubscribed(self=Get FortPlayerPageVM()))
   n64 = Unbind Event from On Clicked(execute=@n21.then, self=Get SwitcherButton_NewReleases(self=@n16), Delegate=@n62)
  ]}
 GetCreatorIDForTelemetry {nodes [
   GetCreatorIDForTelemetry() {is=title K2Node_FunctionEntry}
   n1 = Get FortPlayerPageVM()
   > Return Node(ReturnValue:Flow Analytics Additional Info Structure="Add String (Non-Localized)"(
     Info=Make Flow Analytics Additional Info!() -> (FlowAnalyticsAdditionalInfo:Flow Analytics Additional Info Structure) {is=title K2Node_MakeStruct}
     Key=Creator ID
     Value=GetIdString!(self:Creator View Model Object Reference=@n1) -> (ReturnValue:String)
    ))
  ]}
 GetPrimaryContentSetup {nodes [
   GetPrimaryContentSetup() {is=title K2Node_FunctionEntry}
   > Return Node(ReturnValue:Primary Content Setup Structure=Make Primary Content Setup!(
     bShowBottomBar:bool=True
     bShowChatWidget:bool=false
     bEnableNewTextChat:bool=False
     bShowTopBar:bool=false
     bTabSupportsInterface:bool=False
    ) -> (PrimaryContentSetup:Primary Content Setup Structure) {is=title K2Node_MakeStruct})
  ]}
 GetShowSubscribeTooltip {nodes [
   n1 = GetShowSubscribeTooltip!() -> (then, IsFollowing:bool=~, IsSubscribed:bool=~, IsActivatedFollowingTooltip:bool=~, IsAutoSubscriptionDisabled:bool=~) {is=title K2Node_FunctionEntry}
   > Return Node!(
    execute
    bShowSubscribeTooltip:bool=AND Boolean!(A=@n1, B=NOT Boolean(A=@n1.IsSubscribed), C:bool=NOT Boolean(A=@n1.IsActivatedFollowingTooltip), D:bool=@n1.IsAutoSubscriptionDisabled) -> (ReturnValue)
   )
  ]}
 LoadAvatarImage {nodes [
   LoadAvatarImage!() -> (then, AvatarURL:String) {is=title K2Node_FunctionEntry}
   > LoadAvatarImage(self:Creator View Model Object Reference=Get FortPlayerPageVM()) {is=title K2Node_CallFunction}
  ]}
 LoadHeaderImage {nodes [
   LoadHeaderImage!() -> (then, Url:String) {is=title K2Node_FunctionEntry}
   > LoadHeaderImage(self:Creator View Model Object Reference=Get FortPlayerPageVM()) {is=title K2Node_CallFunction}
  ]}
 NavigateUpFromMegalist {nodes [
   NavigateUpFromMegalist() {is=title K2Node_FunctionEntry}
   > Set with Broadcast LastFocusedCreatorWidget(LastFocusedCreatorWidget=Get LastFocusedWidget!(self:WBP Creator Page Infos Object Reference=Get WBP_CreatorPage_Infos()) -> (LastFocusedWidget:User Widget Object Reference))
   > RequestRefreshFocus()
  ]}
 NavigateUpToMegalist {nodes [
   NavigateUpToMegalist!() -> (then, Navigation:EUINavigation Enum) {is=title K2Node_FunctionEntry}
   n1 = Get ListView(self=Get WBP_Poly_MegaList())
   > Navigate To Item(self=@n1, Item=Get Selected Item(self=@n1))
   > Set with Broadcast LastFocusedCreatorWidget(LastFocusedCreatorWidget=@n1)
   > RequestRefreshFocus()
   > Return Node!(execute, ReturnValue:Widget Object Reference)
  ]}
 OnNavigateDownFromInfoHeader {nodes [
   OnNavigateDownFromInfoHeader!() -> (then, Navigation:EUINavigation Enum) {is=title K2Node_FunctionEntry}
   n1 = Get ListView(self=Get WBP_Poly_MegaList())
   > n2 = Branch(Condition="integer > integer"(A=GetNumItems(self=@n1) -> (ReturnValue=0)))
   n3 = Get ListView(self=Get WBP_Poly_MegaList())
   > SetSelectedIndex(self=@n3, Index=0)
   > Navigate To Item(self=@n3, Item=Get Selected Item(self=@n3))
   > n4 = Set with Broadcast LastFocusedCreatorWidget(LastFocusedCreatorWidget=@n3)
   RequestRefreshFocus(execute=[@n7.then, @n6.then, @n4.then])
   > Return Node()
   n5 = Branch(execute=@n2.else, Condition=Get isFooterEnabled())
   > n6 = Set with Broadcast LastFocusedCreatorWidget(LastFocusedCreatorWidget=Get WBP_CreatorPage_Footer_Support())
   n7 = Set with Broadcast LastFocusedCreatorWidget(execute=@n5.else, LastFocusedCreatorWidget=Get LastFocusedWidget(self=Get WBP_CreatorPage_Infos()))
  ]}
 SetAvatarImageState {nodes [
   n1 = SetAvatarImageState!() -> (then, LoadingEnum:EActivityImageLoadingState Enum=Unset) {is=title K2Node_FunctionEntry}
   > n2 = Switch on EActivityImageLoadingState!(execute, Selection:EActivityImageLoadingState Enum=@n1) -> (Unset:Exec, Loading:Exec, Loaded:Exec, Error:Exec) {is=title K2Node_SwitchEnum}
   SetCreatorAvatarImageTexture!(
    execute=[@n2.Error, @n2.Unset]
    self:WBP Creator Avatar Image Object Reference=Get WBP_Creator_AvatarImage!(self:WBP Creator Info Non Collapsible Object Reference=Get WBP_CreatorInfo_NonCollapsible() -> (WBP_CreatorInfo_NonCollapsible:WBP Creator Info Non Collapsible Object Reference)) -> (WBP_Creator_AvatarImage:WBP Creator Avatar Image Object Reference)
    CreatorAvatarTexture:Texture Object Reference
   ) -> (then)
   SetCreatorAvatarImageTexture(
    execute=@n2.Loaded
    self=Get WBP_Creator_AvatarImage(self=Get WBP_CreatorInfo_NonCollapsible())
    CreatorAvatarTexture=GetAvatarImage!(self:Creator View Model Object Reference=Get FortPlayerPageVM()) -> (ReturnValue:Texture Object Reference)
   )
   SetLoadingState(execute=@n2.Loading, self:WBP Creator Avatar Image Object Reference=Get WBP_Creator_AvatarImage(self=Get WBP_CreatorInfo_NonCollapsible()))
  ]}
 SetCompactUIStyle {nodes [
   SetCompactUIStyle() {is=title K2Node_FunctionEntry}
   > Parent: SetCompactUIStyle() {is=title K2Node_CallParentFunction}
   > Branch(Condition=IsUsingCompactUI())
   > SetButtonPadding!(
    execute
    self:WBP UIKit Button Generic Base Object Reference=[Get WBP_CreatorPage_Footer_Support(), Get WBP_CreatorPage_Footer_Report()]
    Padding_Left:float=24
    Padding_Top:float=24
    Padding_Right:float=24
    Padding_Bottom:float=24
   ) -> (then)
  ]}
 SetCreatorName {nodes [
   n1 = SetCreatorName!() -> (then, CreatorName:Text) {is=title K2Node_FunctionEntry}
   > "SetText (Text)"(
    self=Get Txt_CreatorName!(self:WBP Creator Info Non Collapsible Object Reference=Get WBP_CreatorInfo_NonCollapsible()) -> (Txt_CreatorName:Common Text Object Reference)
    InText=@n1
   )
   > Set CreatorName!(execute, CreatorName:Text=@n1, self:WBP Creator Page Infos Object Reference=Get WBP_CreatorPage_Infos()) -> (then, Output_Get:Text)
   > Set Message Text Overrides!(
    execute
    self:WBP Discovery No Results Message Object Reference=Get WBP_Discovery_NoResultsMessage() -> (WBP_Discovery_NoResultsMessage:WBP Discovery No Results Message Object Reference)
    InTitle:Text="LOCTEXT(\"Nothing Here Yet\")"
    InMessage:Text=Format Text!(Format="LOCTEXT(\"{CreatorName} hasn\\'t added content to their page.\")", CreatorName:Text=@n1) -> (Result)
   ) -> (then) {is=title K2Node_CallFunction}
   > UpdateSupportCreatorVisuals(Has Support A Creator Enabled=Get Has Support A Creator Enabled())
  ]}
 SetHeaderImageState {nodes [
   n1 = SetHeaderImageState!() -> (then, LoadingEnum:EActivityImageLoadingState Enum=Unset) {is=title K2Node_FunctionEntry}
   > Parent: SetHeaderImageState!(execute, LoadingEnum:EActivityImageLoadingState Enum=@n1) -> (then) {is=title K2Node_CallParentFunction}
   > UpdateSpacerHeightBeforeAnimating!(execute, self:WBP Discover Surface Base Object Reference) -> (then, IsHeaderZeroSize:bool)
   > SetVisibility(
    self=Get HeaderLoadingThrobber() -> (HeaderLoadingThrobber:WBP UIKit Throbber Object Reference)
    InVisibility=Select(
     Option 0:ESlateVisibility Enum=Collapsed
     Option 1:ESlateVisibility Enum=SelfHitTestInvisible
     Index="Equal (Enum)"(A:EActivityImageLoadingState Enum=@n1, B:EActivityImageLoadingState Enum=Loading)
    ) -> (ReturnValue:ESlateVisibility Enum)
   )
   > n2 = GetDynamicMaterial(self=Get HeaderImage() -> (HeaderImage:Image Object Reference))
   > Set HeaderDynamicMaterial(HeaderDynamicMaterial:Material Instance Dynamic Object Reference=@n2) -> (Output_Get:Material Instance Dynamic Object Reference)
   > n3 = Switch on EActivityImageLoadingState(Selection=@n1)
   SetScalarParameterValue(
    execute=@n3.Error
    self=Get HeaderDynamicMaterial() -> (HeaderDynamicMaterial:Material Instance Dynamic Object Reference)
    ParameterName=Get ThumbnailOpacity() -> (ThumbnailOpacity:Name=None)
    Value=1
   )
   > SetScalarParameterValue(self=Get HeaderDynamicMaterial(), ParameterName=IsError, Value=1)
   SetTextureParameterValue(
    execute=@n3.Loaded
    self=Get HeaderDynamicMaterial()
    ParameterName=Get ThumbnailParam() -> (ThumbnailParam:Name=None)
    Value=GetHeaderImage!(self:Creator View Model Object Reference=Get FortPlayerPageVM()) -> (ReturnValue:Texture Object Reference)
   )
   > SetScalarParameterValue(self=Get HeaderDynamicMaterial(), ParameterName=Get ThumbnailOpacity() -> (ThumbnailOpacity=None), Value=0)
   > PlayAnimationForward(InAnimation=Get ShowHeaderKeyArt() -> (ShowHeaderKeyArt:Widget Animation Object Reference))
   SetTextureParameterValue(
    execute=@n3.Unset
    self=Get HeaderDynamicMaterial()
    ParameterName=Get ThumbnailParam() -> (ThumbnailParam=None)
    Value=/Game/UI/DiscoveryShared/Art/T_UI_Discovery_DefaultEmpty
   )
   > SetScalarParameterValue(self=Get HeaderDynamicMaterial(), ParameterName=IsError, Value=0)
  ]}
 SetHeaderStateWhenNavigating {nodes [
   n1 = SetHeaderStateWhenNavigating!() -> (then, LastNavigationDirectionWasUp:bool=~, CurrentRowIndex:int, CurrentSubRowIndex:int) {is=title K2Node_FunctionEntry}
   > n2 = Branch(Condition=@n1)
   n3 = "integer > integer"(A=Get CurrentRowIndex() -> (CurrentRowIndex:int=0))
   > Switch Back and Back To Top buttons(Back To Top=@n3)
   > n4 = Branch(Condition=@n3)
   > n5 = TransitionToMiniHeader(IsMiniHeaderWithFullList=true)
   Branch(execute=[@n6.then, @n5.then], Condition=Get isFooterEnabled())
   > EnableFooter(Enable=false)
   n6 = TransitionToMiniHeader(execute=@n4.else, IsMiniHeaderWithFullList="integer > integer"(A=Get CurrentSubRowIndex() -> (CurrentSubRowIndex:int=0)))
   TransitionToMiniHeader(execute=@n2.else, IsMiniHeaderWithFullList=true)
   > Switch Back and Back To Top buttons(Back To Top=true)
  ]}
 SetShowSubscribeTooltip {nodes [
   n1 = SetShowSubscribeTooltip!() -> (then, bShowSubscribeTooltip:bool=~) {is=title K2Node_FunctionEntry}
   > TryShowingCreatorUpdatesCoachMark!(execute, self:WBP Creator Page Infos Object Reference=Get WBP_CreatorPage_Infos(), bShowCreatorTooltip:bool=@n1) -> (then)
  ]}
 SetSocialInfos {nodes [
   n1 = SetSocialInfos!() -> (then, SocialInfosMap:Map of ESocialLinkType Enums to Social Info View Model Object References=X) {is=title K2Node_FunctionEntry}
   > SetSocialInfos!(
    execute
    self:WBP Creator Page Infos Object Reference=Get WBP_CreatorPage_Infos()
    SocialInfoMap:Map of ESocialLinkType Enums to Social Info View Model Object References=@n1
    SocialLinkTypes:Array of ESocialLinkType Enums=GetSocialLinkTypes!(self:Creator View Model Object Reference=Get FortPlayerPageVM()) -> (ReturnValue:Array of ESocialLinkType Enums=X)
   ) -> (then) {is=title K2Node_CallFunction}
  ]}
 SetSupportExpirationTime {nodes [
   n1 = SetSupportExpirationTime!() -> (then, Date_Year:int=0, Date_Month:int=0, Date_Day:int=0, Date_Hour:int=0, Date_Minute:int=0, Date_Second:int=0, Date_Millisecond:int=0) {is=title K2Node_FunctionEntry}
   > Set Text(
    Text=Format Text!(
     Format="LOCTEXT(\"Supporting this creator until {Month}/{Day}/{Year}\")"
     Month:int=@n1.Date_Month
     Day:int=@n1.Date_Day
     Year:String="To String (Integer)"(InInt=@n1)
    ) -> (Result)
    self=Get Txt_Creator_Support(self=Get WBP_CreatorPage_Infos())
   )
  ]}
 SetupFooterButtonsVisuals {nodes [
   SetupFooterButtonsVisuals() {is=title K2Node_FunctionEntry}
   n1 = Get WBP_CreatorPage_Footer_Support()
   n2 = Get WBP_CreatorPage_Footer_Report()
   > SetScalarParameterValue(
    self=[
     Get BackgroundDynamicMaterial!(self:WBP UIKit Block Background Base Object Reference=Get Block_Background!(self:WBP UIKit Button Generic Base Object Reference=@n1) -> (Block_Background:WBP UIKit Block Background Base Object Reference)) -> (BackgroundDynamicMaterial:Material Instance Dynamic Object Reference)
     Get BackgroundDynamicMaterial(self=Get Block_Background(self=@n2))
    ]
    ParameterName=Radius in Pixels
    Value=1
   )
   n3 = Get WBP_CreatorPage_Footer_Report()
   n4 = Get WBP_CreatorPage_Footer_Support()
   > SetScalarParameterValue(
    self=[Get BackgroundDynamicMaterial(self=Get Block_Background(self=@n3)), Get BackgroundDynamicMaterial(self=Get Block_Background(self=@n4))]
    ParameterName=Corner Radius
    Value=16
   )
   > SetWrapping!(
    execute
    self:WBP UIKit Block Text Object Reference=[
     Get Block_SecondText!(self:WBP UIKit Button Generic Base Object Reference=Get WBP_CreatorPage_Footer_Report()) -> (Block_SecondText:WBP UIKit Block Text Object Reference)
     Get Block_SecondText(self=Get WBP_CreatorPage_Footer_Support())
    ]
    IsAutoWrap:bool=true
    WrapTextAt:int=0
   ) -> (then)
   n5 = Get OutlineDynamicMaterial!(self:WBP UIKit Block Background Base Object Reference=Get Block_Background(self=Get WBP_CreatorPage_Footer_Support())) -> (OutlineDynamicMaterial:Material Instance Dynamic Object Reference)
   n6 = Get OutlineDynamicMaterial(self=Get Block_Background(self=Get WBP_CreatorPage_Footer_Report()))
   > SetScalarParameterValue(self=[@n5, @n6], ParameterName=Radius in Pixels, Value=1)
   > SetVectorParameterValue!(execute, self=[@n5, @n6], ParameterName=Corner Radius, Value_R:float=24, Value_G:float=24, Value_B:float=24, Value_A:float=24) -> (then)
   n7 = Get WBP_CreatorPage_Footer_Report()
   > SetColorAndOpacity(
    self:User Widget Object Reference=[
     Get Block_Text!(self:WBP UIKit Button Generic Base Object Reference=@n7) -> (Block_Text:WBP UIKit Block Text Object Reference)
     Get Block_SecondText(self=@n7)
     Get Block_Icon!(self:WBP UIKit Button Generic Base Object Reference=@n7) -> (Block_Icon:WBP UIKit Block Image Object Reference)
    ]
    InColorAndOpacity:Linear Color Structure=(R=0.5776,G=0.5776,B=0.5776,A=0.85)
   )
   n8 = Get WBP_CreatorPage_Footer_Support()
   > SetColorAndOpacity(
    self=[Get Block_Text(self=@n8), Get Block_SecondText(self=@n8), Get Block_Icon(self=@n8)]
    InColorAndOpacity=(R=0.5776,G=0.5776,B=0.5776,A=0.85)
   )
  ]}
 TryShowFollowingCreatorTooltip {nodes [
   TryShowFollowingCreatorTooltip() {is=title K2Node_FunctionEntry}
   n1 = Get WBP_CreatorPage_Followed_CoachMark!(self:WBP Creator Page Infos Object Reference=Get WBP_CreatorPage_Infos()) -> (WBP_CreatorPage_Followed_CoachMark:WBP Creator Page Followed Coach Mark Object Reference)
   n2 = Get FortPlayerPageVM()
   > n3 = Branch(Condition=AND Boolean!(A=Get bAmIFollowing(self=@n2), B=ShouldShowOnboarding!(self:Fort Sidebar Onboard Tooltip Widget Object Reference=@n1) -> (ReturnValue:bool)) -> (ReturnValue))
   > ActivateWidget(self=@n1)
   Branch(execute=@n3.else, Condition=IsActivated!(self:Common Activatable Widget Object Reference=@n1) -> (ReturnValue:bool))
   > DeactivateWidget(self=@n1)
  ]}
 UpdateSupportCreatorVisuals {nodes [
   n1 = UpdateSupportCreatorVisuals!() -> (then, Has Support A Creator Enabled:bool=~) {is=title K2Node_FunctionEntry}
   > n2 = Set Has Support A Creator Enabled(Has Support A Creator Enabled:bool=@n1) -> (Output_Get:bool)
   > n3 = Branch(Condition=@n2)
   n4 = "To String (Text)"(InText=Get Name(self=Get FortPlayerPageVM()))
   > Init!(execute, self:Support A Creator View Model Object Reference=Get FortSupportCreatorVM(), InProspectSupportedCreator:String=@n4) -> (then)
   n5 = Get FortSupportCreatorVM()
   > n6 = Set isSupportingCurrentCreator(isSupportingCurrentCreator:bool="Equal Exactly (String)"(
     A=Get CurrentlySupportedCreatorName!(self:Support A Creator View Model Object Reference=@n5) -> (CurrentlySupportedCreatorName:String)
     B=@n4
    )) -> (Output_Get:bool)
   > SetVisibility(
    self=Get Txt_Creator_Support(self=Get WBP_CreatorPage_Infos())
    InVisibility=Select(Option 0=Collapsed, Option 1=SelfHitTestInvisible, Index=@n6)
   )
   > n7 = Branch(Condition=Get isSupportingCurrentCreator())
   n8 = Get WBP_BottomBarHack_SupportCreator()
   > SetTriggeringInputAction!(
    execute
    self=@n8
    InputActionRow:"Data Table Row Handle Structure (by ref)"=Get Input Action Supporting() -> (Input Action Supporting:Data Table Row Handle Structure)
   ) -> (then)
   SetTriggeringInputAction(execute=@n7.else, self=@n8, InputActionRow=Get Input Action Support() -> (Input Action Support:Data Table Row Handle Structure))
   SetBottomButtonVisibility!(execute=@n3.else, self:WBP Bottom Bar Decoy Button Object Reference=Get WBP_BottomBarHack_SupportCreator(), ShouldButtonBeVisible:bool=false) -> (then)
   > SetVisibility(self=Get Overlay_Footer(), InVisibility=Collapsed)
   Branch(Condition=AND Boolean(A=true, B=Get Has Support A Creator Enabled()))
   > SetVisibility(self=Get Support_CoachMark() -> (Support_CoachMark:Fort Visual Attachment Object Reference), InVisibility=SelfHitTestInvisible)
   > ActivateWidget(self=Get WBP_CreatorPage_Support_CoachMark() -> (WBP_CreatorPage_Support_CoachMark:WBP Creator Page Support Coach Mark Object Reference))
   > PrintString!(execute, InString=Tooltip, bPrintToScreen, bPrintToLog=false, TextColor=(R=0.9427,G=0.1253,B=0.6778,A=1), Duration=4, Key) -> (then)
   ShouldShowOnboarding()
   Get WBP_CreatorPage_Support_CoachMark()
  ]}
]
variables [
 bIsAutoSubscriptionDisabled {type=boolean}
 bShowSubscribeTooltip {type=boolean}
 Favorite_Sound {type=SoundBase, default=/Game/Sounds/Fort_UI_Sounds/FrontEnd/MultiProduct/Common/Buttons/MSS_UI_MP_Button_Click_Favorite}
 Has Support A Creator Enabled {type=boolean}
 Input Action Support {type=DataTableRowHandle, default {dataTable=&InputActionDataTable, rowName=SupportThisCreator}}
 Input Action Supporting {type=DataTableRowHandle, default {dataTable=&InputActionDataTable, rowName=SupportingCreator}}
 isSupportingCurrentCreator {type=boolean}
 LastFocusedCreatorWidget {type=Widget, default=null}
 Notification Array {type="array<ECreatorNotificationTypeVM>", default []}
 Unfavorite_Sound {type=SoundBase, default=/Game/Sounds/Fort_UI_Sounds/FrontEnd/MultiProduct/Common/Buttons/MSS_UI_MP_Button_Click_Unfavorite}
]
widgets {
 info {
  inheritedWidgetCount=0
  namedSlotCount=3
  parentClass=/DiscoveryBrowser/WBP_Discover_Surface_Base.WBP_Discover_Surface_Base_C
  rootWidgetClass=null
  widgetCount=14
 }
 widgets [
  HeaderOverlay {
   widgetClassPath=&Overlay
   bIsVariable=true
   widget=HeaderOverlay
   properties {navigation=&self:WidgetTree.HeaderOverlay.WidgetNavigation_0}
  }
  WBP_CreatorPage_Infos {
   widgetClassPath=&/WBP_CreatorPage_Infos
   bIsVariable=true
   parent=HeaderOverlay
   widget=WBP_CreatorPage_Infos
   properties {bIsFocusable=true}
   slot {class=OverlaySlot, properties {horizontalAlignment=HAlign_Fill, verticalAlignment=VAlign_Fill, padding=<ltrb 0 24 0 0>}}
  }
  Overlay_22 {widgetClassPath=&Overlay, widget=Overlay_22}
  WBP_CreatorInfo_NonCollapsible {
   widgetClassPath=&/WBP_CreatorInfo_NonCollapsible
   bIsVariable=true
   parent=Overlay_22
   widget=WBP_CreatorInfo_NonCollapsible
   slot {class=OverlaySlot, properties {horizontalAlignment=HAlign_Fill}}
  }
  Overlay_0 {widgetClassPath=&Overlay, widget=Overlay_0}
  Overlay_Footer {
   widgetClassPath=&Overlay
   bIsVariable=true
   parent=Overlay_0
   widget=Overlay_Footer
   properties {navigation=&self:WidgetTree.Overlay_Footer.WidgetNavigation_0, visibility=Collapsed}
   slot {class=OverlaySlot}
  }
  divider {
   widgetClassPath=&Image
   parent=Overlay_Footer
   widget=divider
   properties {
    brush {
     bIsDynamicallyLoaded=false
     drawAs=Image
     imageType=NoImage
     mirroring=NoMirror
     resourceName=None
     resourceObject=null
     tiling=NoTile
     imageSize=<1760 2>
     margin=<ltrb 0 0 0 0>
     outlineSettings {
      bUseBrushTransparency=false
      roundingType=HalfHeightRadius
      width=0
      color {colorUseRule=UseColor_Specified, specifiedColor=<rgba 0 0 0 0>}
      cornerRadii=<0 0 0 0>
     }
     tintColor {colorUseRule=UseColor_Specified, specifiedColor=<rgba 1 1 1 1>}
     uVRegion {bIsValid=false, max=<0 0>, min=<0 0>}
    }
    colorAndOpacity=<rgba 1 1 1 0.2>
   }
   slot {class=OverlaySlot}
  }
  FortVisualAttachment_79 {
   widgetClassPath=&FortVisualAttachment
   parent=Overlay_Footer
   widget=FortVisualAttachment_79
   properties {contentAnchor=<0 0.5>}
   slot {class=OverlaySlot, properties {horizontalAlignment=HAlign_Fill, verticalAlignment=VAlign_Center}}
  }
  HorizontalBox_Footer {
   widgetClassPath=&HorizontalBox
   bIsVariable=true
   parent=FortVisualAttachment_79
   widget=HorizontalBox_Footer
   slot {class=SizeBoxSlot}
  }
  WBP_CreatorPage_Footer_Support {
   widgetClassPath=&WBP_UIKit_Button_Generic
   bIsVariable=true
   parent=HorizontalBox_Footer
   widget=WBP_CreatorPage_Footer_Support
   properties {
    bHideInputAction=true
    bHideInputActionWithKeyboard=true
    backgroundMaterial=/FortUILibrary/UIKit/Buttons/MaterialInstances/SelectableButtons/MI_UIKit_ToggleSwitchedBackground
    contentHorizontalAlignment=HAlign_Fill
    contentVerticalAlignment=VAlign_Fill
    enableFocusedSelectedState=false
    iconHorizontalAlignment=HAlign_Left
    iconMaterial=/FortUILibrary/Textures/MaterialInstances/MI_UIKit_Icon_ItemShopCup
    iconVerticalAlignment=VAlign_Top
    maxWidth=570
    minHeight=0
    secondText="Use this creator’s code to support them with your purchases for the next 14 days"
    secondTextBlockAlignment=Top
    secondTextTransformPolicy=None
    text=Support-A-Creator
    textBlockAlignment=Bottom
    textJustification=Left
    textSecondJustification=Left
    textTransformPolicy=None
    useSecondTextBlock=true
    buttonPadding=<ltrb 32 32 32 32>
    font {
     bForceMonospaced=false
     bMaterialIsStencil=false
     fontMaterial=null
     fontObject=&HeadingNow
     letterSpacing=0
     monospacedWidth=1
     size=18
     skewAmount=0
     typefaceFontName=75 Medium
     outlineSettings {
      bApplyOutlineToDropShadows=false
      bMiteredCorners=false
      bSeparateFillAlpha=false
      outlineMaterial=null
      outlineSize=0
      outlineColor=<rgba 0 0 0 1>
     }
    }
    iconSize=<32 32>
    secondFont {
     bForceMonospaced=false
     bMaterialIsStencil=false
     fontMaterial=null
     fontObject=&HeadingNow
     letterSpacing=0
     monospacedWidth=1
     size=12
     skewAmount=0
     typefaceFontName=74 Regular
     outlineSettings {
      bApplyOutlineToDropShadows=false
      bMiteredCorners=false
      bSeparateFillAlpha=false
      outlineMaterial=null
      outlineSize=0
      outlineColor=<rgba 0 0 0 1>
     }
    }
    secondTextPadding=<ltrb 0 0 0 0>
    textPadding=<ltrb 0 12 0 12>
   }
   slot {class=HorizontalBoxSlot, properties {padding=<ltrb 0 0 24 0>}}
  }
  WBP_CreatorPage_Footer_Report {
   widgetClassPath=&WBP_UIKit_Button_Generic
   bIsVariable=true
   parent=HorizontalBox_Footer
   widget=WBP_CreatorPage_Footer_Report
   properties {
    bHideInputAction=true
    bHideInputActionWithKeyboard=true
    backgroundMaterial=&MI_UIKit_ToggleSwitchedBackground
    contentHorizontalAlignment=HAlign_Fill
    contentVerticalAlignment=VAlign_Fill
    enableFocusedSelectedState=false
    iconHorizontalAlignment=HAlign_Left
    iconMaterial=/FortUILibrary/Textures/MaterialInstances/MI_UIKit_Icon_Alert
    iconVerticalAlignment=VAlign_Top
    maxWidth=570
    pressMethod=ButtonPress
    secondText="If you feel this profile is violating community guidelines, please submit a report"
    secondTextBlockAlignment=Top
    secondTextTransformPolicy=None
    text=Report Profile
    textBlockAlignment=Bottom
    textJustification=Left
    textSecondJustification=Left
    textTransformPolicy=None
    useSecondTextBlock=true
    buttonPadding=<ltrb 32 32 32 32>
    font {
     bForceMonospaced=false
     bMaterialIsStencil=false
     fontMaterial=null
     fontObject=&HeadingNow
     letterSpacing=0
     monospacedWidth=1
     size=18
     skewAmount=0
     typefaceFontName=75 Medium
     outlineSettings {
      bApplyOutlineToDropShadows=false
      bMiteredCorners=false
      bSeparateFillAlpha=false
      outlineMaterial=null
      outlineSize=0
      outlineColor=<rgba 0 0 0 1>
     }
    }
    iconSize=<32 32>
    secondFont {
     bForceMonospaced=false
     bMaterialIsStencil=false
     fontMaterial=null
     fontObject=&HeadingNow
     letterSpacing=0
     monospacedWidth=1
     size=12
     skewAmount=0
     typefaceFontName=74 Regular
     outlineSettings {
      bApplyOutlineToDropShadows=false
      bMiteredCorners=false
      bSeparateFillAlpha=false
      outlineMaterial=null
      outlineSize=0
      outlineColor=<rgba 0 0 0 1>
     }
    }
    secondTextPadding=<ltrb 0 0 0 0>
    textPadding=<ltrb 0 12 0 12>
   }
   slot {class=HorizontalBoxSlot}
  }
  Footer_Spacer {
   widgetClassPath=&Spacer
   bIsVariable=true
   parent=Overlay_Footer
   widget=Footer_Spacer
   properties {size=<1760 300>}
   slot {class=OverlaySlot, properties {horizontalAlignment=HAlign_Fill, verticalAlignment=VAlign_Fill}}
  }
  Support_CoachMark {
   widgetClassPath=&FortVisualAttachment
   bIsVariable=true
   parent=Overlay_0
   widget=Support_CoachMark
   properties {visibility=Collapsed, contentAnchor=<1.6 1.3>}
   slot {class=OverlaySlot, properties {horizontalAlignment=HAlign_Right, verticalAlignment=VAlign_Bottom}}
  }
  WBP_CreatorPage_Support_CoachMark {
   widgetClassPath=&/WBP_CreatorPage_Support_CoachMark
   bIsVariable=true
   parent=Support_CoachMark
   widget=WBP_CreatorPage_Support_CoachMark
   properties {bIsBackHandler=false, sidebarOnboardingTooltipDefaultDurationSeconds=120, visibility=Visible}
   slot {class=SizeBoxSlot}
  }
 ]
}
