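# Creator profile info panel in the Discovery browser, showing bio, social link buttons, follow, subscribe and report controls.
# SetSocialInfos builds link buttons with mapped icons; clicks log telemetry, then open the URL or a QR-code modal, depending on platform.
# Runtime options gate the report menu, subscribe button and coach mark.
# Custom navigation rules track last-focused widgets and broadcast changes, while dispatchers relay follow and report actions.
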WidgetBlueprint /DiscoveryBrowser/CreatorProfile/WBP_CreatorPage_Infos
graphs [
 EventGraph {nodes [
   n1 = "On Clicked (Button_Follow)"!() -> (OutputDelegate, then, Button:Common Button Base Object Reference) {is=title K2Node_ComponentBoundEvent}
   > Call Handle Follow Clicked!(execute:Exec, self:WBP Creator Page Infos Object Reference, Button:Common Button Base Object Reference=@n1.Button) -> (then:Exec) {is=title K2Node_CallDelegate}
   ReportCreatorPage!() -> (OutputDelegate, then, InteractionNameID:Name) {is=title K2Node_CustomEvent}
   > Branch(Condition=GetIsCreatorReportFlowEnabled!(self:Fort Runtime Options Object Reference=GetRuntimeOptions()) -> (ReturnValue:bool))
   > Call Handle Creator Reported!(execute:Exec, self:WBP Creator Page Infos Object Reference) -> (then:Exec) {is=title K2Node_CallDelegate}
   > Close(self:Menu Anchor Object Reference=Get Anchor!(self:UIKit Context Menu Anchor Object Reference=Get Button_ContextMenu() -> (Button_ContextMenu:WBP UIKit Context Menu Anchor Button Object Reference)) -> (Anchor:Menu Anchor Object Reference))
   Event PreConstruct()
   > n2 = Construct Object from Class(Class=/DiscoveryBrowser/CreatorProfile/Data/UI_SocialLinkType_Icons_Map.UI_SocialLinkType_Icons_Map_C) -> (ReturnValue:UI Social Link Type Icons Map Object Reference)
   > Set SocialIconsMapReference(SocialIconsMapReference:UI Social Icons Structure Structure=Get SocialLinkIconsMap!(self:UI Social Link Type Icons Map Object Reference=@n2) -> (SocialLinkIconsMap:UI Social Icons Structure Structure)) -> (Output_Get:UI Social Icons Structure Structure)
   > SetMinDimensions!(
    execute
    self:Common Button Base Object Reference=Get Button_Subscribe() -> (Button_Subscribe:WBP UIKit Button Regular Object Reference)
    InMinWidth:int=80
    InMinHeight:int=80
   ) -> (then)
   Event Construct()
   > SetupContextMenu()
   > Set LastFocusedWidget(LastFocusedWidget:User Widget Object Reference=Get Button_Follow() -> (Button_Follow:WBP Creator Page Button Main Object Reference)) -> (Output_Get:User Widget Object Reference)
   > Set LastFocusedSocialButtonIndex(LastFocusedSocialButtonIndex:int=0) -> (Output_Get:int=0)
   > SetNavigationRuleCustomBoundary!(
    execute
    self:Widget Object Reference=Get EntryBox_MediaButtons() -> (EntryBox_MediaButtons:Dynamic Entry Box Object Reference)
    Direction:EUINavigation Enum=Down
    InCustomDelegate:Delegate=Create Event()
   ) -> (then)
   > SetNavigationRuleCustomBoundary(self=[Get Button_ContextMenu(), Get Button_Follow()], Direction=Up, InCustomDelegate=Create Event())
   > SetNavigationRuleCustomBoundary(self=Get Button_ContextMenu(), Direction=Left, InCustomDelegate=Create Event())
  ]}
 HandleCreatorReported {nodes [HandleCreatorReported() {is=title K2Node_FunctionEntry}]}
 HandleFollowClicked {nodes [HandleFollowClicked!() -> (then, Button:Common Button Base Object Reference) {is=title K2Node_FunctionEntry}]}
 LastFocusedUpdatedFromCreatorInfos {nodes [LastFocusedUpdatedFromCreatorInfos!() -> (then, LastFocusedWidget:User Widget Object Reference) {is=title K2Node_FunctionEntry}]}
 NavigateFromContextMenuButton {nodes [
   n1 = NavigateFromContextMenuButton!() -> (then, Navigation:EUINavigation Enum) {is=title K2Node_FunctionEntry}
   > Switch on EUINavigation!(execute, Selection:EUINavigation Enum=@n1) -> (Left:Exec, Right:Exec, Up:Exec, Down:Exec, Next:Exec, Previous:Exec, Invalid:Exec) {is=title K2Node_SwitchEnum}
   n2 = Get O_SubscribeButton() -> (O_SubscribeButton:Overlay Object Reference)
   > n3 = Set LastFocusedWidget(LastFocusedWidget=Select(
     Option 0:User Widget Object Reference=Get Button_Follow()
     Option 1:User Widget Object Reference=Get Button_Subscribe()
     Index=IsVisible(self=@n2)
    ) -> (ReturnValue:User Widget Object Reference))
   > Call Last Focused Updated from Creator Infos!(execute:Exec, self:WBP Creator Page Infos Object Reference, LastFocusedWidget:User Widget Object Reference=@n3) -> (then:Exec) {is=title K2Node_CallDelegate}
   > Return Node!(execute, ReturnValue:Widget Object Reference=@n3)
  ]}
 NavigateFromFollowButton {nodes [
   n1 = NavigateFromFollowButton!() -> (then, Navigation:EUINavigation Enum) {is=title K2Node_FunctionEntry}
   > n2 = Switch on EUINavigation(Selection=@n1)
   n3 = Set LastFocusedWidget(execute=@n2.Right, LastFocusedWidget=Get Button_ContextMenu())
   > Call Last Focused Updated from Creator Infos(LastFocusedWidget=@n3)
   > Return Node(ReturnValue=@n3)
  ]}
 NavigateFromSocialInfos {nodes [
   NavigateFromSocialInfos!() -> (then, Navigation:EUINavigation Enum) {is=title K2Node_FunctionEntry}
   > n1 = Set LastFocusedWidget(LastFocusedWidget=Get Button_Follow())
   > Call Last Focused Updated from Creator Infos(LastFocusedWidget=@n1)
   > Return Node(ReturnValue=@n1)
  ]}
 NavigateToSocialInfos {nodes [
   NavigateToSocialInfos!() -> (then, Navigation:EUINavigation Enum) {is=title K2Node_FunctionEntry}
   n1 = Get EntryBox_MediaButtons()
   > Branch(Condition="integer > integer"(A=GetNumEntries!(self:Dynamic Entry Box Base Object Reference=@n1) -> (ReturnValue:int=0)))
   > n2 = Set LastFocusedWidget(LastFocusedWidget="Get (a copy)"(
     Array:Array of User Widget Object References=GetAllEntries!(self:Dynamic Entry Box Base Object Reference=@n1) -> (ReturnValue:Array of User Widget Object References)
     Dimension 1=Get LastFocusedSocialButtonIndex() -> (LastFocusedSocialButtonIndex:int=0)
    ) -> (Output:User Widget Object Reference))
   > Call Last Focused Updated from Creator Infos(LastFocusedWidget=@n2)
   > Return Node(ReturnValue=@n2)
  ]}
 OnSocialButtonFocused {nodes [
   n1 = OnSocialButtonFocused!() -> (then, FocusedEntryBoxIndex:int) {is=title K2Node_FunctionEntry}
   > Set LastFocusedSocialButtonIndex(LastFocusedSocialButtonIndex=@n1) -> (Output_Get=0)
   n2 = Get EntryBox_MediaButtons()
   > n3 = Set LastFocusedWidget(LastFocusedWidget="Get (a copy)"(Array=GetAllEntries(self=@n2), Dimension 1=Get LastFocusedSocialButtonIndex() -> (LastFocusedSocialButtonIndex=0)))
   > Call Last Focused Updated from Creator Infos(LastFocusedWidget=@n3)
  ]}
 Open Social Info Dialog {nodes [
   n1 = Open Social Info Dialog!() -> (then, Button:Common Button Base Object Reference) {is=title K2Node_FunctionEntry}
   > n2 = Cast To WBP_CreatorPage_Button_Link!(execute, Object=@n1) -> (then, CastFailed, AsWBP Creator Page Button Link:WBP Creator Page Button Link Object Reference) {is=title K2Node_DynamicCast}
   > n3 = FindParentWidgetOfType!(
    execute
    StartingWidget:Widget Object Reference=Self-Reference()
    Type:Widget Class Reference=&/WBP_FollowCreator_ProfilePage.WBP_FollowCreator_ProfilePage_C
   ) -> (then, ReturnValue:WBP Follow Creator Profile Page Object Reference)
   n4 = Get Platform!(self:WBP Creator Page Button Link Object Reference=@n2) -> (Platform:String)
   > Fire Event Element Interacted(
    Element=@n1
    AdditionalInfo="Add String (Non-Localized)"!(
     Info:"Flow Analytics Additional Info Structure (by ref)"=GetCreatorIDForTelemetry!(self:WBP Follow Creator Profile Page Object Reference=@n3) -> (ReturnValue:Flow Analytics Additional Info Structure)
     Key:String=SocialPlatform
     Value:String=@n4
    ) -> (ReturnValue:Flow Analytics Additional Info Structure)
   )
   n5 = ShouldOpenBrowserForCreatorSocialLinksOnPC!(self:Fort Runtime Options Object Reference=GetRuntimeOptions()) -> (ReturnValue:bool)
   n6 = "Equal Exactly (String)"(A=GetPlatformName!() -> (ReturnValue:String), B=Windows)
   > n7 = Branch(Condition=OR Boolean(A=IsMobileGame(), B=AND Boolean(A=@n5, B=@n6)))
   n8 = Get URL!(self:WBP Creator Page Button Link Object Reference=@n2) -> (URL:String)
   > OpenSocialUrl!(
    execute
    self:Fort Player Pages Sub System Object Reference=Get FortPlayerPagesSubSystem() -> (ReturnValue:Fort Player Pages Sub System Object Reference) {is=title K2Node_GetSubsystem}
    InUrl:String=@n8
   ) -> (then)
   n9 = Create Widget!(
    execute:Exec=@n7.else
    Class:User Widget Class Reference=&/WBP_CreatorPage_QRCode_Modal.WBP_CreatorPage_QRCode_Modal_C
    OwningPlayer:Player Controller Object Reference=GetOwningPlayer()
   ) -> (then:Exec, ReturnValue:WBP Creator Page QRCode Modal Object Reference)
   > SetCreatorData!(
    execute
    self:WBP Creator Page QRCode Modal Object Reference=@n9
    Creator Name:Text=Get CreatorName() -> (CreatorName:Text)
    Social Platform:String=@n4
    URL:String=@n8
   ) -> (then)
   > TryPushWidgetAsModalPanel!(execute, PushWidget:User Widget Object Reference=@n9) -> (then)
  ]}
 SetButtonSubscribeIcon {nodes [
   n1 = SetButtonSubscribeIcon!() -> (then, IsSubscribed:bool=~) {is=title K2Node_FunctionEntry}
   > SetMaterial!(
    execute
    self:WBP UIKit Block Image Object Reference=Get Block_Icon!(self:WBP UIKit Button CTA Base Object Reference=Get Button_Subscribe()) -> (Block_Icon:WBP UIKit Block Image Object Reference)
    InMaterial:"Material Instance Object Reference (by ref)"=SelectMaterialInstance!(
     A:Material Instance Object Reference=/DiscoveryBrowser/CreatorProfile/MaterialInstances/MI_CreatorPage_Button_Icon_NotificationOn
     B:Material Instance Object Reference=/DiscoveryBrowser/CreatorProfile/MaterialInstances/MI_CreatorPage_Button_Icon_NotificationOff
     bSelectA:bool=@n1
    ) -> (ReturnValue:Material Instance Object Reference)
   ) -> (then)
  ]}
 SetButtonSubscribeVisibility {nodes [
   n1 = SetButtonSubscribeVisibility!() -> (then, IsFollowing:bool=~) {is=title K2Node_FunctionEntry}
   n2 = GetIsCreatorSubscriptionEnabled!(self:Fort Runtime Options Object Reference=GetRuntimeOptions()) -> (ReturnValue:bool)
   > SetVisibility(
    self=Get O_SubscribeButton()
    InVisibility=Select(Option 0:ESlateVisibility Enum=Collapsed, Option 1:ESlateVisibility Enum=SelfHitTestInvisible, Index=AND Boolean(A=@n1, B=@n2)) -> (ReturnValue:ESlateVisibility Enum)
   )
  ]}
 SetCanBeReported {nodes [
   n1 = SetCanBeReported!() -> (then, Can Be Reported:bool=~) {is=title K2Node_FunctionEntry}
   > SetVisibility(self=Get Button_ContextMenu(), InVisibility=Select(Option 0=Collapsed, Option 1=Visible, Index=@n1))
  ]}
 SetSocialInfos {nodes [
   n1 = SetSocialInfos!() -> (
    then
    SocialInfoMap:Map of ESocialLinkType Enums to Social Info View Model Object References=X
    SocialLinkTypes:Array of ESocialLinkType Enums=X
   ) {is=title K2Node_FunctionEntry}
   n2 = Get EntryBox_MediaButtons()
   > Reset!(execute, self:Dynamic Entry Box Object Reference=@n2, bDeleteWidgets:bool=true) -> (then)
   > n3 = For Each Loop(Array:Array of ESocialLinkType Enums=@n1.SocialLinkTypes) -> (Array Element:ESocialLinkType Enum)
   > n4 = Create Entry of Class!(
    execute
    self:Dynamic Entry Box Object Reference=@n2
    EntryClass:User Widget Class Reference=&/WBP_CreatorPage_Button_Link.WBP_CreatorPage_Button_Link_C
   ) -> (then, ReturnValue:WBP Creator Page Button Link Object Reference)
   > Set EntryBoxIndex!(execute, EntryBoxIndex:int=@n3.Array Index, self:WBP Creator Page Button Link Object Reference=@n4) -> (then, Output_Get:int=0)
   > Set Platform!(
    execute
    Platform:String=Enum to String!(Enumerator:ESocialLinkType Enum=@n3) -> (ReturnValue:String)
    self:WBP Creator Page Button Link Object Reference=@n4
   ) -> (then, Output_Get:String)
   > Set URL!(
    execute
    URL:String=Get Url!(self:Social Info View Model Object Reference=Find(TargetMap:Map of ESocialLinkType Enums to Social Info View Model Object References=@n1, Key:"ESocialLinkType Enum (by ref)"=@n3) -> (Value:Social Info View Model Object Reference)) -> (Url:String)
    self:WBP Creator Page Button Link Object Reference=@n4
   ) -> (then, Output_Get:String)
   n5 = Get SocialIconsMapReference() -> (SocialIconsMapReference:UI Social Icons Structure Structure)
   > SetIconMaterial!(
    execute
    self:WBP UIKit Button Generic Base Object Reference=@n4
    Material Instance:Material Instance Object Reference=Find(
     TargetMap:Map of ESocialLinkType Enums to Material Instance Object References=Break UI Social Icons Structure!(UI_SocialIcons_Structure:"UI Social Icons Structure Structure (by ref)"=@n5) -> (SocialIconsMap_5_#1:Map of ESocialLinkType Enums to Material Instance Object References) {is=title K2Node_BreakStruct}
     Key=@n3
    ) -> (Value:Material Instance Object Reference)
   ) -> (then)
   > Bind Event to On Clicked(self:Common Button Base Object Reference=@n4, Delegate=Create Event()) {is=title K2Node_AddDelegate}
   > Bind Event to On Social Button Focused(self:WBP Creator Page Button Link Object Reference=@n4, Delegate=Create Event()) {is=title K2Node_AddDelegate}
  ]}
 SetupContextMenu {nodes [
   SetupContextMenu() {is=title K2Node_FunctionEntry}
   > Branch(Condition=GetIsCreatorReportFlowEnabled(self=GetRuntimeOptions()))
   > n1 = Construct Object from Class(Class=/Script/FortniteUI.UIKitContextMenuActionContext) -> (ReturnValue:UIKit Context Menu Action Context Object Reference)
   > Set ContextMenuActionContext(ContextMenuActionContext:UIKit Context Menu Action Context Object Reference=@n1) -> (Output_Get:UIKit Context Menu Action Context Object Reference)
   > n2 = Sequence()
   > AddInteraction!(
    execute
    self:UIKit Context Menu Action Context Object Reference=Get ContextMenuActionContext() -> (ContextMenuActionContext:UIKit Context Menu Action Context Object Reference)
    InActionInfo_DisplayName:Text="LOCTEXT(\"Report Profile\")"
    InActionInfo_IconTexture:Texture 2D Soft Object Reference
    InActionInfo_IconMaterial:Material Instance Soft Object Reference
    InActionInfo_SortPriority:int=1000
    InActionInfo_bIsAvailable:bool=True
    InActionInfo_bIsEnabled:bool=True
    InActionInfo_bIsSelected:bool=False
    InActionInfo_bAddDivider:bool=False
    InActionInfo_bPositionDividerAfter:bool=False
    InActionInfo_bCloseMenuOnSelect:bool=False
    InActionInfo_InteractionNameID:Name=ReportCreator
    InActionInfo_Interaction:Data Table Row Handle Structure=(DataTable=None,RowName="")
    InActionInfo_PressedDelegate:Delegate="(null).None"
    InActionInfo_ReleasedDelegate:Delegate="(null).None"
    InInteractionDelegate:"Delegate (by ref)"=Create Event()
   ) -> (then)
   SetContext!(
    execute=@n2.then_1
    self:WBP UIKit Context Menu Anchor Button Base Object Reference=Get Button_ContextMenu()
    InteractionContext_Data:UIKit Context Menu Action Context Object Reference=Get ContextMenuActionContext()
   ) -> (then)
  ]}
 TryShowingCreatorUpdatesCoachMark {nodes [
   n1 = TryShowingCreatorUpdatesCoachMark!() -> (then, bShowCreatorTooltip:bool=~) {is=title K2Node_FunctionEntry}
   n2 = Get WBP_CreatorPage_Subscribe_CoachMark() -> (WBP_CreatorPage_Subscribe_CoachMark:WBP Creator Page Subscribe Coach Mark Object Reference)
   n3 = GetRuntimeOptions()
   > n4 = Branch(Condition=AND Boolean!(
     A=@n1
     B=ShouldShowOnboarding!(self:Fort Sidebar Onboard Tooltip Widget Object Reference=@n2) -> (ReturnValue:bool)
     C:bool=GetIsCreatorSubscriptionEnabled(self=@n3)
    ) -> (ReturnValue))
   n5 = Get WBP_CreatorPage_Subscribe_CoachMark()
   > ActivateWidget(self=@n5)
   Branch(
    execute=@n4.else
    Condition=IsActivated!(self:Common Activatable Widget Object Reference=Get WBP_CreatorPage_Subscribe_CoachMark()) -> (ReturnValue:bool)
   )
   > DeactivateWidget(self=@n5)
  ]}
]
variables [
 AllowShowingCreatorTooltip {type=boolean}
 ContextMenuActionContext {type=UIKitContextMenuActionContext, default=null}
 CreatorName {type=string, default=""}
 HandleCreatorReported {}
 HandleFollowClicked {}
 LastFocusedSocialButtonIndex {type=integer}
 LastFocusedUpdatedFromCreatorInfos {}
 LastFocusedWidget {type=UserWidget, default=null}
 SocialIconsMapReference {type=UI_SocialIcons_Structure, default {socialIconsMap {}}}
]
widgets {
 info {inheritedWidgetCount=0, parentClass=&UserWidget, rootWidgetClass=&GridPanel, widgetCount=15}
 widgets [
  GridRoot {widgetClassPath=&GridPanel, widget=GridRoot, properties {columnFill [1], rowFill [0]}}
  VB_MiniHeaderCollapsibleArea {
   widgetClassPath=&VerticalBox
   bIsVariable=true
   parent=GridRoot
   widget=VB_MiniHeaderCollapsibleArea
   slot {class=GridSlot, properties {row=1, rowSpan=1, verticalAlignment=VAlign_Top}}
  }
  EntryBox_MediaButtons {
   widgetClassPath=&DynamicEntryBox
   bIsVariable=true
   parent=VB_MiniHeaderCollapsibleArea
   widget=EntryBox_MediaButtons
   properties {
    entryHorizontalAlignment=HAlign_Left
    entryWidgetClass=&WBP_CreatorPage_Button_Link_C
    navigation=&self:WidgetTree.EntryBox_MediaButtons.WidgetNavigation_0
    numDesignerPreviewEntries=5
    entrySpacing=<16 0>
   }
   slot {class=VerticalBoxSlot, properties {padding=<ltrb 0 0 0 12>}}
  }
  Txt_Bio {
   widgetClassPath=&CommonTextBlock
   bIsVariable=true
   parent=VB_MiniHeaderCollapsibleArea
   widget=Txt_Bio
   properties {
    applyLineHeightToBottomLine=false
    autoWrapText=true
    text=Creator Bio Text BlockText BlockText BlockText BlockText BlockText BlockText BlockText Block
    wrapTextAt=952
    colorAndOpacity {colorUseRule=UseColor_Specified, specifiedColor=<rgba 0.5776 0.5776 0.5776 1>}
    font {
     bForceMonospaced=false
     bMaterialIsStencil=false
     fontMaterial=null
     fontObject=&HeadingNow
     letterSpacing=10
     monospacedWidth=1
     size=18
     skewAmount=0
     typefaceFontName=74 Regular
     outlineSettings {
      bApplyOutlineToDropShadows=false
      bMiteredCorners=false
      bSeparateFillAlpha=false
      outlineMaterial=/FortUILibrary/MaterialInstances/Text/MI_UI_Block_Text_Outline
      outlineSize=2
      outlineColor=<rgba 0 0 0 1>
     }
    }
    shadowOffset=<0 0>
   }
   slot {class=VerticalBoxSlot, properties {padding=<ltrb 0 0 0 24>}}
  }
  HB_Buttons {
   widgetClassPath=&HorizontalBox
   parent=VB_MiniHeaderCollapsibleArea
   widget=HB_Buttons
   slot {class=VerticalBoxSlot, properties {padding=<ltrb 0 0 0 12>}}
  }
  O_FollowButton {
   widgetClassPath=&Overlay
   parent=HB_Buttons
   widget=O_FollowButton
   slot {class=HorizontalBoxSlot, properties {horizontalAlignment=HAlign_Left, verticalAlignment=VAlign_Bottom}}
  }
  Button_Follow {
   widgetClassPath=&/WBP_CreatorPage_Button_Main
   bIsVariable=true
   parent=O_FollowButton
   widget=Button_Follow
   properties {
    backgroundMaterial=/DiscoveryBrowser/CreatorProfile/MaterialInstances/MI_CreatorPage_Button_Main_Background
    minHeight=88
    minHeightOverride=80
    minWidth=388
    minWidthOverride=400
    mobileFontSizeMultiplier=1.3
    mobileMinHeight=80
    text=Following
    iconSize=<48 48>
   }
   slot {class=OverlaySlot, properties {horizontalAlignment=HAlign_Fill, verticalAlignment=VAlign_Fill}}
  }
  VisualAttachment_Tooltip {
   widgetClassPath=&FortVisualAttachment
   parent=O_FollowButton
   widget=VisualAttachment_Tooltip
   slot {class=OverlaySlot, properties {horizontalAlignment=HAlign_Right}}
  }
  WBP_CreatorPage_Followed_CoachMark {
   widgetClassPath=&/WBP_CreatorPage_Followed_CoachMark
   bIsVariable=true
   parent=VisualAttachment_Tooltip
   widget=WBP_CreatorPage_Followed_CoachMark
   properties {
    navigation=&self:WidgetTree.WBP_CreatorPage_Followed_CoachMark.WidgetNavigation_0
    desiredFocusWidget {widgetName=None}
   }
   slot {class=SizeBoxSlot, properties {padding=<ltrb 32 -24 0 0>}}
  }
  O_SubscribeButton {
   widgetClassPath=&Overlay
   bIsVariable=true
   parent=HB_Buttons
   widget=O_SubscribeButton
   slot {class=HorizontalBoxSlot, properties {padding=<ltrb 24 0 0 0>}}
  }
  Button_Subscribe {
   widgetClassPath=&WBP_UIKit_Button_Regular
   bIsVariable=true
   parent=O_SubscribeButton
   widget=Button_Subscribe
   properties {
    bHideInputAction=true
    bUseNaturalNavigation=true
    iconMaterial=&MI_CreatorPage_Button_Icon_NotificationOff
    minHeight=56
    minHeightOverride=80
    minWidth=56
    minWidthOverride=80
    navigation=&self:WidgetTree.Button_Subscribe.WidgetNavigation_0
    text=""
   }
   slot {class=OverlaySlot, properties {verticalAlignment=VAlign_Center}}
  }
  VA_SubscribedTooltip {
   widgetClassPath=&FortVisualAttachment
   parent=O_SubscribeButton
   widget=VA_SubscribedTooltip
   slot {class=OverlaySlot, properties {horizontalAlignment=HAlign_Right}}
  }
  WBP_CreatorPage_Subscribe_CoachMark {
   widgetClassPath=&/WBP_CreatorPage_Subscribe_CoachMark
   bIsVariable=true
   parent=VA_SubscribedTooltip
   widget=WBP_CreatorPage_Subscribe_CoachMark
   properties {navigation=&self:WidgetTree.WBP_CreatorPage_Subscribe_CoachMark.WidgetNavigation_0}
   slot {class=SizeBoxSlot, properties {padding=<ltrb 25 -24 0 0>}}
  }
  Button_ContextMenu {
   widgetClassPath=/FortUILibrary/UIKit/ContextMenu/WBP_UIKit_ContextMenuAnchorButton
   bIsVariable=true
   parent=HB_Buttons
   widget=Button_ContextMenu
   properties {
    bIsFocusable=true
    buttonClass=/Game/UI/DiscoveryShared/GameDetails/WBP_GameDetails_Details.WBP_GameDetails_Details_C
    buttonIcon=/FortUILibrary/Textures/Icons/T_UI_Icon_UIKit_ContextMenu
    buttonText=""
    min Height=80
    min Width=80
    navigation=&self:WidgetTree.Button_ContextMenu.WidgetNavigation_0
    useBottomBarButtons=true
    visibility=HitTestInvisible
    button Padding=<ltrb 8 8 8 8>
   }
   slot {class=HorizontalBoxSlot, properties {horizontalAlignment=HAlign_Left, verticalAlignment=VAlign_Center, padding=<ltrb 24 0 0 0>}}
  }
  Txt_Creator_Support {
   widgetClassPath=&CommonTextBlock
   bIsVariable=true
   parent=VB_MiniHeaderCollapsibleArea
   widget=Txt_Creator_Support
   properties {
    applyLineHeightToBottomLine=false
    autoWrapText=true
    text="Supporting this creator until {date}"
    visibility=Collapsed
    wrapTextAt=952
    colorAndOpacity {colorUseRule=UseColor_Specified, specifiedColor=<rgba 0.5776 0.5776 0.5776 1>}
    font {
     bForceMonospaced=false
     bMaterialIsStencil=false
     fontMaterial=null
     fontObject=&HeadingNow
     letterSpacing=10
     monospacedWidth=1
     size=11.25
     skewAmount=0
     typefaceFontName=74 Regular
     outlineSettings {
      bApplyOutlineToDropShadows=false
      bMiteredCorners=false
      bSeparateFillAlpha=false
      outlineMaterial=&MI_UI_Block_Text_Outline
      outlineSize=2
      outlineColor=<rgba 0 0 0 1>
     }
    }
    shadowOffset=<0 0>
   }
   slot {class=VerticalBoxSlot}
  }
 ]
}
